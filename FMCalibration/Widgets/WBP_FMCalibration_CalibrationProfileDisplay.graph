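# CommonUI widget that shows a right-aligned settings icon and label above the name of the active Tune-Up (calibration) profile.
# SetActiveProfileIndex stores the index and calls RefreshDisplayedText, which formats it into the localized "Tune-Up Profile {x}" label.
# UpdatePromptVisibilty toggles the widget between hidden and visible without hit-testing itself, based on whether the calibration profile feature has been used.
# On Construct, it hides itself if the CalibrationProfileVM view model is invalid.
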WidgetBlueprint /FMCalibration/Widgets/WBP_FMCalibration_CalibrationProfileDisplay
graphs [
 EventGraph {nodes [
   Event Construct()
   > n1 = Get!(execute:Exec) -> (then:Exec, else:Exec, CalibrationProfileVM:Calibration Profile VM Object Reference)
   Set Visibility!(execute=@n1.else, Visibility=Hidden) -> (then, Output_Get)
  ]}
 RefreshDisplayedText {nodes [
   RefreshDisplayedText() {is=title K2Node_FunctionEntry}
   n1 = Get Profile Index() -> (Profile Index:int=0)
   > Set Text(
    Text=Format Text!(Format="LOCTEXT(\"Tune-Up Profile {x}\")", x:int=int + int(A=@n1)) -> (Result)
    self=Get ProfileName_Label() -> (ProfileName_Label:Common Text Object Reference)
   )
  ]}
 SetActiveProfileIndex {nodes [
   n1 = SetActiveProfileIndex!() -> (then, ProfileIndex:int) {is=title K2Node_FunctionEntry}
   > Set Profile Index(Profile Index:int=@n1) -> (Output_Get:int=0)
   > RefreshDisplayedText() {is=title K2Node_CallFunction}
  ]}
 UpdatePromptVisibilty {nodes [
   n1 = UpdatePromptVisibilty!() -> (then, HasUsedCaibrationProfileFeature:bool=~) {is=title K2Node_FunctionEntry}
   > Set Visibility(Visibility=Select(Option 0=Hidden, Option 1=SelfHitTestInvisible, Index=@n1))
  ]}
]
variables [Profile Index {type=integer}]
widgets {
 info {inheritedWidgetCount=0, parentClass=&CommonUserWidget, rootWidgetClass=&VerticalBox, widgetCount=5}
 widgets [
  VerticalBox_0 {widgetClassPath=&VerticalBox, widget=VerticalBox_0}
  HorizontalBox_0 {
   widgetClassPath=&HorizontalBox
   parent=VerticalBox_0
   widget=HorizontalBox_0
   slot {class=VerticalBoxSlot, properties {horizontalAlignment=HAlign_Right}}
  }
  Image_187 {
   widgetClassPath=&Image
   bIsVariable=true
   parent=HorizontalBox_0
   widget=Image_187
   properties {brush {
     bIsDynamicallyLoaded=false
     drawAs=Image
     imageType=FullColor
     mirroring=NoMirror
     resourceName=None
     resourceObject=/FortUILibrary/Textures/Icons/T_UI_Icon_UIKit_ControllerSettings
     tiling=NoTile
     imageSize=<42 42>
     margin=<ltrb 0 0 0 0>
     outlineSettings {
      bUseBrushTransparency=false
      roundingType=HalfHeightRadius
      width=0
      color {colorUseRule=UseColor_Specified, specifiedColor=<rgba 0 0 0 0>}
      cornerRadii=<0 0 0 0>
     }
     tintColor {colorUseRule=UseColor_Specified, specifiedColor=<rgba 1 1 1 1>}
     uVRegion {bIsValid=false, max=<0 0>, min=<0 0>}
    }}
   slot {class=HorizontalBoxSlot, properties {horizontalAlignment=HAlign_Center, verticalAlignment=VAlign_Center, padding=<ltrb 0 0 8 0>}}
  }
  Seting_Label {
   widgetClassPath=&CommonTextBlock
   parent=HorizontalBox_0
   widget=Seting_Label
   properties {
    bIsScrollingEnabled=false
    text=SETTING
    font {
     bForceMonospaced=false
     bMaterialIsStencil=false
     fontMaterial=null
     fontObject=null
     letterSpacing=0
     monospacedWidth=1
     size=24
     skewAmount=0
     typefaceFontName=None
     outlineSettings {
      bApplyOutlineToDropShadows=false
      bMiteredCorners=false
      bSeparateFillAlpha=false
      outlineMaterial=null
      outlineSize=0
      outlineColor=<rgba 0 0 0 1>
     }
    }
    shadowOffset=<0 0>
   }
   slot {class=HorizontalBoxSlot, properties {verticalAlignment=VAlign_Center}}
  }
  ProfileName_Label {
   widgetClassPath=&CommonTextBlock
   bIsVariable=true
   parent=VerticalBox_0
   widget=ProfileName_Label
   properties {
    justification=Right
    text=Tune-Up Profile X
    font {
     bForceMonospaced=false
     bMaterialIsStencil=false
     fontMaterial=null
     fontObject=&HeadingNow
     letterSpacing=0
     monospacedWidth=1
     size=18
     skewAmount=0
     typefaceFontName=74 Regular
     outlineSettings {
      bApplyOutlineToDropShadows=false
      bMiteredCorners=false
      bSeparateFillAlpha=false
      outlineMaterial=null
      outlineSize=0
      outlineColor=<rgba 0 0 0 1>
     }
    }
    shadowOffset=<0 0>
   }
   slot {class=VerticalBoxSlot, properties {horizontalAlignment=HAlign_Right}}
  }
 ]
}
