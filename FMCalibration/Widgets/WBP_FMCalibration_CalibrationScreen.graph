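# Widget-tree-less calibration screen Blueprint for FM Calibration, providing shared setup, navigation and close logic.
# On Construct it registers itself by class with the player's calibration component, caches the experience tag, broadcasts a modal-opened event and collapses itself.
# While active it listens for a stateful force-close event, which ends the CalibrationProcessVM process and closes the modal.
# Navigation functions dispatch UI state chart transition tags, using backtrack to leave results, error and manual screens.
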WidgetBlueprint /FMCalibration/Widgets/WBP_FMCalibration_CalibrationScreen
graphs [
 BackFromResultsScreen {nodes [
   BackFromResultsScreen() {is=title K2Node_FunctionEntry}
   > DispatchTransitionEvent!(
    execute
    self:UIState Chart Manager Object Reference=Get UIStateChartManager() -> (ReturnValue:UIState Chart Manager Object Reference) {is=title K2Node_GetSubsystem}
    TransitionTag:Gameplay Tag Structure=(TagName="UI.Chart.TransitionTag.FM.Calibration.OpenResultScreen")
    StateChartOwner:Local Player Object Reference=GetOwningLocalPlayer!(self:Widget Object Reference) -> (ReturnValue:Local Player Object Reference)
    AnimationHints:"Gameplay Tag Container Structure (by ref)"
    ContextChanges:"UIState Chart Context Change Set Structure (by ref)"
    Urgency:EUIStateTransitionUrgency Enum=Normal
    bBacktrack:bool=true
   ) -> (then, ReturnValue:bool)
  ]}
 CloseCalibration {nodes [
   n1 = CloseCalibration!() -> (then, CalibrationProcessVM:Calibration Process VM Object Reference) {is=title K2Node_FunctionEntry}
   > EndProcess(self:Calibration Process VM Object Reference=@n1)
   > CloseCalibrationModal!(
    execute
    self:BP FMCalibration Component Object Reference=GetComponentByClass!(
     self:Actor Object Reference=GetOwningPlayer()
     ComponentClass:Actor Component Class Reference=/FMCalibration/Components/BP_FMCalibrationComponent.BP_FMCalibrationComponent_C
    ) -> (ReturnValue:BP FMCalibration Component Object Reference)
    CloseToQuestMenu:bool=false
   ) -> (then)
  ]}
 CloseCalibrationToQuestMenu {nodes [
   n1 = CloseCalibrationToQuestMenu!() -> (then, CalibrationProcessVM:Calibration Process VM Object Reference) {is=title K2Node_FunctionEntry}
   > EndProcess(self=@n1)
   > CloseCalibrationModal(
    self=GetComponentByClass(self=GetOwningPlayer(), ComponentClass=&BP_FMCalibrationComponent_C)
    CloseToQuestMenu=true
   )
  ]}
 CloseErrorScreen {nodes [
   CloseErrorScreen() {is=title K2Node_FunctionEntry}
   > DispatchTransitionEvent(
    self=Get UIStateChartManager()
    TransitionTag=(TagName="UI.Chart.TransitionTag.FM.Calibration.OpenErrorScreen")
    StateChartOwner=GetOwningLocalPlayer()
    Urgency=Normal
    bBacktrack=true
   )
  ]}
 CloseManualCalibration {nodes [
   CloseManualCalibration() {is=title K2Node_FunctionEntry}
   > DispatchTransitionEvent(
    self=Get UIStateChartManager()
    TransitionTag=(TagName="UI.Chart.TransitionTag.FM.Calibration.OpenManualCalibration")
    StateChartOwner=GetOwningLocalPlayer()
    Urgency=Normal
    bBacktrack=true
   )
  ]}
 EventGraph {nodes [
   Event On Deactivated()
   > Parent: On Deactivated() {is=title K2Node_CallParentFunction}
   > StopListeningToStatefulEvent(self:Async Action Start Listening To Stateful Event Object Reference=Get ForceCalibrationClosedEventRef() -> (ForceCalibrationClosedEventRef:Async Action Start Listening To Stateful Event Object Reference))
   Event On Activated()
   > Parent: On Activated() {is=title K2Node_CallParentFunction}
   > SetAllNavigationRules!(execute, self:Widget Object Reference, Rule:EUINavigationRule Enum=Stop, WidgetToFocus:Name=None) -> (then)
   > n1 = StartListeningToStatefulEvent!(
    execute:Exec
    Target:Gameplay Event Router Component Object Reference=GetGlobalEventRouter()
    EventType:Script Struct Object Reference=/Script/FMCalibrationRuntime.FMEvent_Calibration_ForceClose
    Context:Object Reference
    EventBubblingRule:EEventBubblingRule Enum
   ) -> (
    then:Exec
    AsyncTaskProxy:Async Action Start Listening To Stateful Event Object Reference
    OnEventReceived:Exec
    OnSavedEventStateExists:Exec
    OnEventStateCleared:Exec
    Payload:FMEvent Calibration Force Close Structure
   ) {is=title K2Node_StartListeningToStatefulEventAsyncAction}
   > Set ForceCalibrationClosedEventRef(ForceCalibrationClosedEventRef:Async Action Start Listening To Stateful Event Object Reference=@n1) -> (Output_Get:Async Action Start Listening To Stateful Event Object Reference)
   Branch(
    execute=[@n1.OnEventReceived, @n1.OnSavedEventStateExists]
    Condition=Break FMEvent Calibration Force Close!(FMEvent_Calibration_ForceClose:"FMEvent Calibration Force Close Structure (by ref)"=@n1.Payload) -> (isClosed:bool=~) {is=title K2Node_BreakStruct}
   )
   n2 = Get MVVMSubsystem() -> (ReturnValue:Viewmodel Engine Subsytem Object Reference) {is=title K2Node_GetEngineSubsystem}
   n3 = Self-Reference()
   n4 = Cast To CalibrationProcessVM!(Object:Notify Field Value Changed Interface=GetViewModel!(
     self:MVVM View Object Reference=Get View From User Widget!(self:Viewmodel Engine Subsytem Object Reference=@n2, UserWidget:User Widget Object Reference=@n3) -> (ReturnValue:MVVM View Object Reference) {is=title K2Node_CallFunction}
     ViewModelName:Name=CalibrationProcessVM
    ) -> (ReturnValue:Notify Field Value Changed Interface)) -> (AsCalibration Process VM:Calibration Process VM Object Reference, bSuccess:bool=~) {is=title K2Node_DynamicCast}
   > Branch(Condition=@n4.bSuccess)
   > CloseCalibration!(execute, self, CalibrationProcessVM:Calibration Process VM Object Reference=@n4) -> (then) {is=title K2Node_CallFunction}
   Event Construct()
   > Initialize FMCalibration()
   > SetVisibility(InVisibility=Collapsed)
  ]}
 Initialize FMCalibration {nodes [
   n1 = GetComponentByClass(self=GetOwningPlayer(), ComponentClass=&BP_FMCalibrationComponent_C)
   n2 = Consume Wants Preamble!(execute, self:BP FMCalibration Component Object Reference=@n1) -> (then, ReturnValue:bool)
   > Set ShouldUsePreamble(ShouldUsePreamble:bool=@n2) -> (Output_Get:bool)
   Initialize FMCalibration() {is=title K2Node_FunctionEntry}
   > n3 = Sequence!(execute) -> (then_0, then_1, then_2:Exec)
   n4 = GetEventRouter!(execute=@n3.then_1, Scope:Object Class Reference=/Script/FortniteGame.FortPlayspace, ContextActor:Actor Object Reference=GetOwningPlayer()) -> (then, ReturnValue:Gameplay Event Router Component Object Reference)
   > Broadcast Event!(
    execute
    self:Gameplay Event Router Component Object Reference=@n4
    EventData:FMEvent Calibration Modal Opened Structure=Make FMEvent Calibration Modal Opened!() -> (FMEvent_Calibration_ModalOpened:FMEvent Calibration Modal Opened Structure) {is=title K2Node_MakeStruct}
    EventContext:Object Reference
   ) -> (then, ReturnValue:bool)
   Is Valid(exec=@n3.then_0, InputObject=@n1)
   > n5 = Set FMCalibrationComponent(FMCalibrationComponent:BP FMCalibration Component Object Reference=@n1) -> (Output_Get:BP FMCalibration Component Object Reference)
   n6 = Self-Reference()
   > Add!(
    execute
    TargetMap:Map of Widget Class References to Widget Object References=Get CalibrationWidgets!(self:BP FMCalibration Component Object Reference=@n5) -> (CalibrationWidgets:Map of Widget Class References to Widget Object References)
    Key:"Widget Class Reference (by ref)"=Get Class(Object=@n6) -> (ReturnValue:WBP FMCalibration Calibration Screen Class Reference)
    Value:"Widget Object Reference (by ref)"=@n6
   ) -> (then) {is=title K2Node_CallFunction}
   > Set ExperienceMode(ExperienceMode:Gameplay Tag Structure=Get Experience Tag!(self:BP FMCalibration Component Object Reference=@n1) -> (Experience Tag:Gameplay Tag Structure)) -> (Output_Get:Gameplay Tag Structure)
  ]}
 NextCalibrationScreen {nodes [
   NextCalibrationScreen() {is=title K2Node_FunctionEntry}
   > DispatchTransitionEvent(
    self=Get UIStateChartManager()
    TransitionTag=(TagName="UI.Chart.TransitionTag.Generic.Advance")
    StateChartOwner=GetOwningLocalPlayer()
    Urgency=Normal
    bBacktrack=false
   )
  ]}
 OpenAVCalibrationScreen {nodes [
   OpenAVCalibrationScreen() {is=title K2Node_FunctionEntry}
   > DispatchTransitionEvent(
    self=Get UIStateChartManager()
    TransitionTag=(TagName="UI.Chart.TransitionTag.FM.Calibration.OpenAVCalibration")
    StateChartOwner=GetOwningLocalPlayer()
    Urgency=Normal
    bBacktrack=false
   )
  ]}
 OpenResultsSceen {nodes [
   OpenResultsSceen() {is=title K2Node_FunctionEntry}
   > DispatchTransitionEvent(
    self=Get UIStateChartManager()
    TransitionTag=(TagName="UI.Chart.TransitionTag.FM.Calibration.OpenResultScreen")
    StateChartOwner=GetOwningLocalPlayer()
    Urgency=Normal
    bBacktrack=false
   )
  ]}
 PreviousCalibrationScreen {nodes [
   PreviousCalibrationScreen() {is=title K2Node_FunctionEntry}
   > DispatchTransitionEvent(
    self=Get UIStateChartManager()
    TransitionTag=(TagName="UI.Chart.TransitionTag.Generic.Back")
    StateChartOwner=GetOwningLocalPlayer()
    Urgency=Normal
    bBacktrack=false
   )
  ]}
 ShowErrorScreen {nodes [
   ShowErrorScreen() {is=title K2Node_FunctionEntry}
   > DispatchTransitionEvent(
    self=Get UIStateChartManager()
    TransitionTag=(TagName="UI.Chart.TransitionTag.FM.Calibration.OpenErrorScreen")
    StateChartOwner=GetOwningLocalPlayer()
    Urgency=Normal
    bBacktrack=false
   )
  ]}
 ShowInputToVideoScreen {nodes [
   ShowInputToVideoScreen() {is=title K2Node_FunctionEntry}
   > DispatchTransitionEvent(
    self=Get UIStateChartManager()
    TransitionTag=(TagName="UI.Chart.TransitionTag.FM.Calibration.OpenInputToVideoScreen")
    StateChartOwner=GetOwningLocalPlayer()
    Urgency=Normal
    bBacktrack=false
   )
  ]}
 ShowManualCalibration {nodes [
   ShowManualCalibration() {is=title K2Node_FunctionEntry}
   > DispatchTransitionEvent(
    self=Get UIStateChartManager()
    TransitionTag=(TagName="UI.Chart.TransitionTag.FM.Calibration.OpenManualCalibration")
    StateChartOwner=GetOwningLocalPlayer()
    Urgency=Normal
    bBacktrack=false
   )
  ]}
 ShowProVocalsScreen {nodes [
   ShowProVocalsScreen() {is=title K2Node_FunctionEntry}
   > DispatchTransitionEvent(
    self=Get UIStateChartManager()
    TransitionTag=(TagName="UI.Chart.TransitionTag.FM.Calibration.OpenProVocalsScreen")
    StateChartOwner=GetOwningLocalPlayer()
    Urgency=Normal
    bBacktrack=false
   )
  ]}
]
variables [
 ExperienceMode {type=GameplayTag, default {tagName=None}}
 FMCalibrationComponent {type=BP_FMCalibrationComponent_C, default=null}
 ForceCalibrationClosedEventRef {type=AsyncAction_StartListeningToStatefulEvent, default=null}
]
widgets {
 info {inheritedWidgetCount=0, parentClass=/Script/FMCalibrationUI.FMCalibrationScreenWidgetBase, rootWidgetClass=null, widgetCount=0}
 widgets []
}
