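# Error screen widget for the FMCalibration flow, derived from WBP_FMCalibration_CalibrationScreen.
# It shows a centered header and body message over a faded logo, inside a safe zone, with placeholder text that this graph never sets.
# A hold-to-confirm "GOT IT!" button, bound to the GenericForward input action, is the desired focus target and calls CloseErrorScreen when clicked.
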WidgetBlueprint /FMCalibration/Widgets/WBP_FMCalibration_Error
graphs [
 BP_GetDesiredFocusTarget {nodes [
   Get Desired Focus Target()
   > Return Node!(execute, ReturnValue:Widget Object Reference=Get Button_Close_Kit() -> (Button_Close_Kit:WBP UIKit Button Loud Object Reference))
  ]}
 EventGraph {nodes [
   "On Clicked (Button_Close_Kit)"!() -> (OutputDelegate, then, Button:Common Button Base Object Reference) {is=title K2Node_ComponentBoundEvent}
   > CloseErrorScreen(self:WBP FMCalibration Calibration Screen Object Reference)
  ]}
]
variables []
widgets {
 info {
  inheritedWidgetCount=0
  parentClass=&/WBP_FMCalibration_CalibrationScreen.WBP_FMCalibration_CalibrationScreen_C
  rootWidgetClass=&Overlay
  widgetCount=11
 }
 widgets [
  Overlay_11 {widgetClassPath=&Overlay, widget=Overlay_11}
  HMX_Logo_Texture {
   widgetClassPath=&Image
   parent=Overlay_11
   widget=HMX_Logo_Texture
   properties {
    brush {
     bIsDynamicallyLoaded=false
     drawAs=Image
     imageType=FullColor
     mirroring=NoMirror
     resourceName=None
     resourceObject=/FMCalibration/Materials/MI_UI_FTUE_Logo
     tiling=NoTile
     imageSize=<1200 1200>
     margin=<ltrb 0 0 0 0>
     outlineSettings {
      bUseBrushTransparency=false
      roundingType=HalfHeightRadius
      width=0
      color {colorUseRule=UseColor_Specified, specifiedColor=<rgba 0 0 0 0>}
      cornerRadii=<0 0 0 0>
     }
     tintColor {colorUseRule=UseColor_Specified, specifiedColor=<rgba 1 1 1 1>}
     uVRegion {bIsValid=false, max=<0 0>, min=<0 0>}
    }
    colorAndOpacity=<rgba 1 1 1 0.35>
   }
   slot {class=OverlaySlot, properties {horizontalAlignment=HAlign_Center, verticalAlignment=VAlign_Center}}
  }
  RootSafeZone {
   widgetClassPath=&SafeZone
   parent=Overlay_11
   widget=RootSafeZone
   slot {class=OverlaySlot, properties {horizontalAlignment=HAlign_Fill, verticalAlignment=VAlign_Fill}}
  }
  MainContainer {widgetClassPath=&Overlay, parent=RootSafeZone, widget=MainContainer, slot {class=SafeZoneSlot}}
  GridPanel_0 {
   widgetClassPath=&GridPanel
   parent=MainContainer
   widget=GridPanel_0
   properties {columnFill [1], rowFill [1]}
   slot {class=OverlaySlot, properties {horizontalAlignment=HAlign_Fill, verticalAlignment=VAlign_Fill, padding=<ltrb 0 100 0 100>}}
  }
  SlideContent_Padding {widgetClassPath=&VerticalBox, parent=GridPanel_0, widget=SlideContent_Padding, slot {class=GridSlot}}
  ModalContent_Padding {
   widgetClassPath=&VerticalBox
   parent=SlideContent_Padding
   widget=ModalContent_Padding
   slot {class=VerticalBoxSlot, properties {padding=<ltrb 0 0 0 80>}}
  }
  Text_Header {
   widgetClassPath=&CommonTextBlock
   bIsVariable=true
   parent=ModalContent_Padding
   widget=Text_Header
   properties {
    justification=Center
    text=Error
    textTransformPolicy=ToUpper
    wrapTextAt=1700
    font {
     bForceMonospaced=false
     bMaterialIsStencil=false
     fontMaterial=null
     fontObject=&BurbankSmall
     letterSpacing=0
     monospacedWidth=1
     size=25.5
     skewAmount=0
     typefaceFontName=Bold
     outlineSettings {
      bApplyOutlineToDropShadows=false
      bMiteredCorners=false
      bSeparateFillAlpha=false
      outlineMaterial=null
      outlineSize=0
      outlineColor=<rgba 0 0 0 1>
     }
    }
    shadowOffset=<0 0>
   }
   slot {class=VerticalBoxSlot, properties {horizontalAlignment=HAlign_Center, verticalAlignment=VAlign_Top, padding=<ltrb 0 0 0 40>}}
  }
  Text_Body {
   widgetClassPath=&CommonTextBlock
   bIsVariable=true
   parent=ModalContent_Padding
   widget=Text_Body
   properties {
    justification=Center
    text=Error text
    wrapTextAt=900
    colorAndOpacity {colorUseRule=UseColor_Specified, specifiedColor=<rgba 0.4397 0.7835 0.8879 1>}
    font {
     bForceMonospaced=false
     bMaterialIsStencil=false
     fontMaterial=null
     fontObject=&BurbankSmall
     letterSpacing=0
     monospacedWidth=1
     size=18
     skewAmount=0
     typefaceFontName=Medium
     outlineSettings {
      bApplyOutlineToDropShadows=false
      bMiteredCorners=false
      bSeparateFillAlpha=false
      outlineMaterial=null
      outlineSize=0
      outlineColor=<rgba 0 0 0 1>
     }
    }
    shadowOffset=<0 0>
   }
   slot {class=VerticalBoxSlot, properties {horizontalAlignment=HAlign_Center, verticalAlignment=VAlign_Top, padding=<ltrb 0 30 0 0>}}
  }
  SizeBox_ButtonGroup_Close {
   widgetClassPath=&SizeBox
   parent=SlideContent_Padding
   widget=SizeBox_ButtonGroup_Close
   properties {bOverride_MinDesiredWidth=true, minDesiredWidth=408, widthOverride=408}
   slot {
    class=VerticalBoxSlot
    properties {horizontalAlignment=HAlign_Center, verticalAlignment=VAlign_Bottom, padding=<ltrb 0 0 0 50>, size {sizeRule=Fill, value=1}}
   }
  }
  Button_Close_Kit {
   widgetClassPath=/FortUILibrary/UIKit/Buttons/WBP_UIKit_Button_Loud
   bIsVariable=true
   parent=SizeBox_ButtonGroup_Close
   widget=Button_Close_Kit
   properties {
    bRequiresHold=true
    buttonSize=Large
    minHeight=56
    text="GOT IT!"
    triggeringInputAction {dataTable=&InputActionDataTable, rowName=GenericForward}
   }
   slot {class=SizeBoxSlot}
  }
 ]
}
