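# Translucent UI material in FMCalibration, named as a simple blur for the Pilgrim UI.
# It contains only a GetSlatePost0 function call, which reads Slate post buffer 0, and a constant of one.
# The material has no blur math of its own, so any blur depends on what that post buffer holds.
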Material /FMCalibration/Materials/M_UI_Pilgrim_SimpleBlur
graph {
 properties {
  bAutomaticallySetUsageInEditor=false
  bUsedWithStaticMesh=false
  blendMode=BLEND_Translucent
  materialDomain=MD_UI
  pixelDepthOffsetMode=PDOM_Legacy
  usedShadingModels=Default Lit
 }
 nodes [
  Constant() {r=1}
  MaterialFunctionCall!("UVs (Optional)") {materialFunction=/Engine/Functions/UserInterface/GetSlatePost0}
 ]
}
