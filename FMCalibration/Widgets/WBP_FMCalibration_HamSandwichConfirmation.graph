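# Common activatable modal that shows a "Welcome to Main Stage Local Play" explainer with local-player rules and an image over a blurred, tinted backdrop.
# On activation, it plays the Showup slide-and-fade at double speed and offers its Close button as the focus target.
# Close, bound to Confirm, removes the widget and broadcasts OnConfirmSkipClosed.
# ClearAllModals and the ShowButtons animation exist but are not invoked here.
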WidgetBlueprint /FMCalibration/Widgets/WBP_FMCalibration_HamSandwichConfirmation
animations [
 Showup {
  binding_count=1
  length_seconds=1
  object=&self:Showup
  sequence {bindings [BG {
     id=#1
     possessed_class=Overlay
     tracks [
      {
       type=MovieScene2DTransformTrack
       sections [{
         type=MovieScene2DTransformSection
         end_frame=20
         start_frame=0
         channels [
          Translation.X {type=MovieSceneScriptingFloatChannel, const [0,1]}
          Translation.Y {type=MovieSceneScriptingFloatChannel, keys [[0,841.9097],[20,0]]}
         ]
        }]
      }
      {
       type=MovieSceneFloatTrack
       display_name=Render Opacity
       sections [{type=MovieSceneFloatSection, end_frame=20, start_frame=0, channels [None {type=MovieSceneScriptingFloatChannel, keys [[0,0],[20,1]]}]}]
      }
     ]
    }]}
 }
 ShowButtons {
  binding_count=1
  length_seconds=1
  object=&self:ShowButtons
  sequence {bindings [Buttons {
     id=#2
     possessed_class=GridPanel
     tracks [
      {
       type=MovieScene2DTransformTrack
       sections [{
         type=MovieScene2DTransformSection
         end_frame=20
         start_frame=0
         channels [
          Translation.X {type=MovieSceneScriptingFloatChannel, const [0,1]}
          Translation.Y {type=MovieSceneScriptingFloatChannel, keys [[0,100],[20,0]]}
         ]
        }]
      }
      {
       type=MovieSceneFloatTrack
       display_name=Render Opacity
       sections [{type=MovieSceneFloatSection, end_frame=20, start_frame=0, channels [None {type=MovieSceneScriptingFloatChannel, keys [[0,0],[20,1]]}]}]
      }
     ]
    }]}
 }
]
graphs [
 BP_GetDesiredFocusTarget {nodes [
   Get Desired Focus Target()
   > Return Node!(
    execute
    ReturnValue:Widget Object Reference=Get WBP_UIKit_Button_Regular() -> (WBP_UIKit_Button_Regular:WBP UIKit Button Regular Object Reference)
   )
  ]}
 ClearAllModals {nodes [
   ClearAllModals() {is=title K2Node_FunctionEntry}
   > n1 = GetCurrentModal!(execute) -> (then, ReturnValue:Common Activatable Panel Legacy Object Reference)
   > n2 = While Loop!(execute:Exec, Condition:bool=IsValid(Object=@n1)) -> (LoopBody:Exec, Completed:Exec) {is=title K2Node_MacroInstance}
   PrintString!(
    execute=@n2.Completed
    InString="To String (Integer)"(InInt=Get modalspopped() -> (modalspopped:int=0))
    bPrintToScreen
    bPrintToLog
    TextColor
    Duration
    Key
   ) -> (then)
   PopCurrentModal!(execute=@n2.LoopBody) -> (then)
   > Increment Int(Value=Get modalspopped() -> (modalspopped=0))
  ]}
 EventGraph {nodes [
   "On Clicked (WBP_UIKit_Button_Regular)"!() -> (OutputDelegate, then, Button:Common Button Base Object Reference) {is=title K2Node_ComponentBoundEvent}
   > CleanupWidget()
   CleanupWidget() {is=title K2Node_CustomEvent}
   > RemoveFromParent()
   > Call On Confirm Skip Closed!(execute:Exec, self:WBP FMCalibration Ham Sandwich Confirmation Object Reference) -> (then:Exec) {is=title K2Node_CallDelegate}
   Event On Activated()
   > PlayAnimation(InAnimation=Get Showup() -> (Showup:Widget Animation Object Reference), PlaybackSpeed=2)
  ]}
 OnConfirmSkipClosed {nodes [OnConfirmSkipClosed() {is=title K2Node_FunctionEntry}]}
]
variables [OnConfirmSkipClosed {}]
widgets {
 info {inheritedWidgetCount=0, parentClass=&CommonActivatableWidget, rootWidgetClass=&Overlay, widgetCount=17}
 widgets [
  Overlay_1 {widgetClassPath=&Overlay, widget=Overlay_1}
  WBP_CaptureForPostBufferUpdate_1 {
   widgetClassPath=&WBP_CaptureForPostBufferUpdate
   bIsVariable=true
   parent=Overlay_1
   widget=WBP_CaptureForPostBufferUpdate_1
   slot {class=OverlaySlot, properties {horizontalAlignment=HAlign_Fill, verticalAlignment=VAlign_Fill}}
  }
  Image_96 {
   widgetClassPath=&Image
   bIsVariable=true
   parent=Overlay_1
   widget=Image_96
   properties {brush {
     bIsDynamicallyLoaded=false
     drawAs=Image
     imageType=FullColor
     mirroring=NoMirror
     resourceName=None
     resourceObject=/FMCalibration/Materials/M_UI_Pilgrim_SimpleBlur
     tiling=NoTile
     imageSize=<32 32>
     margin=<ltrb 0 0 0 0>
     outlineSettings {
      bUseBrushTransparency=false
      roundingType=HalfHeightRadius
      width=0
      color {colorUseRule=UseColor_Specified, specifiedColor=<rgba 0 0 0 0>}
      cornerRadii=<0 0 0 0>
     }
     tintColor {colorUseRule=UseColor_Specified, specifiedColor=<rgba 1 1 1 0.9>}
     uVRegion {bIsValid=false, max=<0 0>, min=<0 0>}
    }}
   slot {class=OverlaySlot, properties {horizontalAlignment=HAlign_Fill, verticalAlignment=VAlign_Fill}}
  }
  Border_78 {
   widgetClassPath=&Border
   parent=Overlay_1
   widget=Border_78
   properties {background {
     bIsDynamicallyLoaded=false
     drawAs=Image
     imageType=FullColor
     mirroring=NoMirror
     resourceName=None
     resourceObject=null
     tiling=NoTile
     imageSize=<32 32>
     margin=<ltrb 0 0 0 0>
     outlineSettings {
      bUseBrushTransparency=false
      roundingType=HalfHeightRadius
      width=0
      color {colorUseRule=UseColor_Specified, specifiedColor=<rgba 0 0 0 0>}
      cornerRadii=<0 0 0 0>
     }
     tintColor {colorUseRule=UseColor_Specified, specifiedColor=<rgba 0.0033 0.003 0.0497 0.6>}
     uVRegion {bIsValid=false, max=<0 0>, min=<0 0>}
    }}
   slot {class=OverlaySlot, properties {horizontalAlignment=HAlign_Fill, verticalAlignment=VAlign_Fill}}
  }
  BG {
   widgetClassPath=&Overlay
   parent=Overlay_1
   widget=BG
   slot {class=OverlaySlot, properties {horizontalAlignment=HAlign_Center, verticalAlignment=VAlign_Center}}
  }
  WBP_Sparks_BackingPanel_152 {
   widgetClassPath=&/WBP_FMCalibration_BackingPanel
   bIsVariable=true
   parent=BG
   widget=WBP_Sparks_BackingPanel_152
   slot {class=OverlaySlot, properties {horizontalAlignment=HAlign_Fill, verticalAlignment=VAlign_Fill}}
  }
  WBP_CaptureForPostBufferUpdate {
   widgetClassPath=&WBP_CaptureForPostBufferUpdate
   bIsVariable=true
   parent=BG
   slot {class=OverlaySlot, properties {horizontalAlignment=HAlign_Fill, verticalAlignment=VAlign_Fill}}
  }
  Content {
   widgetClassPath=&HorizontalBox
   parent=BG
   widget=Content
   slot {class=OverlaySlot, properties {horizontalAlignment=HAlign_Center, verticalAlignment=VAlign_Center}}
  }
  TextandButtons {
   widgetClassPath=&VerticalBox
   parent=Content
   widget=TextandButtons
   slot {class=HorizontalBoxSlot, properties {padding=<ltrb 80 80 0 60>}}
  }
  TitleTextWidget {
   widgetClassPath=&CommonTextBlock
   parent=TextandButtons
   widget=TitleTextWidget
   properties {
    text=Welcome to Main Stage Local Play
    wrapTextAt=700
    font {
     bForceMonospaced=false
     bMaterialIsStencil=false
     fontMaterial=null
     fontObject=null
     letterSpacing=0
     monospacedWidth=1
     size=24
     skewAmount=0
     typefaceFontName=None
     outlineSettings {
      bApplyOutlineToDropShadows=false
      bMiteredCorners=false
      bSeparateFillAlpha=false
      outlineMaterial=null
      outlineSize=0
      outlineColor=<rgba 0 0 0 1>
     }
    }
    shadowOffset=<0 0>
   }
   slot {class=VerticalBoxSlot}
  }
  Spacer {widgetClassPath=&Spacer, parent=TextandButtons, widget=Spacer, properties {size=<1 50>}, slot {class=VerticalBoxSlot}}
  Text_Body {
   widgetClassPath=&CommonTextBlock
   bIsVariable=true
   parent=TextandButtons
   widget=Text_Body
   properties {
    text="Player 1 has control backstage and in menus.\r\n\r\nOther local players are locked in place until showtime.\r\n\r\n\"Quick emote\" to express yourself onstage.\r\n\r\nPlay with any combo of local and online players."
    wrapTextAt=600
    font {
     bForceMonospaced=false
     bMaterialIsStencil=false
     fontMaterial=null
     fontObject=&BurbankSmall
     letterSpacing=0
     monospacedWidth=1
     size=18
     skewAmount=0
     typefaceFontName=Medium
     outlineSettings {
      bApplyOutlineToDropShadows=false
      bMiteredCorners=false
      bSeparateFillAlpha=false
      outlineMaterial=null
      outlineSize=0
      outlineColor=<rgba 0 0 0 1>
     }
    }
    shadowOffset=<0 0>
   }
   slot {class=VerticalBoxSlot}
  }
  Spacer_1 {widgetClassPath=&Spacer, parent=TextandButtons, widget=Spacer_1, properties {size=<1 50>}, slot {class=VerticalBoxSlot}}
  Buttons {
   widgetClassPath=&GridPanel
   parent=TextandButtons
   widget=Buttons
   slot {class=VerticalBoxSlot, properties {verticalAlignment=VAlign_Bottom, size {sizeRule=Fill, value=1}}}
  }
  WBP_UIKit_Button_Regular {
   widgetClassPath=&WBP_UIKit_Button_Regular
   bIsVariable=true
   parent=Buttons
   widget=WBP_UIKit_Button_Regular
   properties {
    bIsPersistentBinding=true
    text=Close
    visibility=Visible
    triggeringInputAction {dataTable=&InputActionDataTable, rowName=Confirm}
   }
   slot {class=GridSlot, properties {padding=<ltrb 20 0 0 0>}}
  }
  Spacer_2 {widgetClassPath=&Spacer, parent=Content, widget=Spacer_2, properties {size=<100 1>}, slot {class=HorizontalBoxSlot}}
  StageTech_Image {
   widgetClassPath=&Image
   bIsVariable=true
   parent=Content
   widget=StageTech_Image
   properties {
    visibility=SelfHitTestInvisible
    brush {
     bIsDynamicallyLoaded=false
     drawAs=Image
     imageType=FullColor
     mirroring=NoMirror
     resourceName=None
     resourceObject=/FMCalibration/Materials/MI_UI_RoundedTexture_FTUE
     tiling=NoTile
     imageSize=<490 59.1038>
     margin=<ltrb 0 0 0 0.0632>
     outlineSettings {
      bUseBrushTransparency=false
      roundingType=HalfHeightRadius
      width=0
      color {colorUseRule=UseColor_Specified, specifiedColor=<rgba 0 0 0 0>}
      cornerRadii=<0 0 0 0>
     }
     tintColor {colorUseRule=UseColor_Specified, specifiedColor=<rgba 1.172 1 1 1>}
     uVRegion {bIsValid=false, max=<0 0>, min=<0 0>}
    }
   }
   slot {class=HorizontalBoxSlot, properties {size {sizeRule=Fill, value=1}}}
  }
 ]
}
