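# Calibrometer widget, a CommonUserWidget that flashes a bottom-centred square on the music beat as a visual timing reference for calibration.
# On initialization it binds to the owning player controller's FMCalibrationControllerComponent, showing (hit-test invisible) or collapsing itself as the calibrometer toggles.
# While visible, Tick reads the MusicClock's distance from the current beat in video render time, turning the square white near the beat and black otherwise.
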WidgetBlueprint /FMCalibration/Widgets/WBP_CalibrometerWidget
graphs [
 EventGraph {nodes [
   n1 = HandleCalibrometerEnabled!() -> (OutputDelegate, then, Enabled:bool=~) {is=title K2Node_CustomEvent}
   > UpdateVisibility!(execute, self, Display:bool=@n1.Enabled) -> (then)
   Event Tick()
   > Branch(Condition=IsVisible())
   n2 = Get MusicClock() -> (MusicClock:Music Clock Object Reference)
   > Is Valid(InputObject=@n2)
   n3 = GetDistanceFromCurrentBeat!(self:Music Clock Object Reference=@n2, Timebase:ECalibratedMusicTimebase Enum=VideoRenderTime) -> (ReturnValue:float)
   > SetColorAndOpacity(
    self:Image Object Reference=Get WhiteRect() -> (WhiteRect:Image Object Reference)
    InColorAndOpacity:Linear Color Structure=SelectColor(
     A=LinearColor_White!() -> (ReturnValue:Linear Color Structure)
     B=LinearColor_Black!() -> (ReturnValue:Linear Color Structure)
     bPickA="float < float"(A=@n3, B=0.1)
    )
   )
   Event OnInitialized()
   n4 = GetComponentByClass!(
    self:Actor Object Reference=GetOwningPlayer()
    ComponentClass:Actor Component Class Reference=/Script/FMCalibrationRuntime.FMCalibrationControllerComponent
   ) -> (ReturnValue:FMCalibration Controller Component Object Reference)
   > Bind Event to On Calibrometer Enabled Changed(self:FMCalibration Controller Component Object Reference=@n4, Delegate=@n1) {is=title K2Node_AddDelegate}
   > UpdateVisibility(Display=IsCalibrometerEnabled!(self:FMCalibration Controller Component Object Reference=@n4) -> (ReturnValue:bool))
   Event PreConstruct()
   Event Construct()
  ]}
 UpdateVisibility {nodes [
   n1 = UpdateVisibility!() -> (then, Display:bool=~) {is=title K2Node_FunctionEntry}
   > SetVisibility(InVisibility=Select(Option 0=Collapsed, Option 1=HitTestInvisible, Index=@n1))
  ]}
]
variables [MusicClock {type=MusicClockComponent, default=null}]
widgets {
 info {inheritedWidgetCount=0, parentClass=&CommonUserWidget, rootWidgetClass=&Overlay, widgetCount=2}
 widgets [
  Overlay_26 {widgetClassPath=&Overlay, widget=Overlay_26}
  WhiteRect {
   widgetClassPath=&Image
   bIsVariable=true
   parent=Overlay_26
   widget=WhiteRect
   properties {
    brush {
     bIsDynamicallyLoaded=false
     drawAs=Image
     imageType=NoImage
     mirroring=NoMirror
     resourceName=None
     resourceObject=null
     tiling=NoTile
     imageSize=<360 360>
     margin=<ltrb 0 0 0 0>
     outlineSettings {
      bUseBrushTransparency=false
      roundingType=HalfHeightRadius
      width=0
      color {colorUseRule=UseColor_Specified, specifiedColor=<rgba 0 0 0 0>}
      cornerRadii=<0 0 0 0>
     }
     tintColor {colorUseRule=UseColor_Specified, specifiedColor=<rgba 1 1 1 1>}
     uVRegion {bIsValid=false, max=<0 0>, min=<0 0>}
    }
    colorAndOpacity=<rgba 0 0 0 1>
   }
   slot {class=OverlaySlot, properties {horizontalAlignment=HAlign_Center, verticalAlignment=VAlign_Bottom}}
  }
 ]
}
