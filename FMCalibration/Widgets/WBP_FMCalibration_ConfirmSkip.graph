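# Confirmation modal for skipping FM calibration (Tune-Up) or skipping ahead, with Skip and Go Back buttons.
# On activation it plays intro animations, disables main-menu toggling and picks its text from isCalibrationStep.
# Skip broadcasts ModalClosed for calibration steps, otherwise ModalSkipped, via the playspace event router, then cleans up and pops the current modal.
# Go Back, the default focus, only cleans up: removes the widget, fires OnConfirmSkipClosed, restores menu toggling and game-only input.
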WidgetBlueprint /FMCalibration/Widgets/WBP_FMCalibration_ConfirmSkip
animations [
 Showup {
  binding_count=1
  length_seconds=1
  object=&self:Showup
  sequence {bindings [BG {
     id=#1
     possessed_class=Overlay
     tracks [
      {
       type=MovieScene2DTransformTrack
       sections [{
         type=MovieScene2DTransformSection
         end_frame=20
         start_frame=0
         channels [
          Translation.X {type=MovieSceneScriptingFloatChannel, const [0,1]}
          Translation.Y {type=MovieSceneScriptingFloatChannel, keys [[0,841.9097],[20,0]]}
         ]
        }]
      }
      {
       type=MovieSceneFloatTrack
       display_name=Render Opacity
       sections [{type=MovieSceneFloatSection, end_frame=20, start_frame=0, channels [None {type=MovieSceneScriptingFloatChannel, keys [[0,0],[20,1]]}]}]
      }
     ]
    }]}
 }
 ShowButtons {
  binding_count=1
  length_seconds=1
  object=&self:ShowButtons
  sequence {bindings [Buttons {
     id=#2
     possessed_class=GridPanel
     tracks [
      {
       type=MovieScene2DTransformTrack
       sections [{
         type=MovieScene2DTransformSection
         end_frame=20
         start_frame=0
         channels [
          Translation.X {type=MovieSceneScriptingFloatChannel, const [0,1]}
          Translation.Y {type=MovieSceneScriptingFloatChannel, keys [[0,100],[20,0]]}
         ]
        }]
      }
      {
       type=MovieSceneFloatTrack
       display_name=Render Opacity
       sections [{type=MovieSceneFloatSection, end_frame=20, start_frame=0, channels [None {type=MovieSceneScriptingFloatChannel, keys [[0,0],[20,1]]}]}]
      }
     ]
    }]}
 }
]
graphs [
 BP_GetDesiredFocusTarget {nodes [
   Get Desired Focus Target()
   > Return Node!(execute, ReturnValue:Widget Object Reference=Get GoBack_Button() -> (GoBack_Button:WBP UIKit Button Regular Object Reference))
  ]}
 BP_GetDesiredInputConfig {nodes [
   Get Desired Input Config() {is=title K2Node_FunctionEntry}
   > Return Node(ReturnValue:UIInput Config Structure=Make UIInput Config!(
     bIgnoreMoveInput:bool=true
     bIgnoreLookInput:bool=true
     InputMode:ECommonInputMode Enum=Menu
     MouseCaptureMode:EMouseCaptureMode Enum=NoCapture
     MouseLockMode:EMouseLockMode Enum=DoNotLock
     bHideCursorDuringViewportCapture:bool=True
    ) -> (UIInputConfig:UIInput Config Structure) {is=title K2Node_MakeStruct})
  ]}
 ClearAllModals {nodes [
   ClearAllModals() {is=title K2Node_FunctionEntry}
   > n1 = GetCurrentModal!(execute) -> (then, ReturnValue:Common Activatable Panel Legacy Object Reference)
   > n2 = While Loop!(execute:Exec, Condition:bool=IsValid(Object=@n1)) -> (LoopBody:Exec, Completed:Exec) {is=title K2Node_MacroInstance}
   PrintString!(
    execute=@n2.Completed
    InString="To String (Integer)"(InInt=Get modalspopped() -> (modalspopped:int=0))
    bPrintToScreen
    bPrintToLog
    TextColor
    Duration
    Key
   ) -> (then)
   PopCurrentModal!(execute=@n2.LoopBody) -> (then)
   > Increment Int(Value=Get modalspopped() -> (modalspopped=0))
  ]}
 EventGraph {nodes [
   n1 = "On Clicked (GoBack_Button)"!() -> (OutputDelegate, then, Button:Common Button Base Object Reference) {is=title K2Node_ComponentBoundEvent}
   > Fire Event Element Interacted(Element=@n1.Button)
   > CleanupWidget()
   n2 = "On Clicked (Skip_Button)"!() -> (OutputDelegate, then, Button:Common Button Base Object Reference) {is=title K2Node_ComponentBoundEvent}
   > n3 = Sequence!(execute) -> (then_0, then_1, then_2:Exec)
   > Fire Event Element Interacted(Element=@n2.Button)
   n4 = GetEventRouter!(execute=@n3.then_1, Scope:Object Class Reference=/Script/FortniteGame.FortPlayspace, ContextActor:Actor Object Reference=GetOwningPlayer()) -> (then, ReturnValue:Gameplay Event Router Component Object Reference)
   > Is Valid(InputObject=@n4)
   > n5 = Branch(Condition=Get isCalibrationStep() -> (isCalibrationStep:bool))
   > Broadcast Event!(
    execute
    self:Gameplay Event Router Component Object Reference=@n4
    EventData:FMEvent Calibration Modal Closed Structure=Make FMEvent Calibration Modal Closed!() -> (FMEvent_Calibration_ModalClosed:FMEvent Calibration Modal Closed Structure) {is=title K2Node_MakeStruct}
    EventContext:Object Reference
   ) -> (then, ReturnValue:bool)
   Broadcast Event(
    execute=@n5.else
    self=@n4
    EventData:FMEvent Calibration Modal Skipped Structure=Make FMEvent Calibration Modal Skipped!() -> (FMEvent_Calibration_ModalSkipped:FMEvent Calibration Modal Skipped Structure) {is=title K2Node_MakeStruct}
   )
   CleanupWidget(execute=@n3.then_2)
   > PopCurrentModal()
   CleanupWidget() {is=title K2Node_CustomEvent}
   > RemoveFromParent()
   > Call On Confirm Skip Closed!(execute:Exec, self:WBP FMCalibration Confirm Skip Object Reference) -> (then:Exec) {is=title K2Node_CallDelegate}
   > SetToggleMainMenuEnabled!(
    execute
    self:Fort Player Controller Object Reference=GetFortPlayerControllerFromActor!(Actor:Actor Object Reference=GetOwningPlayer()) -> (ReturnValue:Fort Player Controller Object Reference)
    bNewToggleMainMenuEnabled:bool=true
   ) -> (then)
   > SetInputMode_GameOnly!(execute, PlayerController:Player Controller Object Reference=GetOwningPlayer(), bFlushInput:bool=false) -> (then)
   Event On Activated()
   > PlayAnimation(InAnimation=Get Showup() -> (Showup:Widget Animation Object Reference), PlaybackSpeed=2)
   > PlayAnimation(InAnimation=Get ShowButtons() -> (ShowButtons:Widget Animation Object Reference), PlaybackSpeed=2)
   > SetToggleMainMenuEnabled(self=GetFortPlayerControllerFromActor(Actor=GetOwningPlayer()), bNewToggleMainMenuEnabled=false)
   > "SetText (Text)"(
    self=Get TitleTextWidget() -> (TitleTextWidget:Common Text Object Reference)
    InText=SelectText(A="LOCTEXT(\"Are you sure?\")", B="LOCTEXT(\"Skip ahead?\")", bPickA=Get isCalibrationStep())
   )
   > "SetText (Text)"(
    self=Get Text_Body() -> (Text_Body:Common Text Object Reference)
    InText=SelectText(
     A="LOCTEXT(\"If you skip Tune-Up, your audio, video, and input timing might be off when you play.\\r\\n\\r\\nYou can Tune-Up later in the Game Menu backstage.\")"
     B="LOCTEXT(\"Don\\'t worry! You can visit the Game Menu any time to get tips, change controls, or tune-up.\")"
     bPickA=Get isCalibrationStep()
    )
   )
  ]}
 OnConfirmSkipClosed {nodes [OnConfirmSkipClosed() {is=title K2Node_FunctionEntry}]}
]
variables [
 isCalibrationStep {type=boolean}
 OnConfirmSkipClosed {}
]
widgets {
 info {inheritedWidgetCount=0, parentClass=&CommonActivatableWidget, rootWidgetClass=&Overlay, widgetCount=19}
 widgets [
  Overlay_1 {widgetClassPath=&Overlay, widget=Overlay_1}
  WBP_CaptureForPostBufferUpdate_1 {
   widgetClassPath=&WBP_CaptureForPostBufferUpdate
   bIsVariable=true
   parent=Overlay_1
   widget=WBP_CaptureForPostBufferUpdate_1
   slot {class=OverlaySlot, properties {horizontalAlignment=HAlign_Fill, verticalAlignment=VAlign_Fill}}
  }
  Image_96 {
   widgetClassPath=&Image
   bIsVariable=true
   parent=Overlay_1
   widget=Image_96
   properties {brush {
     bIsDynamicallyLoaded=false
     drawAs=Image
     imageType=FullColor
     mirroring=NoMirror
     resourceName=None
     resourceObject=/FMCalibration/Materials/M_UI_Pilgrim_SimpleBlur
     tiling=NoTile
     imageSize=<32 32>
     margin=<ltrb 0 0 0 0>
     outlineSettings {
      bUseBrushTransparency=false
      roundingType=HalfHeightRadius
      width=0
      color {colorUseRule=UseColor_Specified, specifiedColor=<rgba 0 0 0 0>}
      cornerRadii=<0 0 0 0>
     }
     tintColor {colorUseRule=UseColor_Specified, specifiedColor=<rgba 1 1 1 1>}
     uVRegion {bIsValid=false, max=<0 0>, min=<0 0>}
    }}
   slot {class=OverlaySlot, properties {horizontalAlignment=HAlign_Fill, verticalAlignment=VAlign_Fill}}
  }
  Border_78 {
   widgetClassPath=&Border
   parent=Overlay_1
   widget=Border_78
   properties {background {
     bIsDynamicallyLoaded=false
     drawAs=Image
     imageType=FullColor
     mirroring=NoMirror
     resourceName=None
     resourceObject=null
     tiling=NoTile
     imageSize=<32 32>
     margin=<ltrb 0 0 0 0>
     outlineSettings {
      bUseBrushTransparency=false
      roundingType=HalfHeightRadius
      width=0
      color {colorUseRule=UseColor_Specified, specifiedColor=<rgba 0 0 0 0>}
      cornerRadii=<0 0 0 0>
     }
     tintColor {colorUseRule=UseColor_Specified, specifiedColor=<rgba 0.0033 0.003 0.0497 0.6>}
     uVRegion {bIsValid=false, max=<0 0>, min=<0 0>}
    }}
   slot {class=OverlaySlot, properties {horizontalAlignment=HAlign_Fill, verticalAlignment=VAlign_Fill}}
  }
  BG {
   widgetClassPath=&Overlay
   parent=Overlay_1
   widget=BG
   slot {class=OverlaySlot, properties {horizontalAlignment=HAlign_Center, verticalAlignment=VAlign_Center}}
  }
  WBP_Sparks_BackingPanel_152 {
   widgetClassPath=&/WBP_FMCalibration_BackingPanel
   bIsVariable=true
   parent=BG
   widget=WBP_Sparks_BackingPanel_152
   slot {class=OverlaySlot, properties {horizontalAlignment=HAlign_Fill, verticalAlignment=VAlign_Fill}}
  }
  WBP_CaptureForPostBufferUpdate {
   widgetClassPath=&WBP_CaptureForPostBufferUpdate
   bIsVariable=true
   parent=BG
   slot {class=OverlaySlot, properties {horizontalAlignment=HAlign_Fill, verticalAlignment=VAlign_Fill}}
  }
  Content {
   widgetClassPath=&HorizontalBox
   parent=BG
   widget=Content
   slot {class=OverlaySlot, properties {horizontalAlignment=HAlign_Center, verticalAlignment=VAlign_Center}}
  }
  TextandButtons {
   widgetClassPath=&VerticalBox
   parent=Content
   widget=TextandButtons
   slot {class=HorizontalBoxSlot, properties {padding=<ltrb 80 80 0 60>}}
  }
  TitleTextWidget {
   widgetClassPath=&CommonTextBlock
   bIsVariable=true
   parent=TextandButtons
   widget=TitleTextWidget
   properties {
    style=/FortUILibrary/Styles/Typography/Foundation/TextStyle_BurbankSmall_L_White.TextStyle_BurbankSmall_L_White_C
    text=Skip ahead?
    wrapTextAt=700
    font {
     bForceMonospaced=false
     bMaterialIsStencil=false
     fontMaterial=null
     fontObject=&BurbankSmall
     letterSpacing=0
     monospacedWidth=1
     size=36
     skewAmount=0
     typefaceFontName=Black
     outlineSettings {
      bApplyOutlineToDropShadows=false
      bMiteredCorners=false
      bSeparateFillAlpha=false
      outlineMaterial=null
      outlineSize=0
      outlineColor=<rgba 0 0 0 1>
     }
    }
    shadowOffset=<0 0>
   }
   slot {class=VerticalBoxSlot}
  }
  Spacer {widgetClassPath=&Spacer, parent=TextandButtons, widget=Spacer, properties {size=<1 50>}, slot {class=VerticalBoxSlot}}
  Text_Body {
   widgetClassPath=&CommonTextBlock
   bIsVariable=true
   parent=TextandButtons
   widget=Text_Body
   properties {
    text="Don't worry! You can visit the Game Menu any time to get tips, change controls, or tune-up."
    wrapTextAt=600
    font {
     bForceMonospaced=false
     bMaterialIsStencil=false
     fontMaterial=null
     fontObject=&BurbankSmall
     letterSpacing=0
     monospacedWidth=1
     size=18
     skewAmount=0
     typefaceFontName=Medium
     outlineSettings {
      bApplyOutlineToDropShadows=false
      bMiteredCorners=false
      bSeparateFillAlpha=false
      outlineMaterial=null
      outlineSize=0
      outlineColor=<rgba 0 0 0 1>
     }
    }
    shadowOffset=<0 0>
   }
   slot {class=VerticalBoxSlot}
  }
  Spacer_1 {widgetClassPath=&Spacer, parent=TextandButtons, widget=Spacer_1, properties {size=<1 50>}, slot {class=VerticalBoxSlot}}
  Buttons {
   widgetClassPath=&GridPanel
   parent=TextandButtons
   widget=Buttons
   properties {navigation=&self:WidgetTree.Buttons.WidgetNavigation_0}
   slot {class=VerticalBoxSlot, properties {verticalAlignment=VAlign_Bottom, size {sizeRule=Fill, value=1}}}
  }
  Skip_Button {
   widgetClassPath=&WBP_UIKit_Button_Regular
   bIsVariable=true
   parent=Buttons
   widget=Skip_Button
   properties {
    bUseNaturalNavigation=true
    navigation=&self:WidgetTree.Skip_Button.WidgetNavigation_0
    text="Yes, Skip"
    triggeringInputAction {dataTable=&InputActionDataTable, rowName=Confirm}
   }
   slot {class=GridSlot, properties {horizontalAlignment=HAlign_Center, verticalAlignment=VAlign_Center, padding=<ltrb 20 20 20 20>}}
  }
  GoBack_Button {
   widgetClassPath=&WBP_UIKit_Button_Regular
   bIsVariable=true
   parent=Buttons
   widget=GoBack_Button
   properties {
    bUseNaturalNavigation=true
    navigation=&self:WidgetTree.GoBack_Button.WidgetNavigation_0
    text="No, Go Back"
    triggeringInputAction {dataTable=&InputActionDataTable, rowName=Confirm}
   }
   slot {class=GridSlot, properties {column=1, horizontalAlignment=HAlign_Center, verticalAlignment=VAlign_Center, padding=<ltrb 20 20 20 20>}}
  }
  Spacer_2 {widgetClassPath=&Spacer, parent=Content, widget=Spacer_2, properties {size=<100 1>}, slot {class=HorizontalBoxSlot}}
  Border_0 {
   widgetClassPath=&Border
   parent=Content
   widget=Border_0
   properties {
    verticalAlignment=VAlign_Bottom
    background {
     bIsDynamicallyLoaded=false
     drawAs=Image
     imageType=FullColor
     mirroring=NoMirror
     resourceName=None
     resourceObject=/FMCalibration/Materials/MI_UIKit_TagBackground
     tiling=NoTile
     imageSize=<32 32>
     margin=<ltrb 0 0 0 0>
     outlineSettings {
      bUseBrushTransparency=false
      roundingType=HalfHeightRadius
      width=0
      color {colorUseRule=UseColor_Specified, specifiedColor=<rgba 0 0 0 0>}
      cornerRadii=<0 0 0 0>
     }
     tintColor {colorUseRule=UseColor_Specified, specifiedColor=<rgba 1 1 1 1>}
     uVRegion {bIsValid=false, max=<0 0>, min=<0 0>}
    }
    padding=<ltrb 0 0 0 0>
   }
   slot {class=HorizontalBoxSlot}
  }
  StageTech_Image {
   widgetClassPath=&Image
   bIsVariable=true
   parent=Border_0
   widget=StageTech_Image
   properties {
    visibility=SelfHitTestInvisible
    brush {
     bIsDynamicallyLoaded=false
     drawAs=Image
     imageType=FullColor
     mirroring=NoMirror
     resourceName=None
     resourceObject=/FMCalibration/Materials/MI_UI_StageTech_Art
     tiling=NoTile
     imageSize=<500 500>
     margin=<ltrb 0 0 0 0>
     outlineSettings {
      bUseBrushTransparency=false
      roundingType=HalfHeightRadius
      width=0
      color {colorUseRule=UseColor_Specified, specifiedColor=<rgba 0 0 0 0>}
      cornerRadii=<0 0 0 0>
     }
     tintColor {colorUseRule=UseColor_Specified, specifiedColor=<rgba 1 1 1 1>}
     uVRegion {bIsValid=false, max=<0 0>, min=<0 0>}
    }
   }
   slot {class=BorderSlot, properties {verticalAlignment=VAlign_Bottom, padding=<ltrb 0 0 0 0>}}
  }
 ]
}
