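# Skip button for FM calibration, built on the quiet UIKit button, that opens a skip confirmation popup when clicked.
# The click runs the parent handler, fires an element-interacted event, then async-creates WBP_FMCalibration_ConfirmSkip for the owning FortPlayerController, suspending input until complete.
# It passes IsCalibrationStep to the popup, adds it to the viewport and stores a reference.
# The popup's close event is relayed through the button's OnPopupClosed dispatcher.
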WidgetBlueprint /FMCalibration/Widgets/WBP_FMCalibration_SkipButton
graphs [
 EventGraph {nodes [
   n1 = PopupClosed() {is=title K2Node_CustomEvent}
   > Call On Popup Closed!(execute:Exec, self:WBP FMCalibration Skip Button Object Reference) -> (then:Exec) {is=title K2Node_CallDelegate}
   Event On Clicked()
   > Parent: On Clicked() {is=title K2Node_CallParentFunction}
   > n2 = Sequence()
   > Fire Event Element Interacted(Element=Self-Reference())
   n3 = Cast To FortPlayerController(execute=@n2.then_1, Object=GetOwningPlayer())
   > n4 = CreateWidgetAsync!(
    execute:Exec
    UserWidgetSoftClass:User Widget Soft Class Reference=&/WBP_FMCalibration_ConfirmSkip.WBP_FMCalibration_ConfirmSkip_C
    OwningPlayer:Player Controller Object Reference=@n3
    bSuspendInputUntilComplete:bool=true
   ) -> (then:Exec, OnComplete:Exec, OnFailure:Exec, UserWidget:User Widget Object Reference) {is=title K2Node_AsyncAction}
   n5 = Cast To WBP_FMCalibration_ConfirmSkip!(execute=@n4.OnComplete, Object=@n4) -> (then, CastFailed, AsWBP FMCalibration Confirm Skip:WBP FMCalibration Confirm Skip Object Reference) {is=title K2Node_DynamicCast}
   > Set isCalibrationStep!(
    execute
    isCalibrationStep:bool=Get IsCalibrationStep() -> (IsCalibrationStep:bool)
    self:WBP FMCalibration Confirm Skip Object Reference=@n5
   ) -> (then, Output_Get:bool)
   > n6 = Sequence()
   > AddToViewport!(execute, self:User Widget Object Reference=@n4, ZOrder:int=0) -> (then)
   n7 = Cast To WBP_FMCalibration_ConfirmSkip(execute=@n6.then_1, Object=@n4)
   > n8 = Set ConfirmSkipUserWidget(ConfirmSkipUserWidget:WBP FMCalibration Confirm Skip Object Reference=@n7) -> (Output_Get:WBP FMCalibration Confirm Skip Object Reference)
   > Bind Event to On Confirm Skip Closed(self:WBP FMCalibration Confirm Skip Object Reference=@n8, Delegate=@n1) {is=title K2Node_AddDelegate}
  ]}
 OnPopupClosed {nodes [OnPopupClosed() {is=title K2Node_FunctionEntry}]}
]
variables [
 ConfirmSkipUserWidget {type=WBP_FMCalibration_ConfirmSkip_C, default=null}
 IsCalibrationStep {type=boolean}
 OnPopupClosed {}
]
widgets {
 info {
  inheritedWidgetCount=0
  parentClass=/FortUILibrary/UIKit/Buttons/WBP_UIKit_Button_Quiet.WBP_UIKit_Button_Quiet_C
  rootWidgetClass=null
  widgetCount=0
 }
 widgets []
}
