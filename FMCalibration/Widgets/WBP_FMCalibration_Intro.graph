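# Intro screen of the audio/video/input Tune-Up flow; Begin opens AV calibration or the next screen, depending on bus calibration.
# Calibrated players get Back and manual tune-up buttons; others get Skip, with a modal-closed event listener that commits values and closes.
# When enabled, the Calibrometer display runs a helper metronome, feeds its clock to the widget and reveals tester buttons.
# Activation also widens padding on touch platforms; deactivation stops listeners and metronome.
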WidgetBlueprint /FMCalibration/Widgets/WBP_FMCalibration_Intro
graphs [
 BP_GetDesiredFocusTarget {nodes [
   Get Desired Focus Target()
   > Return Node!(execute, ReturnValue:Widget Object Reference=Get Button_Kit_Begin() -> (Button_Kit_Begin:WBP UIKit Button Loud Object Reference))
  ]}
 EventGraph {nodes [
   "On Clicked (WBP_UIKit_Button_Quiet_Back)"!() -> (OutputDelegate, then, Button:Common Button Base Object Reference) {is=title K2Node_ComponentBoundEvent}
   n1 = Get CalibrationProcessVM() -> (CalibrationProcessVM:Calibration Process VM Object Reference)
   > CloseCalibrationToQuestMenu!(execute, self:WBP FMCalibration Calibration Screen Object Reference, CalibrationProcessVM:Calibration Process VM Object Reference=@n1) -> (then)
   n2 = "On Clicked (Button_Kit_Begin)"!() -> (OutputDelegate, then, Button:Common Button Base Object Reference) {is=title K2Node_ComponentBoundEvent}
   > Fire Event Element Interacted(Element=@n2.Button)
   > n3 = Branch(Condition=IsBusCalibrationEnabled!(self:Calibration Process VM Object Reference=@n1) -> (ReturnValue:bool))
   > OpenAVCalibrationScreen(self:WBP FMCalibration Calibration Screen Object Reference)
   NextCalibrationScreen(execute=@n3.else, self:WBP FMCalibration Calibration Screen Object Reference)
   n4 = "On Clicked (WBP_UIKit_Button_Quiet_AdjustManually)"!() -> (OutputDelegate, then, Button:Common Button Base Object Reference) {is=title K2Node_ComponentBoundEvent}
   > Fire Event Element Interacted(Element=@n4.Button)
   > ShowManualCalibration(self:WBP FMCalibration Calibration Screen Object Reference)
   n5 = "On Clicked (WBP_UIKit_Button_Quiet_ProVocals)"!() -> (OutputDelegate, then, Button:Common Button Base Object Reference) {is=title K2Node_ComponentBoundEvent}
   > Fire Event Element Interacted(Element=@n5.Button)
   > ShowProVocalsScreen(self:WBP FMCalibration Calibration Screen Object Reference)
   "On Clicked (WBP_UIKit_Button_Quiet_InputToVideo)"!() -> (OutputDelegate, then, Button:Common Button Base Object Reference) {is=title K2Node_ComponentBoundEvent}
   > Fire Event Element Interacted()
   > ShowInputToVideoScreen(self:WBP FMCalibration Calibration Screen Object Reference)
   n6 = HandleCalibrometerEnabled!() -> (OutputDelegate, then, Enabled:bool=~) {is=title K2Node_CustomEvent}
   UpdateCalibrometerDisplay(execute=[@n13.then, @n6.then])
   Event On Activated()
   > Parent: On Activated() {is=title K2Node_CallParentFunction}
   > n7 = Sequence!(execute) -> (then_0, then_1, then_2:Exec, then_3:Exec)
   > Fire Event Main View Opened!(execute, View:Widget Object Reference=Self-Reference(), AdditionalInfo:Flow Analytics Additional Info Structure, bIsOverlay:bool=false) -> (then)
   n8 = Get FMCalibrationComponent() -> (FMCalibrationComponent:BP FMCalibration Component Object Reference)
   Is Valid(exec=@n7.then_1, InputObject=@n8)
   n9 = GetIsCalibrated!(self:FMCalibration Controller Component Object Reference=@n8) -> (ReturnValue:bool)
   > SetActiveWidgetIndex(
    self=Get CommonActivatableWidgetSwitcher_0() -> (CommonActivatableWidgetSwitcher_0:Common Activatable Widget Switcher Object Reference)
    Index="To Integer (Boolean)"!(InBool:bool=@n9) -> (ReturnValue:int=0)
   )
   > n10 = Branch(Condition=@n9)
   n11 = GetEventRouter!(execute=@n10.else, Scope:Object Class Reference=/Script/FortniteGame.FortPlayspace, ContextActor:Actor Object Reference=GetOwningPlayer()) -> (then, ReturnValue:Gameplay Event Router Component Object Reference)
   > n12 = StartListeningToEvent!(
    execute:Exec
    Target:Gameplay Event Router Component Object Reference=@n11
    EventType:Script Struct Object Reference=/Script/FMCalibrationRuntime.FMEvent_Calibration_ModalClosed
    Context:Object Reference
    EventBubblingRule:EEventBubblingRule Enum
   ) -> (
    then:Exec
    AsyncTaskProxy:Async Action Start Listening To Event Object Reference
    OnEventReceived:Exec
    Payload:FMEvent Calibration Modal Closed Structure
   )
   > Set CalibrationModalSkipped Task Ref(CalibrationModalSkipped Task Ref:Async Action Start Listening To Event Object Reference=@n12) -> (Output_Get:Async Action Start Listening To Event Object Reference)
   CommitCalibrationValues(execute=@n12.OnEventReceived, self:Calibration Process VM Object Reference=Get CalibrationProcessVM())
   > CloseCalibration!(
    execute
    self:WBP FMCalibration Calibration Screen Object Reference
    CalibrationProcessVM:Calibration Process VM Object Reference=Get CalibrationProcessVM()
   ) -> (then)
   Bind Event to On Calibrometer Enabled Changed(execute=@n7.then_3, self:FMCalibration Controller Component Object Reference=Get FMCalibrationComponent(), Delegate=@n6) {is=title K2Node_AddDelegate}
   > n13 = DelayUntilNextTick!(execute) -> (then)
   n14 = IsPlatformTraitSupported!(
    execute=@n7.then_2
    TraitToCheck:"Gameplay Tag Structure (by ref)"=MakeLiteralGameplayTag!(Value:Gameplay Tag Structure=(TagName="Platform.Trait.Input.SupportsTouch")) -> (ReturnValue:Gameplay Tag Structure)
   ) -> (then, ReturnValue:bool)
   > Branch(Condition=@n14)
   > SetPadding!(
    execute
    self:Overlay Slot Object Reference=SlotAsOverlaySlot(Widget=Get BottomNav() -> (BottomNav:Horizontal Box Object Reference))
    InPadding_Left:float=16
    InPadding_Top:float=16
    InPadding_Right:float=48
    InPadding_Bottom:float=24
   ) -> (then)
   > SetPadding(
    self=SlotAsOverlaySlot(Widget=Get MainContainer() -> (MainContainer:Overlay Object Reference))
    InPadding_Left=0
    InPadding_Top=100
    InPadding_Right=0
    InPadding_Bottom=128
   )
   > SetPadding(
    self:Vertical Box Slot Object Reference=SlotAsVerticalBoxSlot(Widget=Get Text_Header() -> (Text_Header:Common Text Object Reference))
    InPadding_Left=0
    InPadding_Top=0
    InPadding_Right=0
    InPadding_Bottom=90
   )
   > SetPadding(
    self=SlotAsVerticalBoxSlot(Widget=Get Text_Body() -> (Text_Body:Common Text Object Reference))
    InPadding_Left=0
    InPadding_Top=0
    InPadding_Right=0
    InPadding_Bottom=0
   )
   Event On Deactivated()
   > Parent: On Deactivated() {is=title K2Node_CallParentFunction}
   > StopListeningToEvent(self:Async Action Start Listening To Event Object Reference=Get CalibrationModalSkipped Task Ref() -> (CalibrationModalSkipped Task Ref:Async Action Start Listening To Event Object Reference))
   > Unbind Event from On Calibrometer Enabled Changed(self:FMCalibration Controller Component Object Reference=Get FMCalibrationComponent(), Delegate=@n6) {is=title K2Node_RemoveDelegate}
   > StopCalibrometerDisplay()
   Event PreConstruct()
   > SetToolTipText!(
    execute
    self:Widget Object Reference=Get WBP_UIKit_Button_Quiet_ProVocals() -> (WBP_UIKit_Button_Quiet_ProVocals:WBP UIKit Button Quiet Object Reference)
    InToolTipText:"Text (by ref)"=Format Text!(
     Format="LOCTEXT(\"Calibrometer {ProVocals} Tester\")"
     ProVocals:Text=MakeLiteralText(Value="LOCTABLE(\"/SparksSettings/UI/ST_Sparks_InstrumentNames\", \"ProVocals\")")
    ) -> (Result)
   ) -> (then)
   > SetToolTipText(
    self=Get WBP_UIKit_Button_Quiet_InputToVideo() -> (WBP_UIKit_Button_Quiet_InputToVideo:WBP UIKit Button Quiet Object Reference)
    InToolTipText="To Text (String)"(InString=MakeLiteralString(Value=Calibrometer Input to Video Testing Screen))
   )
   CloseQuestScreen(self:BP FMCalibration Component Object Reference=GetComponentByClass!(
     self:Actor Object Reference=GetOwningPlayer()
     ComponentClass:Actor Component Class Reference=/FMCalibration/Components/BP_FMCalibrationComponent.BP_FMCalibrationComponent_C
    ) -> (ReturnValue:BP FMCalibration Component Object Reference))
  ]}
 StartCalibrometerDisplay {nodes [
   StartCalibrometerDisplay() {is=title K2Node_FunctionEntry}
   > n1 = Sequence!(execute) -> (then_0, then_1, then_2)
   n2 = GetCalibrationHelper!(self:BP FMCalibration Component Object Reference=Get FMCalibrationComponent()) -> (Calibration Helper:BP FMCalibration Helper Object Reference)
   > Start Metronome!(
    execute
    self:BP FMCalibration Helper Object Reference=@n2
    Tempo:double=GetBpm!(self:Calibration Process VM Object Reference=Get CalibrationProcessVM()) -> (ReturnValue:float)
    Silent:bool=false
    UseHiHats:bool=false
    Use Calibrometer:bool=true
    ProVocals:bool=false
   ) -> (then)
   Set MusicClock!(
    execute=@n1.then_1
    MusicClock:Music Clock Object Reference=Get CalibrationClock!(self:BP FMCalibration Helper Object Reference=@n2) -> (CalibrationClock:Music Clock Object Reference)
    self:WBP Calibrometer Widget Object Reference=Get WBP_CalibrometerWidget_550() -> (WBP_CalibrometerWidget_550:WBP Calibrometer Widget Object Reference)
   ) -> (then, Output_Get:Music Clock Object Reference)
   SetVisibility(execute=@n1.then_2, self=Get WBP_UIKit_Button_Quiet_ProVocals(), InVisibility=SelfHitTestInvisible)
   > SetVisibility(self=Get WBP_UIKit_Button_Quiet_InputToVideo(), InVisibility=SelfHitTestInvisible)
  ]}
 StopCalibrometerDisplay {nodes [
   StopCalibrometerDisplay() {is=title K2Node_FunctionEntry}
   > n1 = Sequence()
   > Stop Metronome(self:BP FMCalibration Helper Object Reference=Get CalibrationHelper!(self:BP FMCalibration Component Object Reference=Get FMCalibrationComponent()) -> (CalibrationHelper:BP FMCalibration Helper Object Reference))
   Set MusicClock(execute=@n1.then_1, self=Get WBP_CalibrometerWidget_550())
   SetVisibility(execute=@n1.then_2, self=Get WBP_UIKit_Button_Quiet_ProVocals(), InVisibility=Collapsed)
   > SetVisibility(self=Get WBP_UIKit_Button_Quiet_InputToVideo(), InVisibility=Collapsed)
  ]}
 UpdateCalibrometerDisplay {nodes [
   UpdateCalibrometerDisplay() {is=title K2Node_FunctionEntry}
   n1 = IsCalibrometerEnabled!(self:Calibration Process VM Object Reference=Get CalibrationProcessVM()) -> (ReturnValue:bool)
   n2 = Get MusicClock!(self:WBP Calibrometer Widget Object Reference=Get WBP_CalibrometerWidget_550()) -> (MusicClock:Music Clock Object Reference)
   > Branch(Condition="Not Equal (Boolean)"(A=@n1, B=IsValid(Object=@n2)) -> (ReturnValue=~))
   > n3 = Branch(Condition=@n1)
   > StartCalibrometerDisplay() {is=title K2Node_CallFunction}
   StopCalibrometerDisplay(execute=@n3.else) {is=title K2Node_CallFunction}
  ]}
]
variables [CalibrationModalSkipped Task Ref {type=AsyncAction_StartListeningToEvent, default=null}]
widgets {
 info {
  inheritedWidgetCount=0
  parentClass=&/WBP_FMCalibration_CalibrationScreen.WBP_FMCalibration_CalibrationScreen_C
  rootWidgetClass=&Overlay
  widgetCount=24
 }
 widgets [
  Overlay_11 {widgetClassPath=&Overlay, widget=Overlay_11}
  HMX_Logo_Texture {
   widgetClassPath=&Image
   parent=Overlay_11
   widget=HMX_Logo_Texture
   properties {
    brush {
     bIsDynamicallyLoaded=false
     drawAs=Image
     imageType=FullColor
     mirroring=NoMirror
     resourceName=None
     resourceObject=/FMCalibration/Materials/MI_UI_FTUE_Logo
     tiling=NoTile
     imageSize=<1200 1200>
     margin=<ltrb 0 0 0 0>
     outlineSettings {
      bUseBrushTransparency=false
      roundingType=HalfHeightRadius
      width=0
      color {colorUseRule=UseColor_Specified, specifiedColor=<rgba 0 0 0 0>}
      cornerRadii=<0 0 0 0>
     }
     tintColor {colorUseRule=UseColor_Specified, specifiedColor=<rgba 1 1 1 1>}
     uVRegion {bIsValid=false, max=<0 0>, min=<0 0>}
    }
    colorAndOpacity=<rgba 1 1 1 0.35>
   }
   slot {class=OverlaySlot, properties {horizontalAlignment=HAlign_Center, verticalAlignment=VAlign_Center}}
  }
  MainContainer {
   widgetClassPath=&Overlay
   bIsVariable=true
   parent=Overlay_11
   widget=MainContainer
   slot {class=OverlaySlot, properties {horizontalAlignment=HAlign_Fill, verticalAlignment=VAlign_Fill, padding=<ltrb 0 100 0 100>}}
  }
  GridPanel_0 {
   widgetClassPath=&GridPanel
   parent=MainContainer
   widget=GridPanel_0
   properties {columnFill [1], rowFill [1]}
   slot {class=OverlaySlot, properties {horizontalAlignment=HAlign_Fill, verticalAlignment=VAlign_Fill}}
  }
  WBP_FMCalibration_StageTech {
   widgetClassPath=&/WBP_FMCalibration_StageTech
   parent=GridPanel_0
   widget=WBP_FMCalibration_StageTech
   properties {renderTransform {angle=0, scale=<1.1 1.1>, shear=<0 0>, translation=<0 0>}}
   slot {class=GridSlot, properties {horizontalAlignment=HAlign_Left, verticalAlignment=VAlign_Bottom, nudge=<-180 0>, padding=<ltrb 10 0 0 -55>}}
  }
  SlideContent_Padding {widgetClassPath=&VerticalBox, parent=GridPanel_0, widget=SlideContent_Padding, slot {class=GridSlot}}
  WBP_UIKit_Button_Quiet_ProVocals {
   widgetClassPath=/FortUILibrary/UIKit/Buttons/WBP_UIKit_Button_Quiet
   bIsVariable=true
   parent=SlideContent_Padding
   widget=WBP_UIKit_Button_Quiet_ProVocals
   properties {
    bNavigateToNextWidgetOnDisable=true
    text=Karaoke
    visibility=Collapsed
    triggeringInputAction {dataTable=&InputActionDataTable, rowName=GenericFaceButtonTop}
   }
   slot {class=VerticalBoxSlot, properties {horizontalAlignment=HAlign_Right, verticalAlignment=VAlign_Top, padding=<ltrb 0 0 16 0>}}
  }
  WBP_UIKit_Button_Quiet_InputToVideo {
   widgetClassPath=&WBP_UIKit_Button_Quiet
   bIsVariable=true
   parent=SlideContent_Padding
   widget=WBP_UIKit_Button_Quiet_InputToVideo
   properties {
    bNavigateToNextWidgetOnDisable=true
    text=InputToVideo
    visibility=Collapsed
    triggeringInputAction {dataTable=&InputActionDataTable, rowName=GenericFaceButtonTop}
   }
   slot {class=VerticalBoxSlot, properties {horizontalAlignment=HAlign_Right, verticalAlignment=VAlign_Top, padding=<ltrb 0 0 16 0>}}
  }
  ModalContent_Padding {
   widgetClassPath=&VerticalBox
   parent=SlideContent_Padding
   widget=ModalContent_Padding
   slot {class=VerticalBoxSlot, properties {padding=<ltrb 0 0 0 80>}}
  }
  Text_Header {
   widgetClassPath=&CommonTextBlock
   bIsVariable=true
   parent=ModalContent_Padding
   widget=Text_Header
   properties {
    justification=Center
    text=TUNE-UP
    textTransformPolicy=ToUpper
    wrapTextAt=900
    font {
     bForceMonospaced=false
     bMaterialIsStencil=false
     fontMaterial=null
     fontObject=&BurbankSmall
     letterSpacing=0
     monospacedWidth=1
     size=25.5
     skewAmount=0
     typefaceFontName=Bold
     outlineSettings {
      bApplyOutlineToDropShadows=false
      bMiteredCorners=false
      bSeparateFillAlpha=false
      outlineMaterial=null
      outlineSize=0
      outlineColor=<rgba 0 0 0 1>
     }
    }
    shadowOffset=<0 0>
   }
   slot {class=VerticalBoxSlot, properties {horizontalAlignment=HAlign_Center, verticalAlignment=VAlign_Top, padding=<ltrb 0 0 0 40>}}
  }
  Text_Body {
   widgetClassPath=&CommonTextBlock
   bIsVariable=true
   parent=ModalContent_Padding
   widget=Text_Body
   properties {
    justification=Center
    text="For the best possible experience, let's make sure that your audio, video, and input are in sync."
    wrapTextAt=900
    colorAndOpacity {colorUseRule=UseColor_Specified, specifiedColor=<rgba 0.4397 0.7835 0.8879 1>}
    font {
     bForceMonospaced=false
     bMaterialIsStencil=false
     fontMaterial=null
     fontObject=&BurbankSmall
     letterSpacing=0
     monospacedWidth=1
     size=18
     skewAmount=0
     typefaceFontName=Medium
     outlineSettings {
      bApplyOutlineToDropShadows=false
      bMiteredCorners=false
      bSeparateFillAlpha=false
      outlineMaterial=null
      outlineSize=0
      outlineColor=<rgba 0 0 0 1>
     }
    }
    shadowOffset=<0 0>
    strikeBrush {
     bIsDynamicallyLoaded=false
     drawAs=Image
     imageType=NoImage
     mirroring=NoMirror
     resourceName=None
     resourceObject=null
     tiling=NoTile
     imageSize=<32 32>
     margin=<ltrb 0 0 0 0>
     outlineSettings {
      bUseBrushTransparency=false
      roundingType=HalfHeightRadius
      width=0
      color {colorUseRule=UseColor_Specified, specifiedColor=<rgba 0 0 0 0>}
      cornerRadii=<0 0 0 0>
     }
     tintColor {colorUseRule=UseColor_Specified, specifiedColor=<rgba 0.0578 0 0 1>}
     uVRegion {bIsValid=false, max=<0 0>, min=<0 0>}
    }
   }
   slot {class=VerticalBoxSlot, properties {horizontalAlignment=HAlign_Center, verticalAlignment=VAlign_Top, padding=<ltrb 0 30 0 0>}}
  }
  WBP_CalibrometerWidget_550 {
   widgetClassPath=&/WBP_CalibrometerWidget
   bIsVariable=true
   parent=SlideContent_Padding
   widget=WBP_CalibrometerWidget_550
   slot {
    class=VerticalBoxSlot
    properties {horizontalAlignment=HAlign_Right, verticalAlignment=VAlign_Bottom, padding=<ltrb 0 0 200 0>, size {sizeRule=Fill, value=1}}
   }
  }
  SizeBox_ButtonGroup_Begin {
   widgetClassPath=&SizeBox
   bIsVariable=true
   parent=GridPanel_0
   widget=SizeBox_ButtonGroup_Begin
   properties {bOverride_WidthOverride=true, widthOverride=408}
   slot {class=GridSlot, properties {horizontalAlignment=HAlign_Center, verticalAlignment=VAlign_Bottom}}
  }
  Button_Kit_Begin {
   widgetClassPath=/FortUILibrary/UIKit/Buttons/WBP_UIKit_Button_Loud
   bIsVariable=true
   parent=SizeBox_ButtonGroup_Begin
   widget=Button_Kit_Begin
   properties {buttonSize=Large, text=Begin, triggeringInputAction {dataTable=&InputActionDataTable, rowName=GenericForward}}
   slot {class=SizeBoxSlot, properties {padding=<ltrb 0 0 0 50>}}
  }
  RootSafeZone {
   widgetClassPath=&SafeZone
   parent=Overlay_11
   widget=RootSafeZone
   properties {padLeft=false, padTop=false}
   slot {class=OverlaySlot, properties {horizontalAlignment=HAlign_Fill, verticalAlignment=VAlign_Fill}}
  }
  NavigationButtons {widgetClassPath=&Overlay, parent=RootSafeZone, widget=NavigationButtons, slot {class=SafeZoneSlot}}
  BottomNav {
   widgetClassPath=&HorizontalBox
   bIsVariable=true
   parent=NavigationButtons
   widget=BottomNav
   slot {class=OverlaySlot, properties {horizontalAlignment=HAlign_Right, verticalAlignment=VAlign_Bottom, padding=<ltrb 16 16 16 16>}}
  }
  Overlay_0 {
   widgetClassPath=&Overlay
   parent=BottomNav
   widget=Overlay_0
   slot {class=HorizontalBoxSlot, properties {horizontalAlignment=HAlign_Left, verticalAlignment=VAlign_Top}}
  }
  WBP_CaptureForPostBufferUpdate {
   widgetClassPath=&WBP_CaptureForPostBufferUpdate
   parent=Overlay_0
   slot {class=OverlaySlot, properties {horizontalAlignment=HAlign_Fill, verticalAlignment=VAlign_Fill}}
  }
  CommonActivatableWidgetSwitcher_0 {
   widgetClassPath=&CommonActivatableWidgetSwitcher
   bIsVariable=true
   parent=Overlay_0
   widget=CommonActivatableWidgetSwitcher_0
   properties {activeWidgetIndex=1, transitionDuration=0}
   slot {class=OverlaySlot}
  }
  WBP_FMCalibration_SkipButton {
   widgetClassPath=&/WBP_FMCalibration_SkipButton
   bIsVariable=true
   parent=CommonActivatableWidgetSwitcher_0
   widget=WBP_FMCalibration_SkipButton
   properties {isCalibrationStep=true}
   slot {class=WidgetSwitcherSlot, properties {horizontalAlignment=HAlign_Left, verticalAlignment=VAlign_Top}}
  }
  HorizontalBox_71 {
   widgetClassPath=&HorizontalBox
   parent=CommonActivatableWidgetSwitcher_0
   widget=HorizontalBox_71
   slot {class=WidgetSwitcherSlot}
  }
  WBP_UIKit_Button_Quiet_AdjustManually {
   widgetClassPath=&WBP_UIKit_Button_Quiet
   bIsVariable=true
   parent=HorizontalBox_71
   widget=WBP_UIKit_Button_Quiet_AdjustManually
   properties {text=Tune-up Manually, triggeringInputAction {dataTable=&InputActionDataTable, rowName=GenericFaceButtonLeft}}
   slot {class=HorizontalBoxSlot, properties {horizontalAlignment=HAlign_Left, verticalAlignment=VAlign_Top, padding=<ltrb 0 0 16 0>}}
  }
  WBP_UIKit_Button_Quiet_Back {
   widgetClassPath=&WBP_UIKit_Button_Quiet
   bIsVariable=true
   parent=HorizontalBox_71
   widget=WBP_UIKit_Button_Quiet_Back
   properties {text=Back, triggeringInputAction {dataTable=&InputActionDataTable, rowName=GenericBack}}
   slot {class=HorizontalBoxSlot}
  }
 ]
}
