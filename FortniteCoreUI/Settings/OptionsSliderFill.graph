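# UI material for an options slider fill, built on a Commander XP bar mask texture with two parameters, fillPercent and fullPercent.
# Each parameter offsets a rotated, heavily scaled U coordinate into the mask, turning the texture into a narrow edge at that position.
# The fill's red channel is subtracted from the full mask to separate the two regions.
# The fill layer adds a blue-tinted green channel, while the full layer's green is multiplied by black.
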Material /FortniteCoreUI/Settings/OptionsSliderFill
graph {
 properties {bUsedWithStaticMesh=false, materialDomain=MD_UI, pixelDepthOffsetMode=PDOM_Legacy, usedShadingModels=Default Lit}
 nodes [
  n1 = MaterialFunctionCall(Float2=MaterialFunctionCall(UVs=TextureCoordinate(), Rotation Center=Constant2Vector() {g=0.5, r=0.5}) {materialFunction=&CustomRotator}.Rotated Values) {materialFunction=&BreakOutFloat2Components}
  n2 = ScalarParameter() {defaultValue=1, group=xp, parameterName=fullPercent, sliderMax=1}
  n3 = TextureSample(UVs=Subtract(A=Multiply(A=@n1.R) {constA=50, constB=80}, B=Multiply(B=@n2) {constA=80, constB=50})) {texture=/Game/UI/ZoneCompletion/NewVersion/CommanderXP/Textures/XP_bar_mask}
  n4 = TextureCoordinate()
  n5 = Constant2Vector() {g=0.5, r=0.5}
  n6 = Add(A=MaterialFunctionCall(Float2=MaterialFunctionCall(UVs=@n4, Rotation Center=@n5) {materialFunction=&CustomRotator}.Rotated Values) {materialFunction=&BreakOutFloat2Components}.R) {constB=0.01}
  n7 = ScalarParameter() {defaultValue=0.5, group=xp, parameterName=fillPercent, sliderMax=1}
  n8 = TextureSample(UVs=Subtract(A=Multiply(A=@n6) {constA=50, constB=80}, B=Multiply(B=@n7) {constA=80, constB=50})) {texture=&XP_bar_mask}
  n9 = Clamp(None=Add(A=@n8.R, B=@n8.R))
  n10 = Subtract(A=@n3.R, B=@n9)
  n11 = Clamp(None=Subtract(A=@n3.G, B=@n9))
  n12 = Constant3Vector() {constant=<rgba 0 0 0 1>}
  n13 = Constant3Vector() {constant=<rgba 0 0.2789 1 1>}
  Add(A=Add(A=Clamp(None=@n10), B=Multiply(A=@n11, B=@n12)), B=Add(A=@n8.R, B=Multiply(A=@n8.G, B=@n13)))
 ]
}
