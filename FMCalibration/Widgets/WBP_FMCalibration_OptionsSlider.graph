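# Labelled options slider row for FMCalibration menus, with a material-driven fill, value text and touch-only step arrows.
# An analog slider holds the normalized value, which GetMappedValue maps to MinMaxValue, optionally snapped to StepSize, for prefix/suffix text.
# Value changes update text and the dynamic material's fillPercent.
# SliderChanged broadcasts the raw slider value on capture end, analog capture or, with MouseDragUpdates, while dragging; the latter two also play a sound.
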WidgetBlueprint /FMCalibration/Widgets/WBP_FMCalibration_OptionsSlider
graphs [
 EventGraph {nodes [
   "On Clicked (MobileArrowRight)"!() -> (OutputDelegate, then, Button:Common Button Base Object Reference) {is=title K2Node_ComponentBoundEvent}
   n1 = Get MenuSlider() -> (MenuSlider:Analog Slider Object Reference)
   n2 = GetValue!(self:Slider Object Reference=@n1) -> (ReturnValue:float)
   n3 = Get StepSize!(self:Slider Object Reference=@n1) -> (StepSize:float)
   > SetValue!(
    execute
    self:Slider Object Reference=Get MenuSlider()
    InValue:float="Min (Float)"(A:double=Get MaxValue!(self:Slider Object Reference=@n1) -> (MaxValue:float), B:double=float + float(A=@n2, B=@n3)) -> (ReturnValue:double) {is=title K2Node_CommutativeAssociativeBinaryOperator}
   ) -> (then)
   "On Clicked (MobileArrowLeft)"!() -> (OutputDelegate, then, Button:Common Button Base Object Reference) {is=title K2Node_ComponentBoundEvent}
   n4 = GetValue(self=@n1)
   > SetValue(
    self=Get MenuSlider()
    InValue="Max (Float)"(A=float - float(A=@n4, B=@n3), B=Get MinValue!(self:Slider Object Reference=@n1) -> (MinValue:float))
   )
   "On Controller Capture End (MenuSlider)"() {is=title K2Node_ComponentBoundEvent}
   n5 = Get MenuSlider()
   > Call Slider Changed!(execute:Exec, self:WBP FMCalibration Options Slider Object Reference, Slider Value:double=GetValue(self=@n5)) -> (then:Exec) {is=title K2Node_CallDelegate}
   n6 = "On Value Changed (MenuSlider)"!() -> (OutputDelegate, then, Value:float) {is=title K2Node_ComponentBoundEvent}
   > UpdateSliderWidget!(execute, self, Value:double=@n6.Value) -> (then)
   "On Mouse Capture End (MenuSlider)"() {is=title K2Node_ComponentBoundEvent}
   > Call Slider Changed(Slider Value=GetValue(self=@n5))
   n7 = "On Analog Capture (MenuSlider)"!() -> (OutputDelegate, then, Value:float) {is=title K2Node_ComponentBoundEvent}
   > Call Slider Changed(Slider Value=@n7.Value)
   > PlaySound2D(Sound=Get SliderSound() -> (SliderSound:Sound Base Object Reference))
   n8 = OnInputMethodChanged!() -> (OutputDelegate, then, InputType:ECommonInputType Enum) {is=title K2Node_CustomEvent}
   > SetStyle!(
    execute
    self:Common Text Object Reference=Get SliderValue() -> (SliderValue:Common Text Object Reference)
    InStyle:Common Text Style Class Reference=Select!(
     MouseAndKeyboard:Common Text Style Class Reference=/FortUILibrary/Styles/Typography/Foundation/TextStyle-Base-XS-B.TextStyle-Base-XS-B_C
     Gamepad:Common Text Style Class Reference=&TextStyle-Base-XS-B_C
     Touch:Common Text Style Class Reference=/FortUILibrary/Styles/Typography/Foundation/TextStyle-Base-S-B.TextStyle-Base-S-B_C
     Count:Common Text Style Class Reference=&TextStyle-Base-XS-B_C
     Index:ECommonInputType Enum=@n8.InputType
    ) -> (ReturnValue:Common Text Style Class Reference)
   ) -> (then)
   Event Construct()
   > Set Text(Text=Get SliderLabelText() -> (SliderLabelText:Text), self=Get SliderText() -> (SliderText:Common Text Object Reference))
   > UpdateSliderWidget(Value=Get StartingValue() -> (StartingValue:double))
   n9 = Get CommonInputSubsystem()
   > Bind Event to On Input Method Changed(self=@n9, Delegate=Create Event())
   > OnInputMethodChanged!(execute, self, InputType:ECommonInputType Enum=GetCurrentInputType(self=@n9)) -> (then) {is=title K2Node_CallFunction}
   Event PreConstruct()
   > SetWidthOverride(self=Get LabelSizeBox() -> (LabelSizeBox:Size Box Object Reference), InWidthOverride=Get Label Width() -> (Label Width:double))
   > n10 = Branch(Condition=Get ShowLightMaterial() -> (ShowLightMaterial:bool))
   > n11 = CreateDynamicMaterialInstance!(execute, Parent=/FortniteCoreUI/Settings/OptionsSliderFill_Light, OptionalName=None, CreationFlags=None) -> (then, ReturnValue)
   n12 = Get SliderBackground() -> (SliderBackground:Image Object Reference)
   > n13 = SetBrushFromMaterial(self=@n12, Material=@n11)
   n14 = GetDynamicMaterial(execute=[@n13.then, @n10.else], self=@n12)
   > Set Slider Texture(Slider Texture:Material Instance Dynamic Object Reference=@n14) -> (Output_Get:Material Instance Dynamic Object Reference)
   Event OnMouseLeave!() -> (OutputDelegate, then, MouseEvent:"Pointer Event Structure (by ref)") {is=title K2Node_Event}
   > "SetText (Text)"(self=Get Tab Tooltip Text() -> (Tab Tooltip Text:Common Text Object Reference))
   > SetSliderBarColor!(
    execute
    self:Slider Object Reference=Get MenuSlider()
    InValue:Linear Color Structure=Make Linear Color!(R:float=0, G:float=0, B:float=.1, A:float=0) -> (LinearColor:Linear Color Structure) {is=title K2Node_MakeStruct}
   ) -> (then)
   Event OnMouseEnter()
   > "SetText (Text)"(self=Get Tab Tooltip Text(), InText=Get Hover Text() -> (Hover Text:Text))
   > SetSliderBarColor(self=Get MenuSlider(), InValue=Make Linear Color(R=0, G=0, B=0, A=0))
  ]}
 GetMappedValue {nodes [
   GetMappedValue() {is=title K2Node_FunctionEntry}
   n1 = Get!() -> (MinMaxValue_X:double, MinMaxValue_Y:double)
   n2 = MapRangeClamped(Value=GetValue(self=Get MenuSlider()), InRangeA=0, InRangeB=1, OutRangeA=@n1, OutRangeB=@n1.MinMaxValue_Y)
   n3 = Get StepSize!() -> (StepSize:double)
   n4 = Round(A=SafeDivide(A=@n2, B=@n3)) {is=title K2Node_CallFunction}
   > Return Node!(
    execute
    MappedValue:double=SelectFloat(A=float * float(A:int=@n4, B=@n3) {is=title K2Node_PromotableOperator}, B=@n2, bPickA=IsUsingStepSize!() -> (IsStepSizeSet:bool))
   )
  ]}
 GetSliderValueText {nodes [
   GetSliderValueText!() -> (then, Value:double=~) {is=title K2Node_FunctionEntry}
   > Return Node!(
    execute
    ValueText:Text=Format Text!(
     Format="INVTEXT(\"{prefix}{value}{suffix}\")"
     prefix:Text=Get ValuePrefixText() -> (ValuePrefixText:Text)
     value:double=GetMappedValue!() -> (MappedValue:double) {is=title K2Node_CallFunction}
     suffix:Text=Get ValueSuffixText() -> (ValueSuffixText:Text)
    ) -> (Result)
   )
  ]}
 IsUsingStepSize {nodes [
   IsUsingStepSize() {is=title K2Node_FunctionEntry}
   n1 = Get StepSize()
   > Return Node!(execute, IsStepSizeSet:bool=NOT Boolean(A="Nearly Equal (Float)"(A=@n1, B=0, ErrorTolerance=0)))
  ]}
 OnFocusReceived {nodes [
   OnFocusReceived!() -> (then, MyGeometry:Geometry Structure, InFocusEvent:Focus Event Structure) {is=title K2Node_FunctionEntry}
   > CenterWidget!(execute, ToCenter:Widget Object Reference=Get MenuSlider()) -> (then)
   > Return Node!(execute, ReturnValue:Event Reply Structure=Handled())
  ]}
 SetMappedValue {nodes [
   n1 = SetMappedValue!() -> (then, MappedValue:double=~) {is=title K2Node_FunctionEntry}
   n2 = Get()
   > UpdateSliderNormalizedValue!(
    execute
    self
    NormalizedValue:double=NormalizeToRange!(Value:double=@n1, RangeMin:double=@n2, RangeMax:double=@n2.MinMaxValue_Y) -> (ReturnValue:double)
   ) -> (then)
  ]}
 SliderChanged {nodes [SliderChanged!() -> (then, Slider Value:double=~) {is=title K2Node_FunctionEntry}]}
 UpdateSlider {nodes [
   n1 = UpdateSlider!() -> (
    then
    Slider Text:Text
    Slider Value:double=~
    Hover Text:Text
    Tooltip Text Block:Common Text Object Reference
    Min Max Value:Vector 2D Structure
   ) {is=title K2Node_FunctionEntry}
   > "SetText (Text)"(self=Get SliderText(), InText=@n1)
   > Set Hover Text(Hover Text:Text=@n1.Hover Text) -> (Output_Get:Text)
   > Set Tab Tooltip Text(Tab Tooltip Text:Common Text Object Reference=@n1.Tooltip Text Block) -> (Output_Get:Common Text Object Reference)
   > Set MinMaxValue(MinMaxValue:Vector 2D Structure=@n1.Min Max Value) -> (Output_Get:Vector 2D Structure)
   > UpdateSliderNormalizedValue(NormalizedValue=@n1.Slider Value)
  ]}
 UpdateSliderNormalizedValue {nodes [
   n1 = UpdateSliderNormalizedValue!() -> (then, NormalizedValue:double=~) {is=title K2Node_FunctionEntry}
   > "SetText (Text)"(self=Get SliderValue(), InText=GetSliderValueText!(Value:double=@n1) -> (ValueText:Text) {is=title K2Node_CallFunction})
   > SetValue(self=Get MenuSlider(), InValue=@n1)
   > SetScalarParameterValue(
    self=Get Slider Texture() -> (Slider Texture:Material Instance Dynamic Object Reference)
    ParameterName=MakeLiteralName(Value=fillPercent)
    Value=@n1
   )
  ]}
 UpdateSliderWidget {nodes [
   n1 = UpdateSliderWidget!() -> (then, Value:double=~) {is=title K2Node_FunctionEntry}
   > "SetText (Text)"(self=Get SliderValue(), InText=GetSliderValueText(Value=@n1))
   n2 = GetValue(self=Get MenuSlider())
   > SetScalarParameterValue(self=Get Slider Texture(), ParameterName=MakeLiteralName(Value=fillPercent), Value=@n2)
   > Branch(Condition=HasMouseCapture!(self:Widget Object Reference=Get MenuSlider()) -> (ReturnValue:bool))
   > Branch(Condition=Get MouseDragUpdates() -> (MouseDragUpdates:bool))
   > Call Slider Changed(Slider Value=@n2)
   > PlaySound2D(Sound=Get SliderSound())
  ]}
]
variables [
 Hover Text {type=string, default=""}
 Label Width {type=number, default=300}
 MaxFractionalDigits {type=integer, default=2}
 MaxIntegralDigits {type=integer, default=1}
 MinFractionalDigits {type=integer, default=2}
 MinIntegralDigits {type=integer, default=1}
 MinMaxValue {type=Vector2D, default=<0 1>}
 MouseDragUpdates {type=boolean}
 ShowLightMaterial {type=boolean}
 Slider Texture {type=MaterialInstanceDynamic}
 SliderChanged {}
 SliderLabelText {type=string, default=""}
 SliderSound {type=SoundBase, default=/Game/Sounds/Fort_UI_Sounds/Replay/DVR_CamOptions_Slider_Cue}
 StartingValue {type=number, default=0.5}
 StepSize {type=number}
 Tab Tooltip Text {type=CommonTextBlock, default=null}
 ValuePrefixText {type=string, default=""}
 ValueSuffixText {type=string, default=""}
]
widgets {
 info {inheritedWidgetCount=0, parentClass=&UserWidget, rootWidgetClass=&HorizontalBox, widgetCount=17}
 widgets [
  HorizontalBox_1 {widgetClassPath=&HorizontalBox, widget=HorizontalBox_1, properties {visibility=Visible}}
  LabelSizeBox {
   widgetClassPath=&SizeBox
   bIsVariable=true
   parent=HorizontalBox_1
   widget=LabelSizeBox
   properties {minDesiredWidth=300, widthOverride=300}
   slot {class=HorizontalBoxSlot, properties {padding=<ltrb 0 0 20 0>, size {sizeRule=Fill, value=1}}}
  }
  SliderText {
   widgetClassPath=&CommonTextBlock
   bIsVariable=true
   parent=LabelSizeBox
   widget=SliderText
   properties {
    text=Slider Text
    font {
     bForceMonospaced=false
     bMaterialIsStencil=false
     fontMaterial=null
     fontObject=&HeadingNow
     letterSpacing=10
     monospacedWidth=1
     size=18
     skewAmount=0
     typefaceFontName=74 Regular
     outlineSettings {
      bApplyOutlineToDropShadows=false
      bMiteredCorners=false
      bSeparateFillAlpha=false
      outlineMaterial=null
      outlineSize=0
      outlineColor=<rgba 0 0 0 1>
     }
    }
    shadowOffset=<0 0>
   }
   slot {class=SizeBoxSlot, properties {horizontalAlignment=HAlign_Left, verticalAlignment=VAlign_Center}}
  }
  CommonHardwareVisibilityBorder_0 {
   widgetClassPath=&CommonHardwareVisibilityBorder
   parent=HorizontalBox_1
   widget=CommonHardwareVisibilityBorder_0
   properties {visibilityQuery {
     autoDescription=" ANY( Input.Touch )"
     tokenStreamVersion=0
     userDescription=""
     queryTokenStream [0,1,1,1,0]
     tagDictionary [{tagName=Input.Touch}]
    }}
   slot {class=HorizontalBoxSlot}
  }
  SizeBox_2 {
   widgetClassPath=&SizeBox
   parent=CommonHardwareVisibilityBorder_0
   widget=SizeBox_2
   properties {bOverride_MinDesiredWidth=true, bOverride_WidthOverride=true, minDesiredWidth=120, widthOverride=120}
   slot {class=BorderSlot, properties {padding=<ltrb 0 0 0 0>}}
  }
  CommonVisualAttachment_2 {
   widgetClassPath=&CommonVisualAttachment
   parent=SizeBox_2
   widget=CommonVisualAttachment_2
   properties {bOverride_HeightOverride=true, bOverride_WidthOverride=true, heightOverride=95, widthOverride=120, contentAnchor=<0 0.125>}
   slot {class=SizeBoxSlot}
  }
  MobileArrowLeft {
   widgetClassPath=&WBP_UIKit_Button_Generic
   bIsVariable=true
   parent=CommonVisualAttachment_2
   widget=MobileArrowLeft
   properties {
    bIsFocusable=false
    icon=/FortUILibrary/Textures/Icons/T_UI_Icon_UIKit_Chevron
    iconMaterial=/FMCalibration/Materials/MI_UIKit_GenericIcon_Regular
    text=""
    useTextBlock=false
    renderTransform {angle=0, scale=<-1 1>, shear=<0 0>, translation=<0 0>}
   }
   slot {class=SizeBoxSlot}
  }
  SizeBox_3 {
   widgetClassPath=&SizeBox
   parent=HorizontalBox_1
   widget=SizeBox_3
   properties {bOverride_MinDesiredHeight=true, bOverride_MinDesiredWidth=true, minDesiredHeight=70, minDesiredWidth=320}
   slot {class=HorizontalBoxSlot, properties {horizontalAlignment=HAlign_Left, padding=<ltrb 20 0 20 0>, size {sizeRule=Fill, value=1}}}
  }
  Button_0 {
   widgetClassPath=&Button
   parent=SizeBox_3
   widget=Button_0
   properties {
    visibility=SelfHitTestInvisible
    widgetStyle {
     clickedSlateSound {resourceObject=null}
     disabled {
      bIsDynamicallyLoaded=false
      drawAs=NoDrawType
      imageType=NoImage
      mirroring=NoMirror
      resourceName=None
      resourceObject=null
      tiling=NoTile
      imageSize=<0 0>
      margin=<ltrb 0 0 0 0>
      outlineSettings {
       bUseBrushTransparency=false
       roundingType=HalfHeightRadius
       width=0
       color {colorUseRule=UseColor_Specified, specifiedColor=<rgba 0 0 0 0>}
       cornerRadii=<0 0 0 0>
      }
      tintColor {colorUseRule=UseColor_Specified, specifiedColor=<rgba 1 1 1 0>}
      uVRegion {bIsValid=false, max=<0 0>, min=<0 0>}
     }
     disabledForeground {colorUseRule=UseColor_Specified, specifiedColor=<rgba 0.5271 0.5271 0.5271 1>}
     hovered {
      bIsDynamicallyLoaded=false
      drawAs=RoundedBox
      imageType=NoImage
      mirroring=NoMirror
      resourceName=None
      resourceObject=/FortUILibrary/Foundation/Buttons/Materials/M_Button_Settings_Select
      tiling=NoTile
      imageSize=<8 8>
      margin=<ltrb 0.49 0.49 0.49 0.49>
      outlineSettings {
       bUseBrushTransparency=true
       roundingType=FixedRadius
       width=1
       color {colorUseRule=UseColor_Specified, specifiedColor=<rgba 0.7243 0.7243 0.7243 1>}
       cornerRadii=<4 4 4 4>
      }
      tintColor {colorUseRule=UseColor_Specified, specifiedColor=<rgba 1 0.4955 0.0131 1>}
      uVRegion {bIsValid=false, max=<0 0>, min=<0 0>}
     }
     hoveredForeground {colorUseRule=UseColor_Specified, specifiedColor=<rgba 0 1 1 1>}
     hoveredSlateSound {resourceObject=null}
     normal {
      bIsDynamicallyLoaded=false
      drawAs=RoundedBox
      imageType=NoImage
      mirroring=NoMirror
      resourceName=None
      resourceObject=null
      tiling=NoTile
      imageSize=<32 32>
      margin=<ltrb 0 0 0 0>
      outlineSettings {
       bUseBrushTransparency=true
       roundingType=FixedRadius
       width=1
       color {colorUseRule=UseColor_Specified, specifiedColor=<rgba 0.6951 0.6951 0.6951 1>}
       cornerRadii=<4 4 4 4>
      }
      tintColor {colorUseRule=UseColor_Specified, specifiedColor=<rgba 1 1 1 0>}
      uVRegion {bIsValid=false, max=<0 0>, min=<0 0>}
     }
     normalForeground {colorUseRule=UseColor_Specified, specifiedColor=<rgba 0 1 1 1>}
     normalPadding=<ltrb 4 4 4 4>
     pressed {
      bIsDynamicallyLoaded=false
      drawAs=RoundedBox
      imageType=NoImage
      mirroring=NoMirror
      resourceName=None
      resourceObject=null
      tiling=NoTile
      imageSize=<32 32>
      margin=<ltrb 0 0 0 0>
      outlineSettings {
       bUseBrushTransparency=true
       roundingType=FixedRadius
       width=1
       color {colorUseRule=UseColor_Specified, specifiedColor=<rgba 0.7243 0.7243 0.7243 1>}
       cornerRadii=<4 4 4 4>
      }
      tintColor {colorUseRule=UseColor_Specified, specifiedColor=<rgba 1 1 1 0>}
      uVRegion {bIsValid=false, max=<0 0>, min=<0 0>}
     }
     pressedForeground {colorUseRule=UseColor_Specified, specifiedColor=<rgba 0 1 1 1>}
     pressedPadding=<ltrb 4 4 4 4>
     pressedSlateSound {resourceObject=null}
    }
   }
   slot {class=SizeBoxSlot}
  }
  Overlay_0 {
   widgetClassPath=&Overlay
   parent=Button_0
   widget=Overlay_0
   slot {class=ButtonSlot, properties {horizontalAlignment=HAlign_Fill, verticalAlignment=VAlign_Fill, padding=<ltrb 0 0 0 0>}}
  }
  MenuSlider {
   widgetClassPath=/Script/CommonUI.AnalogSlider
   bIsVariable=true
   parent=Overlay_0
   widget=MenuSlider
   properties {
    bOverride_Cursor=true
    cursor=ResizeLeftRight
    stepSize=0.005
    sliderBarColor=<rgba 1 1 1 0>
    sliderHandleColor=<rgba 1 1 1 0>
    widgetStyle {
     barThickness=12
     disabledBarImage {
      bIsDynamicallyLoaded=false
      drawAs=Image
      imageType=NoImage
      mirroring=NoMirror
      resourceName=None
      resourceObject=null
      tiling=NoTile
      imageSize=<0 0>
      margin=<ltrb 0 0 0 0>
      outlineSettings {
       bUseBrushTransparency=false
       roundingType=HalfHeightRadius
       width=0
       color {colorUseRule=UseColor_Specified, specifiedColor=<rgba 0 0 0 0>}
       cornerRadii=<0 0 0 0>
      }
      tintColor {colorUseRule=UseColor_Specified, specifiedColor=<rgba 0.5 0.5 0.5 1>}
      uVRegion {bIsValid=false, max=<0 0>, min=<0 0>}
     }
     disabledThumbImage {
      bIsDynamicallyLoaded=false
      drawAs=Image
      imageType=FullColor
      mirroring=NoMirror
      resourceName=../../../Engine/Content/Slate/Common/Button_Disabled.png
      resourceObject=null
      tiling=NoTile
      imageSize=<8 14>
      margin=<ltrb 0 0 0 0>
      outlineSettings {
       bUseBrushTransparency=false
       roundingType=HalfHeightRadius
       width=0
       color {colorUseRule=UseColor_Specified, specifiedColor=<rgba 0 0 0 0>}
       cornerRadii=<0 0 0 0>
      }
      tintColor {colorUseRule=UseColor_Specified, specifiedColor=<rgba 1 1 1 1>}
      uVRegion {bIsValid=false, max=<0 0>, min=<0 0>}
     }
     hoveredBarImage {
      bIsDynamicallyLoaded=false
      drawAs=Image
      imageType=NoImage
      mirroring=NoMirror
      resourceName=None
      resourceObject=null
      tiling=NoTile
      imageSize=<0 0>
      margin=<ltrb 0 0 0 0>
      outlineSettings {
       bUseBrushTransparency=false
       roundingType=HalfHeightRadius
       width=0
       color {colorUseRule=UseColor_Specified, specifiedColor=<rgba 0 0 0 0>}
       cornerRadii=<0 0 0 0>
      }
      tintColor {colorUseRule=UseColor_Specified, specifiedColor=<rgba 1 1 1 1>}
      uVRegion {bIsValid=false, max=<0 0>, min=<0 0>}
     }
     hoveredThumbImage {
      bIsDynamicallyLoaded=false
      drawAs=Image
      imageType=FullColor
      mirroring=NoMirror
      resourceName=../../../Engine/Content/Slate/Common/Button.png
      resourceObject=null
      tiling=NoTile
      imageSize=<8 14>
      margin=<ltrb 0 0 0 0>
      outlineSettings {
       bUseBrushTransparency=false
       roundingType=HalfHeightRadius
       width=0
       color {colorUseRule=UseColor_Specified, specifiedColor=<rgba 0 0 0 0>}
       cornerRadii=<0 0 0 0>
      }
      tintColor {colorUseRule=UseColor_Specified, specifiedColor=<rgba 1 1 1 1>}
      uVRegion {bIsValid=false, max=<0 0>, min=<0 0>}
     }
     normalBarImage {
      bIsDynamicallyLoaded=false
      drawAs=Image
      imageType=NoImage
      mirroring=NoMirror
      resourceName=None
      resourceObject=null
      tiling=NoTile
      imageSize=<0 0>
      margin=<ltrb 0 0 0 0>
      outlineSettings {
       bUseBrushTransparency=false
       roundingType=HalfHeightRadius
       width=0
       color {colorUseRule=UseColor_Specified, specifiedColor=<rgba 0 0 0 0>}
       cornerRadii=<0 0 0 0>
      }
      tintColor {colorUseRule=UseColor_Specified, specifiedColor=<rgba 1 1 1 1>}
      uVRegion {bIsValid=false, max=<0 0>, min=<0 0>}
     }
     normalThumbImage {
      bIsDynamicallyLoaded=false
      drawAs=Image
      imageType=FullColor
      mirroring=NoMirror
      resourceName=../../../Engine/Content/Slate/Common/Button.png
      resourceObject=null
      tiling=NoTile
      imageSize=<8 14>
      margin=<ltrb 0 0 0 0>
      outlineSettings {
       bUseBrushTransparency=false
       roundingType=HalfHeightRadius
       width=0
       color {colorUseRule=UseColor_Specified, specifiedColor=<rgba 0 0 0 0>}
       cornerRadii=<0 0 0 0>
      }
      tintColor {colorUseRule=UseColor_Specified, specifiedColor=<rgba 1 1 1 1>}
      uVRegion {bIsValid=false, max=<0 0>, min=<0 0>}
     }
    }
   }
   slot {class=OverlaySlot, properties {horizontalAlignment=HAlign_Fill, verticalAlignment=VAlign_Fill}}
  }
  SliderBackground {
   widgetClassPath=&Image
   bIsVariable=true
   parent=Overlay_0
   widget=SliderBackground
   properties {
    visibility=SelfHitTestInvisible
    brush {
     bIsDynamicallyLoaded=false
     drawAs=Image
     imageType=NoImage
     mirroring=NoMirror
     resourceName=None
     resourceObject=/FortniteCoreUI/Settings/OptionsSliderFill
     tiling=NoTile
     imageSize=<32 24>
     margin=<ltrb 0 0 0 0>
     outlineSettings {
      bUseBrushTransparency=false
      roundingType=HalfHeightRadius
      width=0
      color {colorUseRule=UseColor_Specified, specifiedColor=<rgba 0 0 0 0>}
      cornerRadii=<0 0 0 0>
     }
     tintColor {colorUseRule=UseColor_Specified, specifiedColor=<rgba 1 1 1 1>}
     uVRegion {bIsValid=false, max=<0 0>, min=<0 0>}
    }
   }
   slot {class=OverlaySlot, properties {horizontalAlignment=HAlign_Fill, verticalAlignment=VAlign_Fill}}
  }
  SliderValue {
   widgetClassPath=&CommonTextBlock
   bIsVariable=true
   parent=Overlay_0
   widget=SliderValue
   properties {
    style=&TextStyle-Base-XS-B_C
    font {
     bForceMonospaced=false
     bMaterialIsStencil=false
     fontMaterial=null
     fontObject=&NotoSans
     letterSpacing=0
     monospacedWidth=1
     size=16
     skewAmount=0
     typefaceFontName=NotoSans-Bold
     outlineSettings {
      bApplyOutlineToDropShadows=false
      bMiteredCorners=false
      bSeparateFillAlpha=false
      outlineMaterial=null
      outlineSize=0
      outlineColor=<rgba 0 0 0 1>
     }
    }
    shadowOffset=<0 0>
   }
   slot {class=OverlaySlot, properties {verticalAlignment=VAlign_Center, padding=<ltrb 15 0 0 0>}}
  }
  CommonHardwareVisibilityBorder_1 {
   widgetClassPath=&CommonHardwareVisibilityBorder
   parent=HorizontalBox_1
   widget=CommonHardwareVisibilityBorder_1
   properties {visibilityQuery {
     autoDescription=" ANY( Input.Touch )"
     tokenStreamVersion=0
     userDescription=""
     queryTokenStream [0,1,1,1,0]
     tagDictionary [{tagName=Input.Touch}]
    }}
   slot {class=HorizontalBoxSlot, properties {horizontalAlignment=HAlign_Right}}
  }
  SizeBox_1 {
   widgetClassPath=&SizeBox
   parent=CommonHardwareVisibilityBorder_1
   widget=SizeBox_1
   properties {bOverride_MinDesiredWidth=true, bOverride_WidthOverride=true, minDesiredWidth=120, widthOverride=120}
   slot {class=BorderSlot, properties {padding=<ltrb 0 0 0 0>}}
  }
  CommonVisualAttachment_3 {
   widgetClassPath=&CommonVisualAttachment
   parent=SizeBox_1
   widget=CommonVisualAttachment_3
   properties {bOverride_HeightOverride=true, bOverride_WidthOverride=true, heightOverride=95, widthOverride=120, contentAnchor=<0 0.125>}
   slot {class=SizeBoxSlot}
  }
  MobileArrowRight {
   widgetClassPath=&WBP_UIKit_Button_Generic
   bIsVariable=true
   parent=CommonVisualAttachment_3
   widget=MobileArrowRight
   properties {
    bIsFocusable=false
    icon=&T_UI_Icon_UIKit_Chevron
    iconMaterial=&MI_UIKit_GenericIcon_Regular
    text=""
    useTextBlock=false
   }
   slot {class=SizeBoxSlot}
  }
 ]
}
