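# Manual audio/video latency calibration screen where players set both offsets with sliders instead of running the regular Tune-Up process.
# Activation seeds the sliders from CalibrationProcessVM's intermediate latencies and focuses the audio slider.
# Save writes both values back and dispatches a UI state chart transition to the result screen; Back calls CloseManualCalibration.
# Layout padding and spacer visibility adapt to touch input, with further padding overrides on touch-capable platforms.
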WidgetBlueprint /FMCalibration/Widgets/WBP_FMCalibration_Manual
graphs [EventGraph {nodes [
   n1 = OnInputMethodChanged!() -> (OutputDelegate, then, InputType:ECommonInputType Enum) {is=title K2Node_CustomEvent}
   n2 = "Equal (Enum)"(A=@n1.InputType, B=Touch)
   SetPadding!(
    execute
    self:Overlay Slot Object Reference=SlotAsOverlaySlot(Widget=Get Overlay_NavButtons() -> (Overlay_NavButtons:Overlay Object Reference))
    InPadding_Left:float=16
    InPadding_Top:float=16
    InPadding_Right:float=Select(Option 0:float=16, Option 1:float=48, Index=@n2) -> (ReturnValue:float)
    InPadding_Bottom:float=Select(Option 0=16, Option 1=24, Index=@n2)
   ) -> (then)
   > SetPadding(
    self=SlotAsOverlaySlot(Widget=Get GridPanel_MainContainer() -> (GridPanel_MainContainer:Grid Panel Object Reference))
    InPadding_Left=0
    InPadding_Top=100
    InPadding_Right=0
    InPadding_Bottom=Select(Option 0=100, Option 1=128, Index=@n2)
   )
   "On Clicked (Button_Save_Kit)"!() -> (OutputDelegate, then, Button:Common Button Base Object Reference) {is=title K2Node_ComponentBoundEvent}
   n3 = Get CalibrationProcessVM() -> (CalibrationProcessVM:Calibration Process VM Object Reference)
   > SetIntermediateAudioLatencyMs!(
    execute
    self:Calibration Process VM Object Reference=@n3
    Latency:float=GetMappedValue!(self:WBP FMCalibration Options Slider Object Reference=Get OptionsSlider_AudioOffset() -> (OptionsSlider_AudioOffset:WBP FMCalibration Options Slider Object Reference)) -> (MappedValue:double)
   ) -> (then)
   > SetIntermediateVideoLatencyMs!(
    execute
    self:Calibration Process VM Object Reference=@n3
    Latency:float=GetMappedValue(self=Get OptionsSlider_VideoOffset() -> (OptionsSlider_VideoOffset:WBP FMCalibration Options Slider Object Reference))
   ) -> (then)
   > DispatchTransitionEvent!(
    execute
    self:UIState Chart Manager Object Reference=Get UIStateChartManager() -> (ReturnValue:UIState Chart Manager Object Reference) {is=title K2Node_GetSubsystem}
    TransitionTag:Gameplay Tag Structure=(TagName="UI.Chart.TransitionTag.FM.Calibration.OpenResultScreen")
    StateChartOwner:Local Player Object Reference=GetOwningLocalPlayer!(self:Widget Object Reference) -> (ReturnValue:Local Player Object Reference)
    AnimationHints:"Gameplay Tag Container Structure (by ref)"
    ContextChanges:"UIState Chart Context Change Set Structure (by ref)"
    Urgency:EUIStateTransitionUrgency Enum=Normal
    bBacktrack:bool=false
   ) -> (then, ReturnValue:bool)
   "On Clicked (WBP_UIKit_Button_Quiet_Back)"!() -> (OutputDelegate, then, Button:Common Button Base Object Reference) {is=title K2Node_ComponentBoundEvent}
   > CloseManualCalibration(self:WBP FMCalibration Calibration Screen Object Reference)
   Event Construct()
   > Parent: Construct()
   n4 = Get CommonInputSubsystem()
   > Bind Event to On Input Method Changed(self=@n4, Delegate=Create Event())
   > OnInputMethodChanged!(execute, self, InputType:ECommonInputType Enum=GetCurrentInputType(self=@n4)) -> (then) {is=title K2Node_CallFunction}
   > n5 = IsPlatformTraitSupported!(
    execute
    TraitToCheck:"Gameplay Tag Structure (by ref)"=MakeLiteralGameplayTag!(Value:Gameplay Tag Structure=(TagName="Platform.Trait.Input.SupportsTouch")) -> (ReturnValue:Gameplay Tag Structure)
   ) -> (then, ReturnValue:bool)
   > Branch(Condition=@n5)
   > SetPadding(self=SlotAsOverlaySlot(Widget=Get Overlay_NavButtons()), InPadding_Left=16, InPadding_Top=16, InPadding_Right=48, InPadding_Bottom=24)
   > SetPadding(self=SlotAsOverlaySlot(Widget=Get GridPanel_MainContainer()), InPadding_Left=0, InPadding_Top=100, InPadding_Right=0, InPadding_Bottom=128)
   > SetPadding(
    self:Vertical Box Slot Object Reference=SlotAsVerticalBoxSlot(Widget=Get Text_Header() -> (Text_Header:Common Text Object Reference))
    InPadding_Left=0
    InPadding_Top=0
    InPadding_Right=0
    InPadding_Bottom=90
   )
   > SetPadding(
    self=SlotAsVerticalBoxSlot(Widget=Get Text_Body() -> (Text_Body:Common Text Object Reference))
    InPadding_Left=0
    InPadding_Top=0
    InPadding_Right=0
    InPadding_Bottom=0
   )
   Event On Activated()
   > Parent: On Activated() {is=title K2Node_CallParentFunction}
   > SetFocus(self=Get OptionsSlider_AudioOffset())
   > SetMappedValue!(
    execute
    self:WBP FMCalibration Options Slider Object Reference=Get OptionsSlider_AudioOffset()
    MappedValue:double=GetIntermediateAudioLatencyMs!(self:Calibration Process VM Object Reference=Get CalibrationProcessVM()) -> (ReturnValue:float)
   ) -> (then)
   > SetMappedValue(
    self=Get OptionsSlider_VideoOffset()
    MappedValue=GetIntermediateVideoLatencyMs!(self:Calibration Process VM Object Reference=Get CalibrationProcessVM()) -> (ReturnValue:float)
   )
   SetVisibility(
    execute=@n1.then
    self=Get TouchSpacer() -> (TouchSpacer:Spacer Object Reference)
    InVisibility=Select!(
     MouseAndKeyboard:ESlateVisibility Enum=Collapsed
     Gamepad:ESlateVisibility Enum=Collapsed
     Touch:ESlateVisibility Enum=HitTestInvisible
     Count:ESlateVisibility Enum=Collapsed
     Index:ECommonInputType Enum=@n1.InputType
    ) -> (ReturnValue:ESlateVisibility Enum)
   )
  ]}]
variables []
widgets {
 info {
  inheritedWidgetCount=0
  parentClass=&/WBP_FMCalibration_CalibrationScreen.WBP_FMCalibration_CalibrationScreen_C
  rootWidgetClass=&Overlay
  widgetCount=19
 }
 widgets [
  Overlay_11 {widgetClassPath=&Overlay, widget=Overlay_11}
  RootSafeZone {
   widgetClassPath=&SafeZone
   parent=Overlay_11
   widget=RootSafeZone
   slot {class=OverlaySlot, properties {horizontalAlignment=HAlign_Fill, verticalAlignment=VAlign_Fill}}
  }
  MainContainer {widgetClassPath=&Overlay, parent=RootSafeZone, widget=MainContainer, slot {class=SafeZoneSlot}}
  GridPanel_MainContainer {
   widgetClassPath=&GridPanel
   bIsVariable=true
   parent=MainContainer
   widget=GridPanel_MainContainer
   properties {columnFill [1], rowFill [1]}
   slot {class=OverlaySlot, properties {horizontalAlignment=HAlign_Fill, verticalAlignment=VAlign_Fill, padding=<ltrb 0 100 0 100>}}
  }
  SlideContent_Padding {
   widgetClassPath=&VerticalBox
   parent=GridPanel_MainContainer
   widget=SlideContent_Padding
   slot {class=GridSlot, properties {padding=<ltrb 0 0 0 50>}}
  }
  ModalContent_Padding {
   widgetClassPath=&VerticalBox
   parent=SlideContent_Padding
   widget=ModalContent_Padding
   slot {class=VerticalBoxSlot, properties {padding=<ltrb 0 0 0 80>}}
  }
  Text_Header {
   widgetClassPath=&CommonTextBlock
   bIsVariable=true
   parent=ModalContent_Padding
   widget=Text_Header
   properties {
    justification=Center
    text=TUNE-UP MANUALLY
    textTransformPolicy=ToUpper
    wrapTextAt=900
    font {
     bForceMonospaced=false
     bMaterialIsStencil=false
     fontMaterial=null
     fontObject=&BurbankSmall
     letterSpacing=0
     monospacedWidth=1
     size=25.5
     skewAmount=0
     typefaceFontName=Bold
     outlineSettings {
      bApplyOutlineToDropShadows=false
      bMiteredCorners=false
      bSeparateFillAlpha=false
      outlineMaterial=null
      outlineSize=0
      outlineColor=<rgba 0 0 0 1>
     }
    }
    shadowOffset=<0 0>
   }
   slot {class=VerticalBoxSlot, properties {horizontalAlignment=HAlign_Center, verticalAlignment=VAlign_Top, padding=<ltrb 0 0 0 40>}}
  }
  Text_Body {
   widgetClassPath=&CommonTextBlock
   bIsVariable=true
   parent=ModalContent_Padding
   widget=Text_Body
   properties {
    justification=Center
    text="If you already know what your Audio Latency and Video Latency values should be, you can enter them manually below. If you aren’t sure, select Back and use the regular Tune-Up process!"
    wrapTextAt=900
    colorAndOpacity {colorUseRule=UseColor_Specified, specifiedColor=<rgba 0.4397 0.7835 0.8879 1>}
    font {
     bForceMonospaced=false
     bMaterialIsStencil=false
     fontMaterial=null
     fontObject=&BurbankSmall
     letterSpacing=0
     monospacedWidth=1
     size=18
     skewAmount=0
     typefaceFontName=Medium
     outlineSettings {
      bApplyOutlineToDropShadows=false
      bMiteredCorners=false
      bSeparateFillAlpha=false
      outlineMaterial=null
      outlineSize=0
      outlineColor=<rgba 0 0 0 1>
     }
    }
    shadowOffset=<0 0>
   }
   slot {class=VerticalBoxSlot, properties {horizontalAlignment=HAlign_Center, verticalAlignment=VAlign_Top, padding=<ltrb 0 30 0 0>}}
  }
  NavContainer {
   widgetClassPath=&VerticalBox
   bIsVariable=true
   parent=SlideContent_Padding
   widget=NavContainer
   slot {class=VerticalBoxSlot, properties {horizontalAlignment=HAlign_Center, size {sizeRule=Fill, value=1}}}
  }
  ManualCalibrationSlide {
   widgetClassPath=&VerticalBox
   parent=NavContainer
   widget=ManualCalibrationSlide
   slot {class=VerticalBoxSlot, properties {verticalAlignment=VAlign_Top}}
  }
  OptionsSlider_AudioOffset {
   widgetClassPath=&/WBP_FMCalibration_OptionsSlider
   bIsVariable=true
   parent=ManualCalibrationSlide
   widget=OptionsSlider_AudioOffset
   properties {
    cursor=ResizeLeftRight
    maxFractionalDigits=0
    maxIntegralDigits=4
    minFractionalDigits=0
    mouseDragUpdates=true
    sliderLabelText=Audio Latency
    stepSize=5
    valueSuffixText=" milliseconds"
    minMaxValue=<-500 500>
   }
   slot {class=VerticalBoxSlot, properties {padding=<ltrb 0 0 0 15>}}
  }
  TouchSpacer {
   widgetClassPath=&Spacer
   bIsVariable=true
   parent=ManualCalibrationSlide
   widget=TouchSpacer
   properties {visibility=Collapsed}
   slot {class=VerticalBoxSlot, properties {padding=<ltrb 0 14 0 14>}}
  }
  OptionsSlider_VideoOffset {
   widgetClassPath=&WBP_FMCalibration_OptionsSlider
   bIsVariable=true
   parent=ManualCalibrationSlide
   widget=OptionsSlider_VideoOffset
   properties {
    cursor=ResizeLeftRight
    maxFractionalDigits=0
    maxIntegralDigits=4
    minFractionalDigits=0
    mouseDragUpdates=true
    sliderLabelText=Video Latency
    stepSize=5
    valueSuffixText=" milliseconds"
    minMaxValue=<-500 500>
   }
   slot {class=VerticalBoxSlot, properties {padding=<ltrb 0 0 0 15>}}
  }
  SizeBox_Button_Save {
   widgetClassPath=&SizeBox
   parent=NavContainer
   widget=SizeBox_Button_Save
   properties {bOverride_MinDesiredWidth=true, minDesiredWidth=408, widthOverride=408}
   slot {class=VerticalBoxSlot, properties {horizontalAlignment=HAlign_Center, verticalAlignment=VAlign_Bottom, size {sizeRule=Fill, value=1}}}
  }
  Button_Save_Kit {
   widgetClassPath=/FortUILibrary/UIKit/Buttons/WBP_UIKit_Button_Loud
   bIsVariable=true
   parent=SizeBox_Button_Save
   widget=Button_Save_Kit
   properties {
    buttonSize=Large
    minHeight=56
    text=Save
    triggeringInputAction {dataTable=/FMCalibration/Data/FMCalibration_InputActionDataTable, rowName=FMCalibrateSave}
   }
   slot {class=SizeBoxSlot}
  }
  Overlay_NavButtons {
   widgetClassPath=&Overlay
   bIsVariable=true
   parent=MainContainer
   widget=Overlay_NavButtons
   slot {class=OverlaySlot, properties {horizontalAlignment=HAlign_Right, verticalAlignment=VAlign_Bottom, padding=<ltrb 16 16 16 16>}}
  }
  WBP_CaptureForPostBufferUpdate {
   widgetClassPath=&WBP_CaptureForPostBufferUpdate
   parent=Overlay_NavButtons
   slot {class=OverlaySlot, properties {horizontalAlignment=HAlign_Fill, verticalAlignment=VAlign_Fill}}
  }
  BottomNav {widgetClassPath=&HorizontalBox, parent=Overlay_NavButtons, widget=BottomNav, slot {class=OverlaySlot}}
  WBP_UIKit_Button_Quiet_Back {
   widgetClassPath=/FortUILibrary/UIKit/Buttons/WBP_UIKit_Button_Quiet
   bIsVariable=true
   parent=BottomNav
   widget=WBP_UIKit_Button_Quiet_Back
   properties {text=Back, triggeringInputAction {dataTable=&InputActionDataTable, rowName=GenericBack}}
   slot {class=HorizontalBoxSlot, properties {horizontalAlignment=HAlign_Left, verticalAlignment=VAlign_Top}}
  }
 ]
}
