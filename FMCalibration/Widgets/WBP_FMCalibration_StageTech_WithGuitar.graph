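# CommonUserWidget for the FM calibration UI that shows the stage tech with guitar artwork over soft colored glows.
# A root Overlay stacks a centered, enlarged glow box holding faint pink and blue glow images, each offset by render transforms, behind the main character image.
# All three visuals use material instances as brushes.
# The event graph holds only unconnected stub events, so the widget is purely presentational.
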WidgetBlueprint /FMCalibration/Widgets/WBP_FMCalibration_StageTech_WithGuitar
graphs [EventGraph {nodes [
   Event PreConstruct()
   Event Construct()
   Event Tick()
  ]}]
variables []
widgets {
 info {inheritedWidgetCount=0, parentClass=&CommonUserWidget, rootWidgetClass=&Overlay, widgetCount=5}
 widgets [
  StageTech_ImageBox {widgetClassPath=&Overlay, bIsVariable=true, widget=StageTech_ImageBox}
  Glow_Box {
   widgetClassPath=&Overlay
   parent=StageTech_ImageBox
   widget=Glow_Box
   properties {renderTransform {angle=0, scale=<1.2 1.2>, shear=<0 0>, translation=<0 0>}}
   slot {class=OverlaySlot, properties {horizontalAlignment=HAlign_Center, verticalAlignment=VAlign_Center}}
  }
  PinkGlow {
   widgetClassPath=&Image
   bIsVariable=true
   parent=Glow_Box
   widget=PinkGlow
   properties {
    brush {
     bIsDynamicallyLoaded=false
     drawAs=Image
     imageType=FullColor
     mirroring=NoMirror
     resourceName=None
     resourceObject=/FMCalibration/Materials/MI_UI_GlowTechPink
     tiling=NoTile
     imageSize=<600 600>
     margin=<ltrb 0 0 0 0>
     outlineSettings {
      bUseBrushTransparency=false
      roundingType=HalfHeightRadius
      width=0
      color {colorUseRule=UseColor_Specified, specifiedColor=<rgba 0 0 0 0>}
      cornerRadii=<0 0 0 0>
     }
     tintColor {colorUseRule=UseColor_Specified, specifiedColor=<rgba 1 1 1 0.2>}
     uVRegion {bIsValid=false, max=<0 0>, min=<0 0>}
    }
    renderTransform {angle=0, scale=<1.2 1.2>, shear=<0 0>, translation=<-24 -4>}
   }
   slot {class=OverlaySlot}
  }
  BlueGlow {
   widgetClassPath=&Image
   bIsVariable=true
   parent=Glow_Box
   widget=BlueGlow
   properties {
    brush {
     bIsDynamicallyLoaded=false
     drawAs=Image
     imageType=FullColor
     mirroring=NoMirror
     resourceName=None
     resourceObject=/FMCalibration/Materials/MI_UI_GlowTechBlue
     tiling=NoTile
     imageSize=<400 400>
     margin=<ltrb 0 0 0 0>
     outlineSettings {
      bUseBrushTransparency=false
      roundingType=HalfHeightRadius
      width=0
      color {colorUseRule=UseColor_Specified, specifiedColor=<rgba 0 0 0 0>}
      cornerRadii=<0 0 0 0>
     }
     tintColor {colorUseRule=UseColor_Specified, specifiedColor=<rgba 1 1 1 0.2>}
     uVRegion {bIsValid=false, max=<0 0>, min=<0 0>}
    }
    renderTransform {angle=0, scale=<1 1>, shear=<0 0>, translation=<180 350>}
   }
   slot {class=OverlaySlot}
  }
  StageTech_Image {
   widgetClassPath=&Image
   bIsVariable=true
   parent=StageTech_ImageBox
   widget=StageTech_Image
   properties {
    visibility=SelfHitTestInvisible
    brush {
     bIsDynamicallyLoaded=false
     drawAs=Image
     imageType=FullColor
     mirroring=NoMirror
     resourceName=None
     resourceObject=/FMCalibration/Materials/MI_UI_StageTech_Guitar
     tiling=NoTile
     imageSize=<980 980>
     margin=<ltrb 0 0 0 0>
     outlineSettings {
      bUseBrushTransparency=false
      roundingType=HalfHeightRadius
      width=0
      color {colorUseRule=UseColor_Specified, specifiedColor=<rgba 0 0 0 0>}
      cornerRadii=<0 0 0 0>
     }
     tintColor {colorUseRule=UseColor_Specified, specifiedColor=<rgba 1 1 1 1>}
     uVRegion {bIsValid=false, max=<0 0>, min=<0 0>}
    }
   }
   slot {class=OverlaySlot}
  }
 ]
}
