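# Tune-Up complete screen, a calibration screen subclass that shows the measured results.
# On activation it picks body text by bus-calibration mode (AV offset or video latency, plus audio latency), calibrometer state (accuracy margins) and profile use, and adjusts padding on touch platforms.
# Done commits the values, closes calibration and broadcasts a modal-closed event through the playspace event router.
# Back returns, while a Pro Vocals tester button follows calibrometer-enabled changes.
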WidgetBlueprint /FMCalibration/Widgets/WBP_FMCalibration_Results
graphs [
 BP_GetDesiredFocusTarget {nodes [
   Get Desired Focus Target()
   > Return Node!(execute, ReturnValue:Widget Object Reference=Get Button_Done_Kit_1() -> (Button_Done_Kit_1:WBP UIKit Button Loud Object Reference))
  ]}
 EventGraph {nodes [
   n1 = "On Clicked (WBP_UIKit_Button_Quiet_ProVocals)"!() -> (OutputDelegate, then, Button:Common Button Base Object Reference) {is=title K2Node_ComponentBoundEvent}
   > Fire Event Element Interacted(Element=@n1.Button)
   > ShowProVocalsScreen(self:WBP FMCalibration Calibration Screen Object Reference)
   n2 = "On Clicked (WBP_UIKit_Button_Quiet_Back)"!() -> (OutputDelegate, then, Button:Common Button Base Object Reference) {is=title K2Node_ComponentBoundEvent}
   > Fire Event Element Interacted(Element=@n2.Button)
   > BackFromResultsScreen(self:WBP FMCalibration Calibration Screen Object Reference)
   n3 = "On Clicked (Button_Done_Kit_1)"!() -> (OutputDelegate, then, Button:Common Button Base Object Reference) {is=title K2Node_ComponentBoundEvent}
   > Fire Event Element Interacted(Element=@n3.Button)
   > CommitCalibrationValues(self:Calibration Process VM Object Reference=Get CalibrationProcessVM() -> (CalibrationProcessVM:Calibration Process VM Object Reference))
   > CloseCalibration!(
    execute
    self:WBP FMCalibration Calibration Screen Object Reference
    CalibrationProcessVM:Calibration Process VM Object Reference=Get CalibrationProcessVM()
   ) -> (then)
   > n4 = GetEventRouter!(execute, Scope:Object Class Reference=/Script/FortniteGame.FortPlayspace, ContextActor:Actor Object Reference=GetOwningPlayer()) -> (then, ReturnValue:Gameplay Event Router Component Object Reference)
   > Is Valid(InputObject=@n4)
   > Broadcast Event!(
    execute
    self:Gameplay Event Router Component Object Reference=@n4
    EventData:FMEvent Calibration Modal Closed Structure=Make FMEvent Calibration Modal Closed!() -> (FMEvent_Calibration_ModalClosed:FMEvent Calibration Modal Closed Structure) {is=title K2Node_MakeStruct}
    EventContext:Object Reference
   ) -> (then, ReturnValue:bool)
   n5 = HandleCalibrometerEnabled!() -> (OutputDelegate, then, Enabled:bool=~) {is=title K2Node_CustomEvent}
   UpdateCalibrometerDisplay(execute=[@n23.then, @n5.then])
   Event On Deactivated()
   > Parent: On Deactivated() {is=title K2Node_CallParentFunction}
   > Unbind Event from On Calibrometer Enabled Changed(
    self:FMCalibration Controller Component Object Reference=Get FMCalibrationComponent() -> (FMCalibrationComponent:BP FMCalibration Component Object Reference)
    Delegate=@n5
   ) {is=title K2Node_RemoveDelegate}
   Event PreConstruct()
   > SetToolTipText!(
    execute
    self:Widget Object Reference=Get WBP_UIKit_Button_Quiet_ProVocals() -> (WBP_UIKit_Button_Quiet_ProVocals:WBP UIKit Button Quiet Object Reference)
    InToolTipText:"Text (by ref)"=Format Text!(
     Format="LOCTEXT(\"Calibrometer {ProVocals} Tester\")"
     ProVocals:Text=MakeLiteralText(Value="LOCTABLE(\"/SparksSettings/UI/ST_Sparks_InstrumentNames\", \"ProVocals\")")
    ) -> (Result)
   ) -> (then)
   Event On Activated()
   > Parent: On Activated() {is=title K2Node_CallParentFunction}
   > n6 = Sequence!(execute) -> (then_0, then_1, then_2:Exec, then_3:Exec)
   > Fire Event Main View Opened!(execute, View:Widget Object Reference=Self-Reference(), AdditionalInfo:Flow Analytics Additional Info Structure, bIsOverlay:bool=false) -> (then)
   n7 = Branch(
    execute=@n6.then_1
    Condition=IsBusCalibrationEnabled!(self:Calibration Process VM Object Reference=Get CalibrationProcessVM()) -> (ReturnValue:bool)
   )
   n8 = Get CalibrationProcessVM()
   > n9 = Branch(Condition=IsCalibrometerEnabled!(self:Calibration Process VM Object Reference=@n8) -> (ReturnValue:bool))
   > Set Text(
    Text=Format Text!(
     Format="LOCTEXT(\"AV Offset: {avdelta}\\r\\nAudio Latency: {inputlatency} ± {audioaccuracy} ms\")"
     avdelta:float=GetIntermediateAVDelta!(self:Calibration Process VM Object Reference=@n8) -> (ReturnValue:float)
     inputlatency:float=GetIntermediateInputLatencyMs!(self:Calibration Process VM Object Reference=@n8) -> (ReturnValue:float)
     audioaccuracy:float=GetIntermediateAudioAccuracyMs!(self:Calibration Process VM Object Reference=@n8) -> (ReturnValue:float)
    ) -> (Result)
    self=Get Text_Body() -> (Text_Body:Common Text Object Reference)
   )
   n10 = Get CalibrationProcessVM()
   n11 = GetIntermediateAVDelta(self=@n10)
   n12 = GetIntermediateInputLatencyMs(self=@n10)
   n13 = int + int(A=Get ActiveProfileIndex!(self:Calibration Profile VM Object Reference=Get CalibrationProfileVM() -> (CalibrationProfileVM:Calibration Profile VM Object Reference)) -> (ActiveProfileIndex:int=0))
   n14 = Get CalibrationProfileVM()
   Set Text(
    execute=@n9.else
    Text=SelectText(
     A=Format Text!(
      Format="LOCTEXT(\"Saved to Tune-Up Profile {profile}\\r\\nAV Offset : {avdelta} milliseconds\\r\\nAudio Latency: {inputlatency} milliseconds\\r\\n\\r\\nIf the game still feels out of sync and you want to try again, select Tune-Up from the Game Menu. If your audio or video setup changes, select a different Tune-Up Profile in Settings.\\r\\n\")"
      profile:int=@n13
      avdelta=@n11
      inputlatency=@n12
     ) -> (Result)
     B=Format Text!(
      Format="LOCTEXT(\"AV Offset: {avdelta} milliseconds\\r\\nAudio Latency: {inputlatency} milliseconds\\r\\n\\r\\nIf the game feels out of sync, or if your audio/video setup changes, visit the Game Menu and tune-up again.\")"
      avdelta=@n11
      inputlatency=@n12
     ) -> (Result)
     bPickA=Get bHasUsedCalibrationProfileFeature!(self:Calibration Profile VM Object Reference=@n14) -> (bHasUsedCalibrationProfileFeature:bool)
    )
    self=Get Text_Body()
   )
   n15 = Get CalibrationProcessVM()
   n16 = Branch(execute=@n7.else, Condition=IsCalibrometerEnabled(self=@n15))
   > Set Text(
    Text=Format Text!(
     Format="LOCTEXT(\"Audio Latency: {audiolatency} ± {audioaccuracy} ms\\r\\nVideo Latency: {videolatency} ± {videoaccuracy} ms\")"
     audiolatency:float=GetIntermediateAudioLatencyMs!(self:Calibration Process VM Object Reference=@n15) -> (ReturnValue:float)
     audioaccuracy:float=GetIntermediateAudioAccuracyMs(self=@n15)
     videolatency:float=GetIntermediateVideoLatencyMs!(self:Calibration Process VM Object Reference=@n15) -> (ReturnValue:float)
     videoaccuracy:float=GetIntermediateVideoAccuracyMs!(self:Calibration Process VM Object Reference=@n15) -> (ReturnValue:float)
    ) -> (Result)
    self=Get Text_Body()
   )
   n17 = Get CalibrationProcessVM()
   n18 = GetIntermediateAudioLatencyMs(self=@n17)
   n19 = GetIntermediateVideoLatencyMs(self=@n17)
   n20 = int + int(A=Get ActiveProfileIndex(self=Get CalibrationProfileVM()) -> (ActiveProfileIndex=0))
   n21 = Get CalibrationProfileVM()
   Set Text(
    execute=@n16.else
    Text=SelectText(
     A=Format Text!(
      Format="LOCTEXT(\"Saved to Tune-Up Profile {profile}\\r\\nAudio Latency: {audiolatency} milliseconds\\r\\nVideo Latency: {videolatency} milliseconds\\r\\n\\r\\nIf the game still feels out of sync and you want to try again, select Tune-Up from the Game Menu. If your audio or video setup changes, select a different Tune-Up Profile in Settings.\\r\\n\")"
      profile:int=@n20
      audiolatency=@n18
      videolatency=@n19
     ) -> (Result)
     B=Format Text!(
      Format="LOCTEXT(\"Audio Latency: {audiolatency} milliseconds\\r\\nVideo Latency: {videolatency} milliseconds\\r\\n\\r\\nIf the game feels out of sync, or if your audio/video setup changes, visit the Game Menu and tune-up again.\")"
      audiolatency=@n18
      videolatency=@n19
     ) -> (Result)
     bPickA=Get bHasUsedCalibrationProfileFeature(self=@n21)
    )
    self=Get Text_Body()
   )
   n22 = IsPlatformTraitSupported!(
    execute=@n6.then_2
    TraitToCheck:"Gameplay Tag Structure (by ref)"=MakeLiteralGameplayTag!(Value:Gameplay Tag Structure=(TagName="Platform.Trait.Input.SupportsTouch")) -> (ReturnValue:Gameplay Tag Structure)
   ) -> (then, ReturnValue:bool)
   > Branch(Condition=@n22)
   > SetPadding!(
    execute
    self:Overlay Slot Object Reference=SlotAsOverlaySlot(Widget=Get BottomNav() -> (BottomNav:Horizontal Box Object Reference))
    InPadding_Left:float=16
    InPadding_Top:float=16
    InPadding_Right:float=48
    InPadding_Bottom:float=24
   ) -> (then)
   > SetPadding(
    self=SlotAsOverlaySlot(Widget=Get GridPanel_MainContainer() -> (GridPanel_MainContainer:Grid Panel Object Reference))
    InPadding_Left=0
    InPadding_Top=100
    InPadding_Right=0
    InPadding_Bottom=128
   )
   > SetPadding(
    self:Vertical Box Slot Object Reference=SlotAsVerticalBoxSlot(Widget=Get Text_Header() -> (Text_Header:Common Text Object Reference))
    InPadding_Left=0
    InPadding_Top=0
    InPadding_Right=0
    InPadding_Bottom=90
   )
   n23 = Bind Event to On Calibrometer Enabled Changed(execute=@n6.then_3, self:FMCalibration Controller Component Object Reference=Get FMCalibrationComponent(), Delegate=@n5) {is=title K2Node_AddDelegate}
  ]}
 UpdateCalibrometerDisplay {nodes [
   UpdateCalibrometerDisplay() {is=title K2Node_FunctionEntry}
   n1 = Get CalibrationProcessVM()
   > SetVisibility(
    self=Get WBP_UIKit_Button_Quiet_ProVocals()
    InVisibility=Select(Option 0=Collapsed, Option 1=Visible, Index=IsCalibrometerEnabled(self=@n1))
   )
  ]}
]
variables []
widgets {
 info {
  inheritedWidgetCount=0
  parentClass=&/WBP_FMCalibration_CalibrationScreen.WBP_FMCalibration_CalibrationScreen_C
  rootWidgetClass=&Overlay
  widgetCount=16
 }
 widgets [
  Overlay_11 {widgetClassPath=&Overlay, widget=Overlay_11}
  MainContainer {
   widgetClassPath=&Overlay
   parent=Overlay_11
   widget=MainContainer
   slot {class=OverlaySlot, properties {horizontalAlignment=HAlign_Fill, verticalAlignment=VAlign_Fill}}
  }
  GridPanel_MainContainer {
   widgetClassPath=&GridPanel
   bIsVariable=true
   parent=MainContainer
   widget=GridPanel_MainContainer
   properties {columnFill [1], rowFill [1]}
   slot {class=OverlaySlot, properties {horizontalAlignment=HAlign_Fill, verticalAlignment=VAlign_Fill, padding=<ltrb 0 100 0 100>}}
  }
  SlideContent_Padding {widgetClassPath=&VerticalBox, parent=GridPanel_MainContainer, widget=SlideContent_Padding, slot {class=GridSlot}}
  ModalContent_Padding {
   widgetClassPath=&VerticalBox
   parent=SlideContent_Padding
   widget=ModalContent_Padding
   slot {class=VerticalBoxSlot, properties {padding=<ltrb 0 0 0 80>}}
  }
  Text_Header {
   widgetClassPath=&CommonTextBlock
   bIsVariable=true
   parent=ModalContent_Padding
   widget=Text_Header
   properties {
    justification=Center
    text=TUNE-UP COMPLETE
    textTransformPolicy=ToUpper
    wrapTextAt=900
    font {
     bForceMonospaced=false
     bMaterialIsStencil=false
     fontMaterial=null
     fontObject=&BurbankSmall
     letterSpacing=0
     monospacedWidth=1
     size=25.5
     skewAmount=0
     typefaceFontName=Bold
     outlineSettings {
      bApplyOutlineToDropShadows=false
      bMiteredCorners=false
      bSeparateFillAlpha=false
      outlineMaterial=null
      outlineSize=0
      outlineColor=<rgba 0 0 0 1>
     }
    }
    shadowOffset=<0 0>
   }
   slot {class=VerticalBoxSlot, properties {horizontalAlignment=HAlign_Center, verticalAlignment=VAlign_Top, padding=<ltrb 0 0 0 40>}}
  }
  Text_Body {
   widgetClassPath=&CommonTextBlock
   bIsVariable=true
   parent=ModalContent_Padding
   widget=Text_Body
   properties {
    justification=Center
    text="Audio Latency: {audio} milliseconds\r\nVideo Latency: {video} milliseconds\r\n\r\nIf the game still feels out of sync and you want to try again, select Tune-Up from the Game Menu.\r\n"
    wrapTextAt=900
    colorAndOpacity {colorUseRule=UseColor_Specified, specifiedColor=<rgba 0.4397 0.7835 0.8879 1>}
    font {
     bForceMonospaced=false
     bMaterialIsStencil=false
     fontMaterial=null
     fontObject=&BurbankSmall
     letterSpacing=0
     monospacedWidth=1
     size=18
     skewAmount=0
     typefaceFontName=Medium
     outlineSettings {
      bApplyOutlineToDropShadows=false
      bMiteredCorners=false
      bSeparateFillAlpha=false
      outlineMaterial=null
      outlineSize=0
      outlineColor=<rgba 0 0 0 1>
     }
    }
    shadowOffset=<0 0>
   }
   slot {class=VerticalBoxSlot, properties {horizontalAlignment=HAlign_Center, verticalAlignment=VAlign_Top, padding=<ltrb 0 30 0 0>}}
  }
  SizeBox_ButtonGroup_Preamble {
   widgetClassPath=&SizeBox
   parent=SlideContent_Padding
   widget=SizeBox_ButtonGroup_Preamble
   properties {bOverride_WidthOverride=true, widthOverride=408}
   slot {
    class=VerticalBoxSlot
    properties {horizontalAlignment=HAlign_Center, verticalAlignment=VAlign_Bottom, padding=<ltrb 0 0 0 50>, size {sizeRule=Fill, value=1}}
   }
  }
  Button_Done_Kit_1 {
   widgetClassPath=/FortUILibrary/UIKit/Buttons/WBP_UIKit_Button_Loud
   bIsVariable=true
   parent=SizeBox_ButtonGroup_Preamble
   widget=Button_Done_Kit_1
   properties {
    buttonSize=Large
    minHeight=56
    text=Done
    triggeringInputAction {dataTable=&InputActionDataTable, rowName=GenericForward}
   }
   slot {class=SizeBoxSlot}
  }
  WBP_FMCalibration_StageTech_WithGuitar {
   widgetClassPath=&/WBP_FMCalibration_StageTech_WithGuitar
   parent=GridPanel_MainContainer
   widget=WBP_FMCalibration_StageTech_WithGuitar
   slot {class=GridSlot, properties {horizontalAlignment=HAlign_Left, verticalAlignment=VAlign_Bottom, nudge=<-180 0>, padding=<ltrb 120 0 0 -100>}}
  }
  WBP_UIKit_Button_Quiet_ProVocals {
   widgetClassPath=/FortUILibrary/UIKit/Buttons/WBP_UIKit_Button_Quiet
   bIsVariable=true
   parent=GridPanel_MainContainer
   widget=WBP_UIKit_Button_Quiet_ProVocals
   properties {
    bNavigateToNextWidgetOnDisable=true
    text=Karaoke
    visibility=Collapsed
    triggeringInputAction {dataTable=&InputActionDataTable, rowName=GenericFaceButtonTop}
   }
   slot {class=GridSlot, properties {horizontalAlignment=HAlign_Right, verticalAlignment=VAlign_Top, padding=<ltrb 0 0 16 0>}}
  }
  RootSafeZone {
   widgetClassPath=&SafeZone
   parent=Overlay_11
   widget=RootSafeZone
   properties {padLeft=false, padTop=false}
   slot {class=OverlaySlot, properties {horizontalAlignment=HAlign_Fill, verticalAlignment=VAlign_Fill}}
  }
  NavigationButtons {widgetClassPath=&Overlay, parent=RootSafeZone, widget=NavigationButtons, slot {class=SafeZoneSlot}}
  WBP_CaptureForPostBufferUpdate {
   widgetClassPath=&WBP_CaptureForPostBufferUpdate
   parent=NavigationButtons
   slot {class=OverlaySlot, properties {horizontalAlignment=HAlign_Center, verticalAlignment=VAlign_Bottom}}
  }
  BottomNav {
   widgetClassPath=&HorizontalBox
   bIsVariable=true
   parent=NavigationButtons
   widget=BottomNav
   slot {class=OverlaySlot, properties {horizontalAlignment=HAlign_Right, verticalAlignment=VAlign_Bottom, padding=<ltrb 16 16 16 16>}}
  }
  WBP_UIKit_Button_Quiet_Back {
   widgetClassPath=&WBP_UIKit_Button_Quiet
   bIsVariable=true
   parent=BottomNav
   widget=WBP_UIKit_Button_Quiet_Back
   properties {text=Back, triggeringInputAction {dataTable=&InputActionDataTable, rowName=GenericBack}}
   slot {class=HorizontalBoxSlot, properties {horizontalAlignment=HAlign_Left, verticalAlignment=VAlign_Top}}
  }
 ]
}
