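# Shared background wrapper widget for FMCalibration screen content, built on CommonUserWidget.
# A transparent border holds an overlay that layers a material-backed image, inset from the top, beneath a named slot for the hosting widget's content.
# On initialization it checks the touch-input platform trait and, when supported, adds top and bottom padding to the whole widget.
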WidgetBlueprint /FMCalibration/Widgets/WBP_FMCalibration_SharedContentBackground
graphs [EventGraph {nodes [
   Event OnInitialized()
   > Parent: OnInitialized() {is=title K2Node_CallParentFunction}
   > n1 = IsPlatformTraitSupported!(
    execute
    TraitToCheck:"Gameplay Tag Structure (by ref)"=MakeLiteralGameplayTag!(Value:Gameplay Tag Structure=(TagName="Platform.Trait.Input.SupportsTouch")) -> (ReturnValue:Gameplay Tag Structure)
   ) -> (then, ReturnValue:bool)
   > Branch(Condition=@n1)
   > SetPadding!(
    execute
    self:User Widget Object Reference
    InPadding_Left:float=0
    InPadding_Top:float=100
    InPadding_Right:float=0
    InPadding_Bottom:float=128
   ) -> (then)
  ]}]
variables []
widgets {
 info {inheritedWidgetCount=0, parentClass=&CommonUserWidget, rootWidgetClass=&Border, widgetCount=4}
 widgets [
  Content_Border {widgetClassPath=&Border, widget=Content_Border, properties {brushColor=<rgba 1 1 1 0>, padding=<ltrb 0 0 0 0>}}
  Overlay_0 {widgetClassPath=&Overlay, parent=Content_Border, widget=Overlay_0, slot {class=BorderSlot, properties {padding=<ltrb 0 0 0 0>}}}
  Underlayer {
   widgetClassPath=&Image
   bIsVariable=true
   parent=Overlay_0
   widget=Underlayer
   properties {brush {
     bIsDynamicallyLoaded=false
     drawAs=Image
     imageType=FullColor
     mirroring=NoMirror
     resourceName=None
     resourceObject=/FMCore/Materials/MusicLibraryBG_Inst_DublicateFMCore
     tiling=NoTile
     imageSize=<32 32>
     margin=<ltrb 0 0 0 0>
     outlineSettings {
      bUseBrushTransparency=false
      roundingType=HalfHeightRadius
      width=0
      color {colorUseRule=UseColor_Specified, specifiedColor=<rgba 0 0 0 0>}
      cornerRadii=<0 0 0 0>
     }
     tintColor {colorUseRule=UseColor_Specified, specifiedColor=<rgba 1 1 1 1>}
     uVRegion {bIsValid=false, max=<0 0>, min=<0 0>}
    }}
   slot {class=OverlaySlot, properties {horizontalAlignment=HAlign_Fill, verticalAlignment=VAlign_Fill, padding=<ltrb 0 90 0 0>}}
  }
  ContentBackgroundContent {
   widgetClassPath=&NamedSlot
   bIsVariable=true
   parent=Overlay_0
   widget=ContentBackgroundContent
   slot {class=OverlaySlot, properties {horizontalAlignment=HAlign_Fill, verticalAlignment=VAlign_Fill}}
  }
 ]
}
