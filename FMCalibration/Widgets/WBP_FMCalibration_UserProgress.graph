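# FM calibration progress widget, a CommonUserWidget whose root is a Pilgrim parent pip list showing the user's step in calibration.
# On construct it sizes the list's IndexMap from the TotalPips entry in the UI State Chart context and sets the active pip from PipIndex, both read per local player.
# It binds to context entry changes, updates the active pip whenever PipIndex changes, and unbinds on destruct.
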WidgetBlueprint /FMCalibration/Widgets/WBP_FMCalibration_UserProgress
graphs [
 EventGraph {nodes [
   n1 = On Any Entry Changed!() -> (OutputDelegate, then, Key:Gameplay Tag Structure, Entry:UIState Chart Context Entry Object Reference) {is=title K2Node_CustomEvent}
   > Branch(Condition="Equal (GameplayTag)"(A=@n1.Key, B=(TagName="UI.State.FMCalibration.PipIndex")))
   > SetActivePipByIndex!(
    execute
    self
    RawIndex:int=Get Player Value As Int!(
     self:UIState Chart Context Entry Object Reference=@n1.Entry
     InLocalPlayer:Local Player Object Reference=GetOwningLocalPlayer!(self:Widget Object Reference) -> (ReturnValue:Local Player Object Reference)
    ) -> (ReturnValue:int=0) {is=title K2Node_CallFunction}
   ) -> (then)
   Event Destruct()
   > Unbind Event from On Any Entry Changed(
    self:UIState Chart Context Object Reference=GetChartContext!(self:UIState Chart Manager Object Reference=Get UIStateChartManager() -> (ReturnValue:UIState Chart Manager Object Reference) {is=title K2Node_GetSubsystem}) -> (ReturnValue:UIState Chart Context Object Reference)
    Delegate=@n1
   ) {is=title K2Node_RemoveDelegate}
   Event Construct()
   n2 = GetChartContext(self=Get UIStateChartManager())
   > Bind Event to On Any Entry Changed(self:UIState Chart Context Object Reference=@n2, Delegate=@n1) {is=title K2Node_AddDelegate}
   n3 = GetOwningLocalPlayer()
   > Set Array Elem!(
    execute
    TargetArray:Array of Integers=Get IndexMap!(self:WBP Pilgrim Parent Pip List Object Reference=Get WBP_Pilgrim_Parent_Pip_List() -> (WBP_Pilgrim_Parent_Pip_List:WBP Pilgrim Parent Pip List Object Reference)) -> (IndexMap:Array of Integers)
    Index:int=0
    Item:"Integer (by ref)"=Get Player Value As Int(
     self=Find Entry!(self:UIState Chart Context Object Reference=@n2, Key:Gameplay Tag Structure=(TagName="UI.State.FMCalibration.TotalPips")) -> (ReturnValue:UIState Chart Context Entry Object Reference)
     InLocalPlayer=@n3
    ) -> (ReturnValue=0)
    bSizeToFit:bool=true
   ) -> (then)
   > SetActivePipByIndex(RawIndex=Get Player Value As Int(self=Find Entry(self=@n2, Key=(TagName="UI.State.FMCalibration.PipIndex")), InLocalPlayer=@n3) -> (ReturnValue=0))
  ]}
 SetActivePipByIndex {nodes [
   n1 = SetActivePipByIndex!() -> (then, RawIndex:int) {is=title K2Node_FunctionEntry}
   > SetCurrentIndex!(execute, self:WBP Pilgrim Parent Pip List Object Reference=Get WBP_Pilgrim_Parent_Pip_List(), RawIndex:int=@n1) -> (then)
  ]}
]
variables [FMCalibrationComponent {type=BP_FMCalibrationComponent_C, default=null}]
widgets {
 info {
  inheritedWidgetCount=0
  parentClass=&CommonUserWidget
  rootWidgetClass=/SparksUICommon/Widgets/WBP_Pilgrim_Parent_Pip_List.WBP_Pilgrim_Parent_Pip_List_C
  widgetCount=1
 }
 widgets [WBP_Pilgrim_Parent_Pip_List {
   widgetClassPath=/SparksUICommon/Widgets/WBP_Pilgrim_Parent_Pip_List
   bIsVariable=true
   widget=WBP_Pilgrim_Parent_Pip_List
   properties {indexMap [4]}
  }]
}
