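# FMCalibration progress bar widget, a CommonUserWidget wrapping a styled linear UIKit progress bar that animates increases in value.
# SetProgress stores the target, applying it immediately unless it exceeds the displayed value, in which case UpdateProgress starts a looping timer if none is active.
# Each tick steps the bar by a fixed increment, clamped to the target.
# The timer clears on reaching the target or on Destruct.
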WidgetBlueprint /FMCalibration/Widgets/WBP_FMCalibration_ProgressBar
animations [NewAnimation {binding_count=0, length_seconds=5, object=&self:NewAnimation, sequence {bindings []}}]
graphs [
 EventGraph {nodes [
   UpdateProgress() {is=title K2Node_CustomEvent}
   n1 = Get GrowthTimer() -> (GrowthTimer:Timer Handle Structure)
   > Branch(Condition=NOT Boolean(A=Is Timer Active by Handle!(Handle:Timer Handle Structure=@n1) -> (ReturnValue:bool)))
   > n2 = Set Timer by Event(Delegate=Create Event(), Time=Get GrowthUpdateFrequency() -> (GrowthUpdateFrequency:float), bLooping=true, bMaxOncePerFrame=true)
   > Set GrowthTimer(GrowthTimer:Timer Handle Structure=@n2) -> (Output_Get:Timer Handle Structure)
   UpdateFrequencyEvent() {is=title K2Node_CustomEvent}
   n3 = Get GrowthUpdateAmount() -> (GrowthUpdateAmount:double)
   n4 = Get ProgressValue!(self:WBP UIKit Progress Bar Base Object Reference=Get ProgressBar() -> (ProgressBar:WBP UIKit Progress Bar Object Reference)) -> (ProgressValue:double)
   n5 = Get DesiredProgress() -> (DesiredProgress:double)
   > n6 = Set NewProgress(NewProgress:double="Min (Float)"(A:double=float + float(A=@n4, B=@n3), B:double=@n5) -> (ReturnValue:double) {is=title K2Node_CommutativeAssociativeBinaryOperator}) -> (Output_Get:double)
   > SetProgress!(execute, self:WBP UIKit Progress Bar Base Object Reference=Get ProgressBar(), Progress:double=@n6) -> (then)
   > Branch(Condition="float >= float"(A=@n6, B=@n5))
   > Clear and Invalidate Timer by Handle(Handle=Get GrowthTimer())
   Event Destruct()
   > Clear and Invalidate Timer by Handle(Handle=Get GrowthTimer())
  ]}
 SetProgress {nodes [
   n1 = SetProgress!() -> (then, Progress:double=~) {is=title K2Node_FunctionEntry}
   > n2 = Set DesiredProgress(DesiredProgress:double=@n1) -> (Output_Get:double)
   n3 = Get ProgressBar()
   > n4 = Branch(Condition="float > float"(A=@n2, B=Get ProgressValue(self=@n3)))
   > UpdateProgress() {is=title K2Node_CallFunction}
   SetProgress(execute=@n4.else, self=Get ProgressBar(), Progress=Get DesiredProgress()) {is=title K2Node_CallFunction}
  ]}
]
variables [
 DesiredProgress {type=number}
 GrowthTimer {type=TimerHandle}
 GrowthUpdateAmount {type=number, default=0.015}
 GrowthUpdateFrequency {type=number, default=0.015}
 NewProgress {type=number}
]
widgets {
 info {
  inheritedWidgetCount=0
  parentClass=&CommonUserWidget
  rootWidgetClass=/FortUILibrary/UIKit/ProgressBar/WBP_UIKit_ProgressBar.WBP_UIKit_ProgressBar_C
  widgetCount=1
 }
 widgets [ProgressBar {
   widgetClassPath=/FortUILibrary/UIKit/ProgressBar/WBP_UIKit_ProgressBar
   bIsVariable=true
   widget=ProgressBar
   properties {
    height=22
    linearSDF_Sharpness=2.58
    outline_PX=0
    progressValue=1
    useCircular=false
    useOutline=true
    width=550
    colorA=<rgba 0.6098 1 0.4904 1>
    colorB=<rgba 0.3776 1 0.1706 0.513>
    colorC=<rgba 0.5114 0.8646 0.0153 1>
    emptyColor=<rgba 0.7305 1 0.6445 0.5>
   }
  }]
}
