WidgetBlueprint /FMCalibration/Widgets/WBP_FMCalibration_Video
graphs [
 BP_GetDesiredFocusTarget {nodes [
   Get Desired Focus Target()
   > Return Node!(execute, ReturnValue:Widget Object Reference=Get Button_Kit_StartCal() -> (Button_Kit_StartCal:WBP UIKit Button Loud Object Reference))
  ]}
 BP_GetDesiredInputConfig {nodes [
   Get Desired Input Config() {is=title K2Node_FunctionEntry}
   > Return Node(ReturnValue:UIInput Config Structure=Make UIInput Config!(
     bIgnoreMoveInput:bool=true
     bIgnoreLookInput:bool=true
     InputMode:ECommonInputMode Enum=All
     MouseCaptureMode:EMouseCaptureMode Enum=NoCapture
     MouseLockMode:EMouseLockMode Enum=DoNotLock
     bHideCursorDuringViewportCapture:bool=True
    ) -> (UIInputConfig:UIInput Config Structure) {is=title K2Node_MakeStruct})
  ]}
 CalibrateTick {nodes [
   CalibrateTick() {is=title K2Node_FunctionEntry}
   > n1 = Get!(execute:Exec) -> (then:Exec, else:Exec, CalibrationHelper:BP FMCalibration Helper Object Reference)
   > n2 = Get Video Fractional Beat!(execute, self:BP FMCalibration Helper Object Reference=@n1) -> (then, Fractional Beat:double, FirstFrameOfBeat:bool) {is=title K2Node_CallFunction}
   n3 = Get BallDynamicMaterial() -> (BallDynamicMaterial:Material Instance Dynamic Object Reference)
   > SetScalarParameterValue(self=@n3, ParameterName=Ball Animation Parameter, Value=@n2)
   n4 = GetDebugVideoMode!(self:Calibration Process VM Object Reference=Get CalibrationProcessVM() -> (CalibrationProcessVM:Calibration Process VM Object Reference)) -> (ReturnValue:int=0)
   > n5 = Branch(Condition="Equal (Integer)"(A=@n4, B=4))
   n6 = SelectedGemColor!(self) -> (Color:Linear Color Structure)
   n7 = BreakColor(InColor=@n6)
   n8 = Make Linear Color!(R:float=@n7, G:float=@n7.G, B:float=@n7.B, A:float=0) -> (LinearColor:Linear Color Structure) {is=title K2Node_MakeStruct}
   > SetVectorParameterValue(self=@n3, ParameterName=Gem1ColorSelected, Value=SelectColor(A=@n6, B=@n8, bPickA=@n2.FirstFrameOfBeat))
   Branch(execute=@n5.else, Condition="Equal (Integer)"(A=@n4, B=5))
   > SetVectorParameterValue(
    self=@n3
    ParameterName=Gem1ColorSelected
    Value="Lerp (LinearColor)"(A=@n6, B=@n8, Alpha=float * float(A=@n2, B=1) {is=title K2Node_PromotableOperator})
   )
  ]}
 DebugSetMaterial {nodes [
   DebugSetMaterial() {is=title K2Node_FunctionEntry}
   n1 = GetDebugVideoMode(self=Get CalibrationProcessVM()) -> (ReturnValue=0)
   > Branch(Condition="Not Equal (Integer)"(A=@n1, B=Get DebugLastVideoMode() -> (DebugLastVideoMode:int=0)))
   > Set DebugLastVideoMode(DebugLastVideoMode:int=@n1) -> (Output_Get:int=0)
   > n2 = Sequence!(execute) -> (then_0, then_1, then_2:Exec, then_3:Exec)
   n3 = Get PressCueAnimation() -> (PressCueAnimation:Image Object Reference)
   > SetBrushFromMaterial(self=@n3, Material=/FMCalibration/Materials/MI_UI_FTUE_SlidingNotes_Anim)
   n4 = GetDynamicMaterial(execute=@n2.then_1, self=@n3)
   > Set BallDynamicMaterial(BallDynamicMaterial:Material Instance Dynamic Object Reference=@n4) -> (Output_Get:Material Instance Dynamic Object Reference)
   n5 = GetDebugVideoMode(self=Get CalibrationProcessVM()) -> (ReturnValue=0)
   n6 = Get BallDynamicMaterial()
   SetScalarParameterValue(execute=@n2.then_2, self=@n6, ParameterName=CatcherSize, Value="To Float (Boolean)"(InBool="Equal (Integer)"(A=@n5, B=2)))
   SetVectorParameterValue(
    execute=@n2.then_3
    self=@n6
    ParameterName=Gem1ColorSelected
    Value=SelectColor(A=SelectedGemColor(), B=LinearColor_White!() -> (ReturnValue:Linear Color Structure), bPickA="integer < integer"(A=@n5, B=4))
   )
  ]}
 EventGraph {nodes [
   n1 = "On Pressed (Button_Kit_StartCal)"!() -> (OutputDelegate, then, Button:Common Button Base Object Reference) {is=title K2Node_ComponentBoundEvent}
   > Fire Event Element Interacted(Element=@n1.Button)
   n2 = Get ButtonHasRecivedInitialPress() -> (ButtonHasRecivedInitialPress:bool)
   > Branch(Condition=@n2)
   > n3 = Set bProcessingInputFromPress(bProcessingInputFromPress:bool=true) -> (Output_Get:bool)
   OnInputReceived!(
    execute=[@n9.then_1, @n3.then]
    self:Calibration Process VM Object Reference=Get CalibrationProcessVM()
    MusicClock:Music Clock Object Reference=Get CalibrationClock!(self:BP FMCalibration Helper Object Reference=Get CalibrationHelper() -> (CalibrationHelper:BP FMCalibration Helper Object Reference)) -> (CalibrationClock:Music Clock Object Reference)
   ) -> (then)
   n4 = "On Clicked (WBP_UIKit_Button_Quiet_Back)"!() -> (OutputDelegate, then, Button:Common Button Base Object Reference) {is=title K2Node_ComponentBoundEvent}
   > Fire Event Element Interacted(Element=@n4.Button)
   > PreviousCalibrationScreen(self:WBP FMCalibration Calibration Screen Object Reference)
   n5 = "On Clicked (Button_Kit_StartCal)"!() -> (OutputDelegate, then, Button:Common Button Base Object Reference) {is=title K2Node_ComponentBoundEvent}
   > n6 = Fire Event Element Interacted(Element=@n5.Button)
   n7 = Branch(execute=[@n20.then, @n6.then], Condition=@n2)
   > n8 = Branch(Condition=Get bProcessingInputFromPress() -> (bProcessingInputFromPress:bool))
   > Set bProcessingInputFromPress(bProcessingInputFromPress=false)
   n9 = Sequence!(execute=@n8.else) -> (then_0, then_1)
   n10 = Get Button_Kit_StartCal()
   n11 = Get Anim_Pressed_Base!(self:WBP UIKit Button CTA Base Object Reference=@n10) -> (Anim_Pressed_Base:Widget Animation Object Reference)
   > Branch(Condition="Equal (Float)"(A=GetAnimationCurrentTime!(self:User Widget Object Reference=@n10, InAnimation:Widget Animation Object Reference=@n11) -> (ReturnValue:float)))
   > PlayAnimation(self=@n10, InAnimation=@n11, PlayMode=PingPong)
   n12 = Get CalibrationProcessVM()
   n13 = Get CalibrationHelper()
   Start Metronome!(
    execute=@n7.else
    self:BP FMCalibration Helper Object Reference=@n13
    Tempo:double=GetBpm!(self:Calibration Process VM Object Reference=@n12) -> (ReturnValue:float)
    Silent:bool=true
    UseHiHats:bool=false
    Use Calibrometer:bool=IsCalibrometerEnabled!(self:Calibration Process VM Object Reference=@n12) -> (ReturnValue:bool)
    ProVocals:bool=false
   ) -> (then)
   > Set MusicClock!(
    execute
    MusicClock:Music Clock Object Reference=Get CalibrationClock(self=@n13)
    self:WBP Calibrometer Widget Object Reference=Get WBP_CalibrometerDebug() -> (WBP_CalibrometerDebug:WBP Calibrometer Widget Object Reference)
   ) -> (then, Output_Get:Music Clock Object Reference)
   > n14 = Sequence()
   > n15 = Set Timer by Function Name!(
    execute
    Object:Object Reference
    FunctionName:String=CalibrateTick
    Time:float=0
    bLooping:bool=true
    bMaxOncePerFrame:bool=true
    InitialStartDelay:float=0
    InitialStartDelayVariance:float=0
   ) -> (then, ReturnValue:Timer Handle Structure) {is=title K2Node_CallFunction}
   > Set AnimationTimer(AnimationTimer:Timer Handle Structure=@n15) -> (Output_Get:Timer Handle Structure)
   > Set PressedSound!(execute, PressedSound:Sound Base Object Reference, self:WBP UIKit Button CTA Base Object Reference=Get Button_Kit_StartCal()) -> (then, Output_Get:Sound Base Object Reference)
   > ShowProgressBar()
   > Set ButtonHasRecivedInitialPress(ButtonHasRecivedInitialPress:bool=true) -> (Output_Get:bool)
   > UpdateButtonText()
   > SetVisibility(self=Get Text_SubBody() -> (Text_SubBody:Common Text Object Reference), InVisibility=Hidden)
   n16 = Get!(execute=@n14.then_1) -> (then, else, MidiDeviceHandler:Harmonix Midi Device Event Handler Object Reference)
   > SetMusicClock!(
    execute
    self:Harmonix Midi Device Event Handler Object Reference=@n16
    MusicClock:Music Clock Object Reference=Get CalibrationClock(self=Get CalibrationHelper())
   ) -> (then)
   n17 = "On Clicked (WBP_UIKit_Button_Quiet_Skip)"!() -> (OutputDelegate, then, Button:Common Button Base Object Reference) {is=title K2Node_ComponentBoundEvent}
   > Fire Event Element Interacted(Element=@n17.Button)
   > StageSkipped(self:Calibration Process VM Object Reference=Get CalibrationProcessVM())
   > GotoNextScreen()
   n18 = OnMidiInputMidiNote!() -> (OutputDelegate, then, bIsNoteOn:bool=~, NoteNumber:Byte, Velocity:Byte) {is=title K2Node_CustomEvent}
   > n19 = Branch(Condition=@n18.bIsNoteOn)
   n20 = FireEventElementInteractedByName!(execute=[@n38.Started, @n19.then], Element Name:String=Video_IA_FMCalibrateSwing, WidgetCallingFrom:Widget Object Reference=Self-Reference()) -> (then, Success:bool)
   GotoNextScreen() {is=title K2Node_CustomEvent}
   > OpenResultsSceen(self:WBP FMCalibration Calibration Screen Object Reference)
   ResetScreen() {is=title K2Node_CustomEvent}
   > SetText!(
    execute
    self:Fort CTAButton Object Reference=Get Button_Kit_StartCal()
    ButtonText:"Text (by ref)"=Get ButtonBeginText() -> (ButtonBeginText:Text)
   ) -> (then)
   > Set PressedSound(PressedSound=Get Default Style Pressed Sound() -> (Default Style Pressed Sound:Sound Base Object Reference), self=Get Button_Kit_StartCal())
   > HideProgressBar()
   > StopAnimationTimer()
   > SetVisibility(self=Get Button_Kit_StartCal(), InVisibility=Visible)
   > SetVisibility(
    self=Get WBP_UIKit_Button_Quiet_Skip() -> (WBP_UIKit_Button_Quiet_Skip:WBP UIKit Button Quiet Object Reference)
    InVisibility=SelfHitTestInvisible
   )
   > Set ButtonHasRecivedInitialPress(ButtonHasRecivedInitialPress=false)
   > SetPressMethod!(execute, self:Common Button Base Object Reference=Get Button_Kit_StartCal(), InPressMethod:EButtonPressMethod Enum=ButtonRelease) -> (then)
   > Update Visuals For Touch Input()
   HideProgressBar() {is=title K2Node_CustomEvent}
   > SetVisibility(
    self=Get WBP_FMCalibration_ProgressBar_C_0() -> (WBP_FMCalibration_ProgressBar_C_0:WBP FMCalibration Progress Bar Object Reference)
    InVisibility=Hidden
   )
   ShowProgressBar() {is=title K2Node_CustomEvent}
   > SetVisibility(self=Get WBP_FMCalibration_ProgressBar_C_0(), InVisibility=SelfHitTestInvisible)
   n21 = RefreshInstructionalText!() -> (OutputDelegate, then, UserId:Platform User Id Structure, DeviceId:Input Device Id Structure) {is=title K2Node_CustomEvent}
   n22 = GetOwningPlayer()
   > Branch(Condition="Equal (PlatformUserId)"(
     A:Platform User Id Structure=@n21.UserId
     B:Platform User Id Structure=GetPlatformUserId!(self:Player Controller Object Reference=@n22) -> (ReturnValue:Platform User Id Structure)
    ) -> (ReturnValue:bool=~) {is=title K2Node_PromotableOperator})
   n23 = GetInputDeviceHardwareIdentifier!(
    self:Input Device Subsystem Object Reference=Get InputDeviceSubsystem() -> (ReturnValue:Input Device Subsystem Object Reference) {is=title K2Node_GetEngineSubsystem}
    InputDevice:Input Device Id Structure=@n21.DeviceId
   ) -> (ReturnValue:Hardware Device Identifier Structure)
   > Set DeviceInstrumentType(DeviceInstrumentType:EFMPeripheralInputMode Enum=GetDevicePeripheralType!(DeviceId:Hardware Device Identifier Structure=@n23) -> (ReturnValue:EFMPeripheralInputMode Enum=None)) -> (Output_Get:EFMPeripheralInputMode Enum=None)
   > n24 = Branch(Condition=NOT Boolean(A=Get IsUsingTouchInput() -> (IsUsingTouchInput:bool)))
   n25 = Format Text(Format="LOCTEXT(\"Tap when each note is centered in the hit zone outline.\")")
   n26 = Format Text(Format="LOCTEXT(\"Press the button when each note is centered in the hit zone outline.\")")
   n27 = Format Text(Format="LOCTEXT(\"Strum when each note is centered in the hit zone outline.\")")
   n28 = Format Text(Format="LOCTEXT(\"Hit the green pad when each note is centered in the hit zone outline.\")")
   n29 = Get DeviceInstrumentType() -> (DeviceInstrumentType:EFMPeripheralInputMode Enum=None)
   > n30 = "SetText (Text)"(
    self=Get Text_Body() -> (Text_Body:Common Text Object Reference)
    InText=Select!(
     Unspecified:Text=@n26
     KeyboardAndMouse:Text=@n26
     Gamepad:Text=@n26
     Touch:Text=@n25
     MotionTracking:Text
     RacingWheel:Text
     FlightStick:Text
     Camera:Text
     Instrument:Text=Select!(None:Text=@n26, Guitar:Text=@n27, Drums:Text=@n28, Index:EFMPeripheralInputMode Enum=@n29) -> (ReturnValue)
     CustomTypeA:Text
     CustomTypeB:Text
     CustomTypeC:Text
     CustomTypeD:Text
     Index:EHardwareDevicePrimaryType Enum=Break Hardware Device Identifier!(HardwareDeviceIdentifier:"Hardware Device Identifier Structure (by ref)"=@n23) -> (InputClassName:Name, HardwareDeviceIdentifier:Name, PrimaryDeviceType:EHardwareDevicePrimaryType Enum, SupportedFeaturesMask:Bitmask Integer) {is=title K2Node_BreakStruct}.PrimaryDeviceType
    ) -> (ReturnValue:Text)
   )
   UpdateButtonText(execute=[@n31.then, @n30.then])
   n31 = "SetText (Text)"(execute=@n24.else, self=Get Text_Body(), InText=@n25)
   n32 = OnStageComplete!() -> (OutputDelegate, then, Result:EFMCalibrationProcessStageResult Enum=ButtonMash) {is=title K2Node_CustomEvent}
   > n33 = Switch on EFMCalibrationProcessStageResult!(execute, Selection:EFMCalibrationProcessStageResult Enum=@n32.Result) -> (ButtonMash:Exec, TooInaccurate:Exec, Skipped:Exec, Success:Exec, InProgress:Exec) {is=title K2Node_SwitchEnum}
   ShowErrorScreen(execute=[@n33.ButtonMash, @n33.TooInaccurate], self:WBP FMCalibration Calibration Screen Object Reference)
   n34 = Sequence(execute=@n33.Success)
   ShowComplete(
    execute=@n34.then_1
    self:WBP FMCalibration Complete Display Object Reference=Get WBP_FMCalibration_CompleteDisplay() -> (WBP_FMCalibration_CompleteDisplay:WBP FMCalibration Complete Display Object Reference)
   )
   > StopAnimationTimer()
   > Stop Metronome(self:BP FMCalibration Helper Object Reference=Get CalibrationHelper())
   > SetVisibility(
    self=Get Button_Kit_StartCal()
    InVisibility=Select!(Option 0:ESlateVisibility Enum=Hidden, Option 1:ESlateVisibility Enum=Collapsed, Index:bool=Get IsUsingTouchInput()) -> (ReturnValue:ESlateVisibility Enum)
   )
   > Branch(Condition=Get IsUsingTouchInput())
   > SetAllInputZonesActive!(
    execute
    self:Polygonal Input Zone Controller Base Object Reference=Get WBP_PolygonalInputZoneController_CalibrateSwing() -> (WBP_PolygonalInputZoneController_CalibrateSwing:WBP Polygonal Input Zone Controller Object Reference)
    Activated:bool=false
   ) -> (then)
   n35 = Get CalibrationProcessVM()
   n36 = Branch(execute=@n34.then_0, Condition=IsCalibrometerEnabled(self=@n35))
   > "SetText (Text)"(
    self=Get Text_Body()
    InText=Format Text!(
     Format="LOCTEXT(\"Video Latency: {vid} ± {vid_accuracy} ms\\r\\nTargetFPS: {TFPS}\\r\\nMeasuredFPS: {MFPS}\")"
     vid:float=GetIntermediateVideoLatencyMs!(self:Calibration Process VM Object Reference=@n35) -> (ReturnValue:float)
     vid_accuracy:float=GetIntermediateVideoAccuracyMs!(self:Calibration Process VM Object Reference=@n35) -> (ReturnValue:float)
     TFPS:int=GetTargetFPS!(self:Calibration Process VM Object Reference=@n35) -> (ReturnValue:int=0)
     MFPS:int=GetMeasuredFPS!(self:Calibration Process VM Object Reference=@n35) -> (ReturnValue:int=0)
    ) -> (Result)
   )
   n37 = Get WBP_UIKit_Button_Quiet_Skip()
   > SetVisibility(self=@n37, InVisibility=Visible)
   > Set Text(Text="LOCTEXT(\"Next\")", self:Fort CTAButton Object Reference=@n37)
   > Unbind Event from On Input Hardware Device Changed(self:Input Device Subsystem Object Reference=Get InputDeviceSubsystem(), Delegate=Create Event()) {is=title K2Node_RemoveDelegate}
   SetVisibility(execute=@n36.else, self=@n37, InVisibility=Collapsed)
   > Bind Event to Show Complete Anim Finished(self:WBP FMCalibration Complete Display Object Reference=Get WBP_FMCalibration_CompleteDisplay(), Delegate=Create Event()) {is=title K2Node_AddDelegate}
   n38 = EnhancedInputAction IA_FMCalibrateSwing!() -> (
    Triggered:Exec
    Started:Exec
    Ongoing:Exec
    Canceled:Exec
    Completed:Exec
    ActionValue:bool
    ElapsedSeconds:double=~
    TriggeredSeconds:double=~
    InputAction:Input Action Object Reference=/FMCalibration/Data/IA_FMCalibrateSwing
   ) {is=title K2Node_EnhancedInputAction}
   Event On Deactivated()
   > Parent: On Deactivated() {is=title K2Node_CallParentFunction}
   > n39 = Sequence!(execute) -> (then_0, then_1, then_2:Exec)
   > Unbind Event from On Stage Completion(self:Calibration Process VM Object Reference=Get CalibrationProcessVM(), Delegate=@n32) {is=title K2Node_RemoveDelegate}
   Stop Metronome(execute=@n39.then_1, self=Get CalibrationHelper())
   > StopAnimationTimer()
   > Unbind Event from Show Complete Anim Finished(self:WBP FMCalibration Complete Display Object Reference=Get WBP_FMCalibration_CompleteDisplay(), Delegate=Create Event()) {is=title K2Node_RemoveDelegate}
   > Unbind Event from On Input Hardware Device Changed(self=Get InputDeviceSubsystem(), Delegate=Create Event())
   > Clear and Invalidate Timer by Handle(Handle=Get DebugVideoModeTimer() -> (DebugVideoModeTimer:Timer Handle Structure))
   > Branch(Condition=Get IsUsingTouchInput())
   > SetAllInputZonesActive(self=Get WBP_PolygonalInputZoneController_CalibrateSwing(), Activated=false)
   n40 = Get MidiDeviceHandler() -> (MidiDeviceHandler:Harmonix Midi Device Event Handler Object Reference)
   Is Valid(exec=@n39.then_2, InputObject=@n40)
   > StopListening(self:Harmonix Midi Device Event Handler Object Reference=@n40)
   > Set MidiDeviceHandler(MidiDeviceHandler:Harmonix Midi Device Event Handler Object Reference) -> (Output_Get:Harmonix Midi Device Event Handler Object Reference)
   Event On Activated()
   > Parent: On Activated() {is=title K2Node_CallParentFunction}
   > n41 = Sequence!(execute) -> (then_0, then_1, then_2, then_3:Exec)
   > Fire Event Main View Opened!(execute, View:Widget Object Reference=Self-Reference(), AdditionalInfo:Flow Analytics Additional Info Structure, bIsOverlay:bool=false) -> (then)
   > n42 = IsPlatformTraitSupported!(
    execute
    TraitToCheck:"Gameplay Tag Structure (by ref)"=MakeLiteralGameplayTag!(Value:Gameplay Tag Structure=(TagName="Platform.Trait.Input.SupportsTouch")) -> (ReturnValue:Gameplay Tag Structure)
   ) -> (then, ReturnValue:bool)
   > Branch(Condition=@n42)
   > SetPadding!(
    execute
    self:Overlay Slot Object Reference=SlotAsOverlaySlot(Widget=Get SafeZone() -> (SafeZone:Safe Zone Object Reference))
    InPadding_Left:float=16
    InPadding_Top:float=16
    InPadding_Right:float=48
    InPadding_Bottom:float=24
   ) -> (then)
   > SetPadding(
    self=SlotAsOverlaySlot(Widget=Get Overlay_MainContainer() -> (Overlay_MainContainer:Overlay Object Reference))
    InPadding_Left=0
    InPadding_Top=100
    InPadding_Right=0
    InPadding_Bottom=128
   )
   Set CalibrationHelper(
    execute=@n41.then_1
    CalibrationHelper:BP FMCalibration Helper Object Reference=GetCalibrationHelper!(self:BP FMCalibration Component Object Reference=Get FMCalibrationComponent() -> (FMCalibrationComponent:BP FMCalibration Component Object Reference)) -> (Calibration Helper:BP FMCalibration Helper Object Reference)
   ) -> (Output_Get:BP FMCalibration Helper Object Reference)
   > n43 = Get!(execute, self:WBP UIKit Button CTA Base Object Reference=Get Button_Kit_StartCal()) -> (then, else, PressedSound:Sound Base Object Reference)
   > n44 = Set Default Style Pressed Sound(Default Style Pressed Sound:Sound Base Object Reference=@n43) -> (Output_Get:Sound Base Object Reference)
   ResetScreen(execute=[@n44.then, @n43.else]) {is=title K2Node_CallFunction}
   > StartVideoCalibration(self:Calibration Process VM Object Reference=Get CalibrationProcessVM())
   n45 = Get InputDeviceSubsystem()
   > Bind Event to On Input Hardware Device Changed(self:Input Device Subsystem Object Reference=@n45, Delegate=Create Event()) {is=title K2Node_AddDelegate}
   n46 = GetPlatformUserId(self=GetOwningPlayer())
   > RefreshInstructionalText!(
    execute
    self
    UserId:Platform User Id Structure=@n46
    DeviceId:Input Device Id Structure=GetMostRecentlyUsedInputDeviceId!(
     self:Input Device Subsystem Object Reference=@n45
     InUserId:Platform User Id Structure=@n46
     OfType:EHardwareDevicePrimaryType Enum=Unspecified
    ) -> (ReturnValue:Input Device Id Structure)
   ) -> (then) {is=title K2Node_CallFunction}
   > DebugSetMaterial() {is=title K2Node_CallFunction}
   > n47 = Set Timer by Function Name(FunctionName=DebugSetMaterial, Time=0, bLooping=true, bMaxOncePerFrame=true, InitialStartDelay=0, InitialStartDelayVariance=0)
   > Set DebugVideoModeTimer(DebugVideoModeTimer:Timer Handle Structure=@n47) -> (Output_Get:Timer Handle Structure)
   > n48 = GetDynamicMaterial(self=Get PressCueAnimation())
   > Set BallDynamicMaterial(BallDynamicMaterial=@n48)
   n49 = CreateMidiDeviceEventHandler!(execute=@n41.then_2) -> (then, ReturnValue:Harmonix Midi Device Event Handler Object Reference)
   > n50 = Set MidiDeviceHandler(MidiDeviceHandler=@n49)
   > Begin Listening!(execute, self:Harmonix Midi Device Event Handler Object Reference=@n50, InOnMidiNoteEvent:Delegate=Create Event(), DeviceId:int=-1) -> (then)
   Bind Event to On Stage Completion(execute=@n41.then_3, self:Calibration Process VM Object Reference=Get CalibrationProcessVM(), Delegate=@n32) {is=title K2Node_AddDelegate}
  ]}
 SelectedGemColor {nodes [
   SelectedGemColor() {is=title K2Node_FunctionEntry}
   > Return Node!(execute, Color:Linear Color Structure=Make Linear Color(R=0.4969, G=0.8714, B=0.0296, A=1))
   LinearColor_Blue!() -> (ReturnValue:Linear Color Structure)
  ]}
 StopAnimationTimer {nodes [
   StopAnimationTimer() {is=title K2Node_FunctionEntry}
   > SetScalarParameterValue(self=Get BallDynamicMaterial(), ParameterName=Ball Animation Parameter, Value=-2)
   > Clear and Invalidate Timer by Handle(Handle=Get AnimationTimer() -> (AnimationTimer:Timer Handle Structure))
  ]}
 Update Visuals For Touch Input {nodes [
   Update Visuals For Touch Input() {is=title K2Node_FunctionEntry}
   n1 = GetOwningPlayer()
   > n2 = Set IsUsingTouchInput(IsUsingTouchInput:bool=IsUsingInputType!(
     InputType:ECommonInputType Enum=Touch
     PlayerController:Player Controller Object Reference=GetFortPlayerControllerFromActor!(Actor:Actor Object Reference=@n1) -> (ReturnValue:Fort Player Controller Object Reference)
    ) -> (ReturnValue:bool)) -> (Output_Get:bool)
   > SetVisibility(self=Get WBP_PolygonalInputZoneController_CalibrateSwing(), InVisibility=Select(Option 0=Collapsed, Option 1=Visible, Index=@n2))
   > SetVisibility(self=Get Button_Kit_StartCal(), InVisibility=Select(Option 0=SelfHitTestInvisible, Option 1=Collapsed, Index=@n2))
   > SetPadding(
    self:Vertical Box Slot Object Reference=SlotAsVerticalBoxSlot(Widget=Get Text_Header() -> (Text_Header:Common Text Object Reference))
    InPadding_Left=0
    InPadding_Top=0
    InPadding_Right=0
    InPadding_Bottom=Select(Option 0:float=40, Option 1:float=90, Index=@n2) -> (ReturnValue:float)
   )
   > Branch(Condition=@n2)
   > SetAllInputZonesActive(self=Get WBP_PolygonalInputZoneController_CalibrateSwing(), Activated=true)
  ]}
 UpdateButtonText {nodes [
   UpdateButtonText() {is=title K2Node_FunctionEntry}
   n1 = Get DeviceInstrumentType() -> (DeviceInstrumentType=None)
   > SetTriggeringInputAction!(
    execute
    self:Common Button Base Object Reference=Get Button_Kit_StartCal()
    InputActionRow:"Data Table Row Handle Structure (by ref)"=Make Data Table Row Handle!(
     DataTable:Data Table Object Reference=/FMCalibration/Data/FMCalibration_InputActionDataTable
     RowName:Name=Select!(None:Name=FMCalibrateSwing, Guitar:Name=FMCalibrateSwing_Guitar, Drums:Name=FMCalibrateSwing, Index:EFMPeripheralInputMode Enum=@n1) -> (ReturnValue:Name)
    ) -> (DataTableRowHandle:Data Table Row Handle Structure) {is=title K2Node_MakeStruct}
   ) -> (then)
   > Branch(Condition=Get ButtonHasRecivedInitialPress())
   > SetText(
    self=Get Button_Kit_StartCal()
    ButtonText=Select(
     None:Text=Get ButtonActionText() -> (ButtonActionText:Text)
     Guitar:Text=Get ButtonActionText_Instrument() -> (ButtonActionText_Instrument:Text)
     Drums:Text=Get ButtonActionText()
     Index=Get DeviceInstrumentType() -> (DeviceInstrumentType=None)
    ) -> (ReturnValue:Text)
   )
  ]}
]
variables [
 AnimationTimer {type=TimerHandle}
 BallDynamicMaterial {type=MaterialInstanceDynamic}
 bProcessingInputFromPress {type=boolean}
 ButtonActionText {type=string, default=Tap}
 ButtonActionText_Instrument {type=string, default=Strum}
 ButtonBeginText {type=string, default=Begin}
 ButtonHasRecivedInitialPress {type=boolean}
 CalibrationHelper {type=BP_FMCalibration_Helper_C, default=null}
 DebugLastVideoMode {type=integer, default=-1}
 DebugVideoModeTimer {type=TimerHandle}
 Default Style Pressed Sound {type=SoundBase, default=null}
 DefaultButtonWidth {type=number}
 DeviceInstrumentType {type=EFMPeripheralInputMode, default=None}
 IsUsingTouchInput {type=boolean}
 MidiDeviceHandler {type=HarmonixMidiDeviceEventHandler, default=null}
 PortraitButtonHeight {type=number}
 PortraitButtonWidth {type=number}
]
widgets {
 info {
  inheritedWidgetCount=0
  parentClass=&/WBP_FMCalibration_CalibrationScreen.WBP_FMCalibration_CalibrationScreen_C
  rootWidgetClass=&Overlay
  widgetCount=28
 }
 widgets [
  Overlay_1 {widgetClassPath=&Overlay, widget=Overlay_1, properties {bOverride_Cursor=true}}
  NavigationContainer {
   widgetClassPath=&Overlay
   parent=Overlay_1
   widget=NavigationContainer
   slot {class=OverlaySlot, properties {horizontalAlignment=HAlign_Fill, verticalAlignment=VAlign_Fill}}
  }
  SafeZone {
   widgetClassPath=&SafeZone
   bIsVariable=true
   parent=NavigationContainer
   widget=SafeZone
   properties {padLeft=false, padTop=false}
   slot {class=OverlaySlot, properties {horizontalAlignment=HAlign_Right, verticalAlignment=VAlign_Bottom, padding=<ltrb 16 16 16 16>}}
  }
  Overlay_0 {widgetClassPath=&Overlay, parent=SafeZone, widget=Overlay_0, slot {class=SafeZoneSlot}}
  WBP_CaptureForPostBufferUpdate {
   widgetClassPath=&WBP_CaptureForPostBufferUpdate
   bIsVariable=true
   parent=Overlay_0
   slot {class=OverlaySlot, properties {horizontalAlignment=HAlign_Fill, verticalAlignment=VAlign_Fill}}
  }
  BottomNav {widgetClassPath=&HorizontalBox, parent=Overlay_0, widget=BottomNav, slot {class=OverlaySlot}}
  WBP_UIKit_Button_Quiet_Skip {
   widgetClassPath=/FortUILibrary/UIKit/Buttons/WBP_UIKit_Button_Quiet
   bIsVariable=true
   parent=BottomNav
   widget=WBP_UIKit_Button_Quiet_Skip
   properties {
    bHideInputActionWithKeyboard=true
    holdData=null
    text=Skip Video
    triggeringInputAction {dataTable=&FMCalibration_InputActionDataTable, rowName=SkipStep}
   }
   slot {class=HorizontalBoxSlot, properties {horizontalAlignment=HAlign_Left, verticalAlignment=VAlign_Top, padding=<ltrb 0 0 16 0>}}
  }
  WBP_UIKit_Button_Quiet_Back {
   widgetClassPath=&WBP_UIKit_Button_Quiet
   bIsVariable=true
   parent=BottomNav
   widget=WBP_UIKit_Button_Quiet_Back
   properties {text=Back, triggeringInputAction {dataTable=&InputActionDataTable, rowName=GenericBack}}
   slot {class=HorizontalBoxSlot, properties {horizontalAlignment=HAlign_Left, verticalAlignment=VAlign_Top}}
  }
  Overlay_MainContainer {
   widgetClassPath=&Overlay
   bIsVariable=true
   parent=Overlay_1
   widget=Overlay_MainContainer
   slot {class=OverlaySlot, properties {horizontalAlignment=HAlign_Fill, verticalAlignment=VAlign_Fill, padding=<ltrb 0 100 0 100>}}
  }
  MainContainer {
   widgetClassPath=&VerticalBox
   parent=Overlay_MainContainer
   widget=MainContainer
   properties {bOverride_Cursor=true}
   slot {class=OverlaySlot, properties {horizontalAlignment=HAlign_Fill, verticalAlignment=VAlign_Fill}}
  }
  SlideContent_Padding {
   widgetClassPath=&VerticalBox
   parent=MainContainer
   widget=SlideContent_Padding
   slot {class=VerticalBoxSlot, properties {size {sizeRule=Fill, value=0.15}}}
  }
  ModalContent_Padding {
   widgetClassPath=&VerticalBox
   parent=SlideContent_Padding
   widget=ModalContent_Padding
   slot {class=VerticalBoxSlot, properties {size {sizeRule=Fill, value=1}}}
  }
  Text_Header {
   widgetClassPath=&CommonTextBlock
   bIsVariable=true
   parent=ModalContent_Padding
   widget=Text_Header
   properties {
    justification=Center
    text=Sync Your Video
    textTransformPolicy=ToUpper
    wrapTextAt=900
    font {
     bForceMonospaced=false
     bMaterialIsStencil=false
     fontMaterial=null
     fontObject=&BurbankSmall
     letterSpacing=0
     monospacedWidth=1
     size=25.5
     skewAmount=0
     typefaceFontName=Bold
     outlineSettings {
      bApplyOutlineToDropShadows=false
      bMiteredCorners=false
      bSeparateFillAlpha=false
      outlineMaterial=null
      outlineSize=0
      outlineColor=<rgba 0 0 0 1>
     }
    }
    shadowOffset=<0 0>
   }
   slot {class=VerticalBoxSlot, properties {horizontalAlignment=HAlign_Center, verticalAlignment=VAlign_Top, padding=<ltrb 0 0 0 40>}}
  }
  Text_Body {
   widgetClassPath=&CommonTextBlock
   bIsVariable=true
   parent=ModalContent_Padding
   widget=Text_Body
   properties {
    justification=Center
    text="Press the button when each note is  centered in the hit zone outline."
    wrapTextAt=900
    colorAndOpacity {colorUseRule=UseColor_Specified, specifiedColor=<rgba 0.4397 0.7835 0.8879 1>}
    font {
     bForceMonospaced=false
     bMaterialIsStencil=false
     fontMaterial=null
     fontObject=&BurbankSmall
     letterSpacing=0
     monospacedWidth=1
     size=18
     skewAmount=0
     typefaceFontName=Medium
     outlineSettings {
      bApplyOutlineToDropShadows=false
      bMiteredCorners=false
      bSeparateFillAlpha=false
      outlineMaterial=null
      outlineSize=0
      outlineColor=<rgba 0 0 0 1>
     }
    }
    shadowOffset=<0 0>
    strikeBrush {
     bIsDynamicallyLoaded=false
     drawAs=Image
     imageType=NoImage
     mirroring=NoMirror
     resourceName=None
     resourceObject=null
     tiling=NoTile
     imageSize=<32 32>
     margin=<ltrb 0 0 0 0>
     outlineSettings {
      bUseBrushTransparency=false
      roundingType=HalfHeightRadius
      width=0
      color {colorUseRule=UseColor_Specified, specifiedColor=<rgba 0 0 0 0>}
      cornerRadii=<0 0 0 0>
     }
     tintColor {colorUseRule=UseColor_Specified, specifiedColor=<rgba 0.6308 0.9131 1 1>}
     uVRegion {bIsValid=false, max=<0 0>, min=<0 0>}
    }
   }
   slot {class=VerticalBoxSlot, properties {horizontalAlignment=HAlign_Center, verticalAlignment=VAlign_Top}}
  }
  Overlay_AnimationContainer {
   widgetClassPath=&Overlay
   parent=MainContainer
   widget=Overlay_AnimationContainer
   slot {class=VerticalBoxSlot, properties {horizontalAlignment=HAlign_Center, padding=<ltrb 0 40 0 0>, size {sizeRule=Fill, value=0.61}}}
  }
  AnimationContainer {
   widgetClassPath=&VerticalBox
   bIsVariable=true
   parent=Overlay_AnimationContainer
   widget=AnimationContainer
   slot {class=OverlaySlot, properties {horizontalAlignment=HAlign_Fill, verticalAlignment=VAlign_Fill}}
  }
  CommonVisualAttachment_129 {
   widgetClassPath=&CommonVisualAttachment
   parent=AnimationContainer
   widget=CommonVisualAttachment_129
   properties {contentAnchor=<0.5 -0.1>}
   slot {class=VerticalBoxSlot, properties {horizontalAlignment=HAlign_Center, verticalAlignment=VAlign_Top}}
  }
  WBP_FMCalibration_CompleteDisplay {
   widgetClassPath=&/WBP_FMCalibration_CompleteDisplay
   bIsVariable=true
   parent=CommonVisualAttachment_129
   widget=WBP_FMCalibration_CompleteDisplay
   slot {class=SizeBoxSlot, properties {horizontalAlignment=HAlign_Center, verticalAlignment=VAlign_Center}}
  }
  ScaleBox_0 {
   widgetClassPath=&ScaleBox
   parent=AnimationContainer
   widget=ScaleBox_0
   properties {stretch=ScaleToFitY}
   slot {class=VerticalBoxSlot, properties {horizontalAlignment=HAlign_Center, verticalAlignment=VAlign_Bottom, size {sizeRule=Fill, value=1}}}
  }
  PressCueAnimation {
   widgetClassPath=&Image
   bIsVariable=true
   parent=ScaleBox_0
   widget=PressCueAnimation
   properties {brush {
     bIsDynamicallyLoaded=false
     drawAs=Image
     imageType=FullColor
     mirroring=NoMirror
     resourceName=None
     resourceObject=&MI_UI_FTUE_SlidingNotes_Anim
     tiling=NoTile
     imageSize=<400 400>
     margin=<ltrb 0 0 0 0>
     outlineSettings {
      bUseBrushTransparency=false
      roundingType=HalfHeightRadius
      width=0
      color {colorUseRule=UseColor_Specified, specifiedColor=<rgba 0 0 0 0>}
      cornerRadii=<0 0 0 0>
     }
     tintColor {colorUseRule=UseColor_Specified, specifiedColor=<rgba 1 1 1 1>}
     uVRegion {bIsValid=false, max=<0 0>, min=<0 0>}
    }}
   slot {class=ScaleBoxSlot}
  }
  CommonHardwareVisibilityBorder_1 {
   widgetClassPath=&CommonHardwareVisibilityBorder
   parent=Overlay_AnimationContainer
   widget=CommonHardwareVisibilityBorder_1
   properties {visibilityQuery {
     autoDescription=" ANY( Input.Touch )"
     tokenStreamVersion=0
     userDescription=""
     queryTokenStream [0,1,1,1,0]
     tagDictionary [{tagName=Input.Touch}]
    }}
   slot {class=OverlaySlot, properties {horizontalAlignment=HAlign_Center, padding=<ltrb 0 43 0 0>}}
  }
  Text_SubBody {
   widgetClassPath=&CommonTextBlock
   bIsVariable=true
   parent=CommonHardwareVisibilityBorder_1
   widget=Text_SubBody
   properties {
    justification=Center
    text=TAP ANYWHERE TO BEGIN
    wrapTextAt=900
    colorAndOpacity {colorUseRule=UseColor_Specified, specifiedColor=<rgba 0.4397 0.7835 0.8879 1>}
    font {
     bForceMonospaced=false
     bMaterialIsStencil=false
     fontMaterial=null
     fontObject=&HeadingNow
     letterSpacing=0
     monospacedWidth=1
     size=31.5
     skewAmount=0
     typefaceFontName=76 Bold Italic
     outlineSettings {
      bApplyOutlineToDropShadows=false
      bMiteredCorners=false
      bSeparateFillAlpha=false
      outlineMaterial=null
      outlineSize=0
      outlineColor=<rgba 0 0 0 1>
     }
    }
    shadowOffset=<0 0>
   }
   slot {class=BorderSlot, properties {padding=<ltrb 0 0 0 0>}}
  }
  WBP_FMCalibration_ProgressBar_C_0 {
   widgetClassPath=&/WBP_FMCalibration_ProgressBar
   bIsVariable=true
   parent=MainContainer
   widget=WBP_FMCalibration_ProgressBar_C_0
   properties {bOverride_Cursor=true}
   slot {
    class=VerticalBoxSlot
    properties {horizontalAlignment=HAlign_Center, verticalAlignment=VAlign_Bottom, padding=<ltrb 0 15 0 45>, size {sizeRule=Fill, value=0.06}}
   }
  }
  SizeBox_ButtonGroup_Preamble {
   widgetClassPath=&SizeBox
   bIsVariable=true
   parent=MainContainer
   widget=SizeBox_ButtonGroup_Preamble
   properties {bOverride_WidthOverride=true, widthOverride=408}
   slot {
    class=VerticalBoxSlot
    properties {horizontalAlignment=HAlign_Center, verticalAlignment=VAlign_Center, padding=<ltrb 0 0 0 30>, size {sizeRule=Automatic, value=0.18}}
   }
  }
  Button_Kit_StartCal {
   widgetClassPath=/FortUILibrary/UIKit/Buttons/WBP_UIKit_Button_Loud
   bIsVariable=true
   parent=SizeBox_ButtonGroup_Preamble
   widget=Button_Kit_StartCal
   properties {
    buttonSize=Large
    minWidthOverride=420
    text=Begin
    triggeringInputAction {dataTable=&FMCalibration_InputActionDataTable, rowName=FMCalibrateSwing}
   }
   slot {class=SizeBoxSlot}
  }
  CommonHardwareVisibilityBorder_0 {
   widgetClassPath=&CommonHardwareVisibilityBorder
   parent=Overlay_MainContainer
   widget=CommonHardwareVisibilityBorder_0
   properties {visibilityQuery {
     autoDescription=" ANY( Input.Touch )"
     tokenStreamVersion=0
     userDescription=""
     queryTokenStream [0,1,1,1,0]
     tagDictionary [{tagName=Input.Touch}]
    }}
   slot {class=OverlaySlot, properties {horizontalAlignment=HAlign_Fill, verticalAlignment=VAlign_Fill}}
  }
  WBP_PolygonalInputZoneController_CalibrateSwing {
   widgetClassPath=/PolygonalInputZones/Assets/WBP_PolygonalInputZoneController
   bIsVariable=true
   parent=CommonHardwareVisibilityBorder_0
   widget=WBP_PolygonalInputZoneController_CalibrateSwing
   properties {
    interactionMode=InputActions
    inputZones [{
      activated=true
      deactivateOnRelease=false
      priority=0
      zoneAction=&IA_FMCalibrateSwing
      zoneId=""
      zoneWidget=null
      zoneVertices [<0 0>, <0 1>, <1 1>, <1 0>]
     }]
   }
   slot {class=BorderSlot, properties {padding=<ltrb 0 0 0 0>}}
  }
  WBP_CalibrometerDebug {
   widgetClassPath=&/WBP_CalibrometerWidget
   bIsVariable=true
   parent=Overlay_MainContainer
   widget=WBP_CalibrometerDebug
   slot {class=OverlaySlot, properties {horizontalAlignment=HAlign_Right, verticalAlignment=VAlign_Bottom, padding=<ltrb 0 0 200 0>}}
  }
 ]
}
